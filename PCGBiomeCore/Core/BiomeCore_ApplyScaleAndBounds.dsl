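; PCG subgraph for Biome Core that applies an asset's bounds, transform and scale settings to incoming points.
; It keeps only the needed Asset attributes, then writes BoundsMin and BoundsMax onto each point.
; Optional branches compose the asset Transform with the point transform (ApplyAssetTransform, off by default) and overwrite point scale with Transform.Scale (SetScale, on by default).
; Finally, it always multiplies point extents by ExtentsMultiplier.
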(graph BiomeCore_ApplyScaleAndBounds
  (params (SetScale :type Bool :default true) (ApplyAssetTransform :type Bool))
  (inputs (Asset :type Param))
  (bind DeleteAttributes_5 (Delete%20Attributes :bTokenizeOnWhiteSpace true :Operation KeepSelectedAttributes :SelectedAttributes "BoundsMin,BoundsMax,Transform,ExtentsMultiplier"))
  (bind AttributeMathsOp_1 (Attribute%20Maths%20Op :InputSource1 $BoundsMin :InputSource2 BoundsMin :Operation Set))
  (bind AttributeMathsOp_10 (Attribute%20Maths%20Op :InputSource1 $BoundsMax :InputSource2 BoundsMax :Operation Set))
  (bind Branch_1 (Branch :bOutputToB-param ApplyAssetTransform))
  (bind AttributeTransformOp_4 (Attribute%20Transform%20Op :InputSource1 Transform :InputSource2 $Transform))
  (bind Branch_0 (Branch :bOutputToB-param SetScale))
  (bind AttributeMathsOp_0 (Attribute%20Maths%20Op :InputSource1 $Scale :InputSource2 Transform.Scale :Operation Set))
  (bind AttributeMathsOp_2 (Attribute%20Maths%20Op :InputSource1 $Extents :InputSource2 ExtentsMultiplier :Operation Multiply))
  (connect Branch_0 Output%20B AttributeMathsOp_0 InA)
  (connect DeleteAttributes_5 Out AttributeMathsOp_0 InB)
  (connect DefaultInputNode In AttributeMathsOp_1 InA)
  (connect DeleteAttributes_5 Out AttributeMathsOp_1 InB)
  (connect AttributeMathsOp_0 Out AttributeMathsOp_2 InA)
  (connect Branch_0 Output%20A AttributeMathsOp_2 InA)
  (connect DeleteAttributes_5 Out AttributeMathsOp_2 InB)
  (connect AttributeMathsOp_1 Out AttributeMathsOp_10 InA)
  (connect DeleteAttributes_5 Out AttributeMathsOp_10 InB)
  (connect DeleteAttributes_5 Out AttributeTransformOp_4 InA)
  (connect Branch_1 Output%20B AttributeTransformOp_4 InB)
  (connect AttributeTransformOp_4 Out Branch_0 In)
  (connect Branch_1 Output%20A Branch_0 In)
  (connect AttributeMathsOp_10 Out Branch_1 In)
  (connect DefaultInputNode Asset DeleteAttributes_5 In)
  (output Out AttributeMathsOp_2))
