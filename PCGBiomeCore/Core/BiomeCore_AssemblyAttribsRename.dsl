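; PCG subgraph for Biome Core that normalizes assembly attribute names and types on its input data.
; A linear chain of Attribute Rename nodes drops the "b" prefix from shadow, HLOD, lighting, ray tracing, navigation and rasterization flags.
; The collision flag is the exception in naming, as bUseDefaultCollision becomes AllowCollision.
; Two final in-place casts convert DetailMode to Integer32 and HLODLayer to SoftObjectPath before output.
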(graph BiomeCore_AssemblyAttribsRename
  (bind AttributeRename_0 (Attribute%20Rename :AttributeToRename bCastShadow :NewAttributeName CastShadow))
  (bind AttributeRename_1 (Attribute%20Rename :AttributeToRename bCastHiddenShadow :NewAttributeName CastHiddenShadow))
  (bind AttributeRename_2 (Attribute%20Rename :AttributeToRename bIncludeInHLOD :NewAttributeName IncludeInHLOD))
  (bind AttributeRename_3 (Attribute%20Rename :AttributeToRename bAffectDistanceFieldLighting :NewAttributeName AffectDistanceFieldLighting))
  (bind AttributeRename_5 (Attribute%20Rename :AttributeToRename bAffectDynamicIndirectLighting :NewAttributeName AffectDynamicIndirectLighting))
  (bind AttributeRename_6 (Attribute%20Rename :AttributeToRename bVisibleInRayTracing :NewAttributeName VisibleInRayTracing))
  (bind AttributeRename_4 (Attribute%20Rename :AttributeToRename bUseDefaultCollision :NewAttributeName AllowCollision))
  (bind AttributeRename_7 (Attribute%20Rename :AttributeToRename bCanEverAffectNavigation :NewAttributeName CanEverAffectNavigation))
  (bind AttributeRename_8 (Attribute%20Rename :AttributeToRename bRasterizeAsFilledConvexVolume :NewAttributeName RasterizeAsFilledConvexVolume))
  (bind AttributeCast_1 (Attribute%20Cast :InputSource DetailMode :OutputTarget DetailMode :OutputType Integer32))
  (bind AttributeCast_0 (Attribute%20Cast :InputSource HLODLayer :OutputTarget HLODLayer :OutputType SoftObjectPath))
  (connect AttributeCast_1 Out AttributeCast_0 In)
  (connect AttributeRename_8 Out AttributeCast_1 In)
  (connect DefaultInputNode In AttributeRename_0 In)
  (connect AttributeRename_0 Out AttributeRename_1 In)
  (connect AttributeRename_1 Out AttributeRename_2 In)
  (connect AttributeRename_2 Out AttributeRename_3 In)
  (connect AttributeRename_6 Out AttributeRename_4 In)
  (connect AttributeRename_3 Out AttributeRename_5 In)
  (connect AttributeRename_5 Out AttributeRename_6 In)
  (connect AttributeRename_4 Out AttributeRename_7 In)
  (connect AttributeRename_7 Out AttributeRename_8 In)
  (output Out AttributeCast_0))
