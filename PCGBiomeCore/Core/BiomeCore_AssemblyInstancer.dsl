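; PCG Biome Core graph that expands assembly points into instance points from their referenced PCG data asset.
; Loaded points get normalized attributes, mesh option presets and a collision override, then are copied onto targets carrying a unique assembly hash.
; An optional, bypassable AssemblyBehavior subgraph adjusts results using terrain, and points lacking mesh references are dropped.
; Depending on GenerationMethod, meshes also resolve to hashed "_MS" runtime assembly paths, output as RuntimeAssetTables.
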(graph BiomeCore_AssemblyInstancer
  (params (AssemblyBehavior :type (Object /Script/PCG.PCGGraphInterface)) (Debug-BypassAssemblyBehavior :type Bool) (DefaultTerrainGetterSetting :type (Object /Script/PCG.PCGSettingsInterface)))
  (inputs (AssetTable :type Param))
  (outputs (RuntimeAssetTables :type Param))
  (bind Subgraph_0 (Subgraph :Graph /PCGBiomeCore/AssemblyBehavior/BiomeCore_AssemblyBehavior_Template.BiomeCore_AssemblyBehavior_Template :SubgraphOverride-param AssemblyBehavior))
  (bind AttributeReduce_1 (Attribute%20Reduce :InputSource RequiredAttributes :Operation Join))
  (bind DeleteTags_1 (Delete%20Tags :SelectedTags "Assembly"))
  (bind GetPointsCount_1 (Get%20Element%20Count))
  (bind Branch_2 (Branch))
  (bind GetAttributeFromPointIndex_0 (Get%20Attribute%20From%20Point%20Index :bForceOutputAttributeToBeInElementsDomain true :InputSource Assembly :OutputAttributeName None))
  (bind PCGLoadDataAsset_3 (Load%20PCG%20Data%20Asset :bWarnIfNoAsset false))
  (bind FilterByTag_2 (Filter%20Data%20By%20Tag :bTokenizeOnWhiteSpace true :SelectedTags "Points"))
  (bind ToPoint_4 (To%20Point))
  (bind PointToAttributeSet_1 (Point%20To%20Attribute%20Set))
  (bind FilterDataByAttribute_3 (Filter%20Data%20By%20Attribute :Attribute SkeletalMesh :bIgnoreProperties true))
  (bind AddAttribute_0 (Add%20Attribute :AttributeTypes (struct :Type SoftObjectPath) :OutputTarget SkeletalMesh))
  (bind FilterDataByAttribute_2 (Filter%20Data%20By%20Attribute :Attribute Actor :bIgnoreProperties true))
  (bind AddAttribute_3 (Add%20Attribute :AttributeTypes (struct :Type SoftClassPath) :OutputTarget Actor))
  (bind FilterDataByAttribute_0 (Filter%20Data%20By%20Attribute :Attribute HLODLayer :bIgnoreProperties true))
  (bind AddAttribute_2 (Add%20Attribute :AttributeTypes (struct :Type SoftObjectPath) :OutputTarget HLODLayer))
  (bind Subgraph_1 (Subgraph :Graph /PCGBiomeCore/Core/BiomeCore_AssemblyAttribsRename.BiomeCore_AssemblyAttribsRename))
  (bind DeleteAttributes_9 (Delete%20Attributes :SelectedAttributes "MeshOptionsPreset"))
  (bind DeleteAttributes_0 (Delete%20Attributes :Operation KeepSelectedAttributes :SelectedAttributes "AssemblyAllowCollision, DataIndex"))
  (bind FilterElementsByIndex_0 (Filter%20Elements%20By%20Index :bSelectIndicesByInput false :SelectedIndices "0"))
  (bind PointToAttributeSet_0 (Point%20To%20Attribute%20Set))
  (bind AddAttribute_4 (Add%20Attribute :InputSource DataIndex :OutputTarget DataIndex))
  (bind ExtractAttribute_0 (Extract%20Attribute%20at%20Index :bForceOutputAttributeToBeInElementsDomain true :InputSource AssemblyAllowCollision))
  (bind Branch_1 (Branch))
  (bind DeleteAttributes_5 (Delete%20Attributes :bTokenizeOnWhiteSpace true :Operation KeepSelectedAttributes :SelectedAttributes "Mesh,Material,CastShadow,Visible,CastHiddenShadow,IncludeInHLOD,HLODLayer,WorldPositionOffsetDisableDistance,DetailMode,AffectDistanceFieldLighting,SkeletalMesh,AllowCollision,CastDynamicShadow,EvaluateWorldPositionOffset,IsSpatiallyLoaded,ComponentTags"))
  (bind CopyPoints_2 (Copy%20Points :TagInheritance Target))
  (bind AttributeCompareOp_0 (Attribute%20Compare%20Op :Operation NotEqual :InA AssemblyBehavior))
  (bind Branch_0 (Branch))
  (bind Proxy_0 (Proxy :Settings-param DefaultTerrainGetterSetting))
  (bind FilterDataByType_0 (Filter%20Data%20By%20Type :TargetType (struct :Ids [(struct :Struct /Script/PCG.PCGDataTypeInfoSurface)])))
  (bind Subgraph_2 (Subgraph :Graph /PCGBiomeCore/AssemblyBehavior/BiomeCore_AssemblyBehavior_Template.BiomeCore_AssemblyBehavior_Template :SubgraphOverride-param AssemblyBehavior))
  (bind AttributeBooleanOp_0 (Attribute%20Boolean%20Op :Operation Not :In Debug-BypassAssemblyBehavior))
  (bind AttributeBooleanOp_1 (Attribute%20Boolean%20Op))
  (bind FilterDataByAttribute_1 (Filter%20Data%20By%20Attribute :Attribute RecursionLevel))
  (bind GetAttributeFromPointIndex_5 (Get%20Attribute%20From%20Point%20Index :bForceOutputAttributeToBeInElementsDomain true :InputSource AssetID :OutputAttributeName AssetID))
  (bind Branch_7 (Branch))
  (bind AttributeFilter_5 (Filter%20Attribute%20Elements :bUseSpatialQuery true :Operator Equal :TargetAttribute AssetID :ThresholdAttribute AssetID))
  (bind DeleteAttributes_1 (Delete%20Attributes :Operation KeepSelectedAttributes :SelectedAttributes "GenerationMethod"))
  (bind AttributeCompareOp_1 (Attribute%20Compare%20Op :InputSource1 Mesh :Operation NotEqual :OutputTarget ValidMeshRef))
  (bind AttributeCompareOp_3 (Attribute%20Compare%20Op :InputSource1 SkeletalMesh :Operation NotEqual :OutputTarget ValidSkeletalMeshRef))
  (bind AttributeBooleanOp_2 (Attribute%20Boolean%20Op :InputSource1 ValidMeshRef :InputSource2 ValidSkeletalMeshRef :Operation Or :OutputTarget ValidRef))
  (bind AttributeFilter_2 (Filter%20Attribute%20Elements :AttributeTypes (struct :Type Boolean :BoolValue true) :bUseConstantThreshold true :Operator Equal :TargetAttribute ValidRef :ThresholdAttribute @Last))
  (bind DeleteAttributes_2 (Delete%20Attributes :SelectedAttributes "ValidMeshRef, ValidSkeletalMeshRef, ValidRef"))
  (bind Branch_6 (Branch))
  (bind CreateAttribute_0 (Create%20Constant :AttributeTypes (struct :Type String :StringValue "_MS")))
  (bind AttributePartition_1 (Attribute%20Partition :PartitionAttributeSelectors [Mesh]))
  (bind GetAttributeFromPointIndex_8 (Get%20Attribute%20From%20Point%20Index :bForceOutputAttributeToBeInElementsDomain true :InputSource Mesh :OutputAttributeName Mesh))
  (bind MergeAttributes_0 (Merge%20Attributes))
  (bind AttributeStringOp_1 (Attribute%20String%20Op :InputSource1 Mesh :OutputTarget RuntimeAssembly))
  (bind CreateAttribute_5 (Create%20Constant :AttributeTypes (struct :Type String :StringValue ".")))
  (bind AttributeStringOp_7 (Attribute%20String%20Op))
  (bind AttributeStringOp_8 (Attribute%20String%20Op :InputSource1 RuntimeAssembly :Operation Replace))
  (bind HashAttribute_1 (Hash%20Attribute :InputSource1 RuntimeAssembly :OutputTarget RuntimeAssemblyHashValue))
  (bind DeleteAttributes_3 (Delete%20Attributes :SelectedAttributes "RuntimeAssembly"))
  (bind MatchAndSetAttributes_0 (Match%20And%20Set%20Attributes :bKeepUnmatched false :bMatchAttributes true :InputAttribute Mesh :MatchAttribute Mesh))
  (bind AddTags_1 (Add%20Tags :TagsToAdd "Assembly:Point"))
  (bind DeleteAttributes_4 (Delete%20Attributes :SelectedAttributes "Mesh"))
  (bind DataNum_1 (Data%20Count))
  (bind GetAttributeFromPointIndex_1 (Get%20Attribute%20From%20Point%20Index :bForceOutputAttributeToBeInElementsDomain true :InputSource GenerationMethod :OutputAttributeName GenerationMethod))
  (bind Select_1 (Select))
  (bind AttributeCompareOp_2 (Attribute%20Compare%20Op))
  (bind DeleteAttributes_8 (Delete%20Attributes :Operation KeepSelectedAttributes :SelectedAttributes "RecursionLevel,DataIndex,AssemblyHashValue"))
  (bind AddTags_2 (Add%20Tags :TagsToAdd "Assembly:Source"))
  ;[0 0 0 0.5] Getting optional MeshOptionsPresets from assembly points
  (bind AttributeCompareOp_4 (Attribute%20Compare%20Op :InputSource1 MeshOptionsPreset :Operation NotEqual))
  (bind AttributePartition_2 (Attribute%20Partition :PartitionAttributeSelectors [$Index]))
  (bind GetPropertyFromObjectPath_0 (Get%20Property%20From%20Object%20Path :bSanitizeOutputAttributeName false :InputSource MeshOptionsPreset :OutputAttributeName None :PropertyName MeshOptionsPreset))
  (bind CopyAttributes_1 (Copy%20Attributes :InputSource MeshOptionsPreset))
  (bind MergeAttributes_2 (Merge%20Attributes))
  (bind CopyAttributes_2 (Copy%20Attributes :InputSource $Index :OutputTarget IndexBck))
  (bind AttributeFilter_1 (Filter%20Attribute%20Elements :AttributeTypes (struct :Type Boolean :BoolValue true) :bUseConstantThreshold true :Operator Equal :TargetAttribute MeshOptionsPreset :ThresholdAttribute @Last))
  (bind DeleteAttributes_7 (Delete%20Attributes :SelectedAttributes "MeshOptionsPreset"))
  ;[0 0 0 0.5] Apply MeshOptionsPresets
  (bind MatchAndSetAttributes_1 (Match%20And%20Set%20Attributes :bMatchAttributes true :InputAttribute $Index :MatchAttribute IndexBck))
  ;[0 0 0 0.5] AssemblyAllowCollision Master Override
  (bind Branch_3 (Branch))
  (bind AttributeMathsOp_1 (Attribute%20Maths%20Op :InputSource1 AllowCollision :InputSource2 AssemblyAllowCollision :Operation Set))
  ;[0 0 0 0.5] Generate Unique Assembly Hash
  (bind GetLoopIndex_1 (Get%20Loop%20Index))
  (bind AttributeStringOp_2 (Attribute%20String%20Op))
  (bind AttributeStringOp_3 (Attribute%20String%20Op :OutputTarget AssemblyHashValue))
  (bind AttributeStringOp_4 (Attribute%20String%20Op))
  (bind GetActorData_0 (Get%20Actor%20Data :Mode GetActorReference))
  (bind AttributeStringOp_6 (Attribute%20String%20Op))
  (bind HashAttribute_2 (Hash%20Attribute))
  (bind DeleteAttributes_6 (Delete%20Attributes :SelectedAttributes "GenerationMethod,AssetID,Assembly,Mesh,Actor,Material,AssemblyAllowCollision"))
  ;[0.15 0.15 0.15 0.5] Needed for Behavior graph
  (bind CopyAttributes_0 (Copy%20Attributes :InputSource $Index :OutputTarget TargetIndex))
  (reroute Auto%20Resolve%20Runtime%20Assemblies AttributeCompareOp_2)
  (reroute HasAssemblyBehavior AttributeBooleanOp_1)
  (connect FilterDataByAttribute_3 OutsideFilter AddAttribute_0 In)
  (connect FilterDataByAttribute_0 OutsideFilter AddAttribute_2 In)
  (connect FilterDataByAttribute_2 OutsideFilter AddAttribute_3 In)
  (connect PointToAttributeSet_0 Out AddAttribute_4 Attributes)
  (connect MatchAndSetAttributes_1 Out AddAttribute_4 In)
  (connect Branch_6 Output%20A AddTags_1 In)
  (connect MatchAndSetAttributes_0 Out AddTags_1 In)
  (connect DeleteAttributes_8 Out AddTags_2 In)
  (connect AttributeBooleanOp_0 Out AttributeBooleanOp_1 InA)
  (connect AttributeCompareOp_0 Out AttributeBooleanOp_1 InB)
  (connect AttributeCompareOp_3 Out AttributeBooleanOp_2 InA)
  (connect AttributeCompareOp_3 Out AttributeBooleanOp_2 InB)
  (connect Branch_0 Output%20A AttributeCompareOp_1 InA)
  (connect Subgraph_2 Out AttributeCompareOp_1 InA)
  (connect Select_1 Out AttributeCompareOp_2 InA)
  (connect AttributeCompareOp_1 Out AttributeCompareOp_3 InA)
  (connect PointToAttributeSet_1 Out AttributeCompareOp_4 InA)
  (connect CopyAttributes_2 Out AttributeFilter_1 In)
  (connect AttributeBooleanOp_2 Out AttributeFilter_2 In)
  (connect GetAttributeFromPointIndex_5 Attribute AttributeFilter_5 Filter)
  (connect Branch_7 Output%20B AttributeFilter_5 In)
  (connect Branch_3 Output%20A AttributeMathsOp_1 InA)
  (connect ExtractAttribute_0 Out AttributeMathsOp_1 InB)
  (connect Branch_6 Output%20B AttributePartition_1 In)
  (connect AttributeCompareOp_4 Out AttributePartition_2 In)
  (connect Subgraph_0 RequiredAttributes AttributeReduce_1 In)
  (connect MergeAttributes_0 Out AttributeStringOp_1 InA)
  (connect CreateAttribute_0 Out AttributeStringOp_1 InB)
  (connect GetLoopIndex_1 Out AttributeStringOp_2 InA)
  (connect CopyAttributes_0 Out AttributeStringOp_3 InA)
  (connect AttributeStringOp_3 Out AttributeStringOp_4 InA)
  (connect AttributeStringOp_2 Out AttributeStringOp_4 InB)
  (connect AttributeStringOp_4 Out AttributeStringOp_6 InA)
  (connect GetActorData_0 Out AttributeStringOp_6 InB)
  (connect CreateAttribute_0 Out AttributeStringOp_7 InA)
  (connect CreateAttribute_5 Out AttributeStringOp_7 InB)
  (connect AttributeStringOp_1 Out AttributeStringOp_8 In)
  (connect AttributeStringOp_7 Out AttributeStringOp_8 Replace)
  (connect CreateAttribute_5 Out AttributeStringOp_8 Search)
  (connect HasAssemblyBehavior Out Branch_0 bOutputToB)
  (connect CopyPoints_2 Out Branch_0 In)
  (connect HasAssemblyBehavior Out Branch_1 bOutputToB)
  (connect AttributeMathsOp_1 Out Branch_1 In)
  (connect Branch_3 Output%20B Branch_1 In)
  (connect GetPointsCount_1 Out Branch_2 bOutputToB)
  (connect DeleteTags_1 Out Branch_2 In)
  (connect ExtractAttribute_0 Out Branch_3 bOutputToB)
  (connect AddAttribute_4 Out Branch_3 In)
  (connect Auto%20Resolve%20Runtime%20Assemblies Out Branch_6 bOutputToB)
  (connect DeleteAttributes_2 Out Branch_6 In)
  (connect GetPointsCount_1 Out Branch_7 bOutputToB)
  (connect DefaultInputNode AssetTable Branch_7 In)
  (connect Branch_2 Output%20B CopyAttributes_0 Source)
  (connect Branch_2 Output%20B CopyAttributes_0 Target)
  (connect AttributePartition_2 Out CopyAttributes_1 Source)
  (connect GetPropertyFromObjectPath_0 Out CopyAttributes_1 Target)
  (connect MergeAttributes_2 Out CopyAttributes_2 Source)
  (connect MergeAttributes_2 Out CopyAttributes_2 Target)
  (connect Branch_1 Output%20B CopyPoints_2 Source)
  (connect DeleteAttributes_5 Out CopyPoints_2 Source)
  (connect DeleteAttributes_6 Out CopyPoints_2 Target)
  (connect FilterDataByAttribute_1 InsideFilter DataNum_1 In)
  (connect Branch_2 Output%20B DeleteAttributes_0 In)
  (connect AttributeFilter_5 InsideFilter DeleteAttributes_1 In)
  (connect AttributeFilter_2 InsideFilter DeleteAttributes_2 In)
  (connect HashAttribute_1 Out DeleteAttributes_3 In)
  (connect HashAttribute_1 Out DeleteAttributes_4 In)
  (connect Branch_1 Output%20A DeleteAttributes_5 In)
  (connect HashAttribute_2 Out DeleteAttributes_6 In)
  (connect AttributeFilter_1 InsideFilter DeleteAttributes_7 In)
  (connect DeleteAttributes_6 Out DeleteAttributes_8 In)
  (connect Subgraph_1 Out DeleteAttributes_9 In)
  (connect DefaultInputNode In DeleteTags_1 In)
  (connect PointToAttributeSet_0 Out ExtractAttribute_0 In)
  (connect PCGLoadDataAsset_3 Out FilterByTag_2 In)
  (connect AddAttribute_3 Out FilterDataByAttribute_0 In)
  (connect FilterDataByAttribute_2 InsideFilter FilterDataByAttribute_0 In)
  (connect Branch_2 Output%20B FilterDataByAttribute_1 In)
  (connect AddAttribute_0 Out FilterDataByAttribute_2 In)
  (connect FilterDataByAttribute_3 InsideFilter FilterDataByAttribute_2 In)
  (connect ToPoint_4 Out FilterDataByAttribute_3 In)
  (connect Proxy_0 Out FilterDataByType_0 In)
  (connect DeleteAttributes_0 Out FilterElementsByIndex_0 In)
  (connect Branch_2 Output%20B GetAttributeFromPointIndex_0 In)
  (connect FilterDataByAttribute_1 InsideFilter GetAttributeFromPointIndex_1 In)
  (connect FilterDataByAttribute_1 OutsideFilter GetAttributeFromPointIndex_5 In)
  (connect AttributePartition_1 Out GetAttributeFromPointIndex_8 In)
  (connect DeleteTags_1 Out GetPointsCount_1 In)
  (connect PointToAttributeSet_1 Out GetPropertyFromObjectPath_0 In)
  (connect AttributeStringOp_8 Out HashAttribute_1 In)
  (connect AttributeStringOp_6 Out HashAttribute_2 In)
  (connect Branch_6 Output%20B MatchAndSetAttributes_0 In)
  (connect DeleteAttributes_3 Out MatchAndSetAttributes_0 Match%20Data)
  (connect DeleteAttributes_9 Out MatchAndSetAttributes_1 In)
  (connect DeleteAttributes_7 Out MatchAndSetAttributes_1 Match%20Data)
  (connect GetAttributeFromPointIndex_8 Attribute MergeAttributes_0 In)
  (connect CopyAttributes_1 Out MergeAttributes_2 In)
  (connect GetAttributeFromPointIndex_0 Attribute PCGLoadDataAsset_3 Asset)
  (connect AttributeReduce_1 Out PCGLoadDataAsset_3 CommaSeparatedDefaultAttributeOverrides)
  (connect FilterElementsByIndex_0 In%20Filter PointToAttributeSet_0 In)
  (connect ToPoint_4 Out PointToAttributeSet_1 In)
  (connect DataNum_1 Out Select_1 bUseInputB)
  (connect DeleteAttributes_1 Out Select_1 Input%20A)
  (connect GetAttributeFromPointIndex_1 Attribute Select_1 Input%20B)
  (connect AddAttribute_2 Out Subgraph_1 In)
  (connect FilterDataByAttribute_0 InsideFilter Subgraph_1 In)
  (connect Branch_0 Output%20B Subgraph_2 In)
  (connect FilterDataByType_0 InsideFilter Subgraph_2 TerrainSurface)
  (connect FilterByTag_2 InsideFilter ToPoint_4 In)
  (output Out AddTags_1)
  (output Out AddTags_2)
  (output RuntimeAssetTables DeleteAttributes_4))
