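; PCG biome core subgraph that assigns assets from a root asset table to input points by generator.
; The table is trimmed to asset attributes, partitioned by Generator and GeneratorSubType, then run through a tagging loop.
; Points are partitioned by GeneratorSubType and fed with the tagged asset sets into a per-generator loop.
; Output points keep only selected attributes: AssetID, DataIndex and density, water distance, sun, flow and height ranges.
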(graph BiomeCore_AssignAssetsToPoints
  (inputs (PointsIn :type Point) (RootAssetTable :type Param))
  (outputs (PointsOut :type Point))
  (bind DeleteAttributes_0 (Delete%20Attributes :bTokenizeOnWhiteSpace true :Operation KeepSelectedAttributes :SelectedAttributes "AssetID,,DensityMin,DensityMax,WaterDistanceMin,WaterDistanceMax,SunExposureMin,SunExposureMax,FlowMin,FlowMax,HeightMin,HeightMax,DataIndex"))
  ;[0 0 0 1 32] Assign Assets by Generator
  (bind DeleteAttributes_1 (Delete%20Attributes :bTokenizeOnWhiteSpace true :Operation KeepSelectedAttributes :SelectedAttributes "AssetType,Generator,GeneratorType,GeneratorSubType,AssetID,BiomeColor,Weight,BoundsMin,BoundsMax,Transform,OrientUpward,ExtentsMultiplier,,DensityMin,DensityMax,WaterDistanceMin,WaterDistanceMax,SunExposureMin,SunExposureMax,FlowMin,FlowMax,HeightMin,HeightMax"))
  (bind AttributePartition_2 (Attribute%20Partition :bTokenizeOnWhiteSpace true :PartitionAttributeSelectors [Generator GeneratorSubType]))
  (bind Loop_0 (Loop :bTokenizeOnWhiteSpace true :Graph /PCGBiomeCore/Core/BiomeCore_TagAssetsWithGenerators.BiomeCore_TagAssetsWithGenerators))
  (bind AttributePartition_3 (Attribute%20Partition :bTokenizeOnWhiteSpace true :PartitionAttributeSelectors [GeneratorSubType]))
  (bind Loop_1 (Loop :bTokenizeOnWhiteSpace true :Graph /PCGBiomeCore/Core/BiomeCore_PerGeneratorLoop.BiomeCore_PerGeneratorLoop))
  (connect DeleteAttributes_1 Out AttributePartition_2 In)
  (connect DefaultInputNode PointsIn AttributePartition_3 In)
  (connect Loop_1 DefinedPoints DeleteAttributes_0 In)
  (connect DefaultInputNode RootAssetTable DeleteAttributes_1 In)
  (connect AttributePartition_2 Out Loop_0 AttributeSetsToTag)
  (connect AttributePartition_3 Out Loop_1 PointDataByGenerator)
  (connect Loop_0 TaggedAttributeSets Loop_1 TaggedAttributeSets)
  (output PointsOut DeleteAttributes_0))
