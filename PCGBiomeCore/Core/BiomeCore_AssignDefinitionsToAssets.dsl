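; PCG subgraph for Biome Core that stamps biome assets with data from their biome definitions.
; If either input is empty it passes nothing on; otherwise it copies BiomeColor and BiomeID from the definitions onto the assets.
; Assets without a Generator are filtered out unless AllowNoGenerator is set.
; It then reads the Generator property from each referenced object and writes it as a "GEN:"-prefixed GeneratorType attribute.
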(graph BiomeCore_AssignDefinitionsToAssets
  (params (AllowNoGenerator :type Bool))
  (inputs (BiomeDefinitions :type Param) (BiomeAssets :type Param))
  (outputs (AssetsWithDefinitions :type Param))
  ;[0 0 0] Skip empty inputs
  (bind GetEntriesCount_16 (Get%20Element%20Count))
  (bind AttributeReduce_17 (Attribute%20Reduce :bMergeOutputAttributes true :Operation Min))
  (bind CreateAttribute_7 (Create%20Constant :AttributeTypes (struct :Type Integer32)))
  (bind AttributeCompareOp_5 (Attribute%20Compare%20Op))
  (bind Select_8 (Select :bUseInputB true))
  (bind Select_10 (Select :bUseInputB true))
  ;[0 0 0] Add BiomeColor
  (bind AddAttribute_6 (Add%20Attribute :InputSource BiomeColor.rgb :OutputTarget BiomeColor))
  ;[0 0 0 0.5] Add BiomeID
  (bind FilterDataByAttribute_0 (Filter%20Data%20By%20Attribute :Attribute BiomeID))
  (bind AddAttribute_0 (Add%20Attribute :AttributeTypes (struct :Type Integer32) :InputSource BiomeID :OutputTarget BiomeID))
  ;[0 0 0] Remove Asset if generator is missing
  (bind Branch_1 (Branch :bOutputToB-param AllowNoGenerator))
  (bind AttributeFilter_1 (Filter%20Attribute%20Elements :AttributeTypes (struct :Type SoftObjectPath) :bUseConstantThreshold true :bUseSpatialQuery true :Operator NotEqual :TargetAttribute Generator))
  (bind Gather_2 (Gather))
  ;[0 0 0] Get Generator Type from Object Path
  (bind GetPropertyFromObjectPath_0 (Get%20Property%20From%20Object%20Path :bSanitizeOutputAttributeName false :InputSource Generator :OutputAttributeName None :PropertyName Generator))
  (bind CreateAttribute_2 (Create%20Constant :AttributeTypes (struct :Type String :StringValue "GEN:")))
  (bind MergeAttributes_4 (Merge%20Attributes))
  (bind AttributeStringOp_1 (Attribute%20String%20Op :InputSource2 GeneratorType :OutputTarget GeneratorType))
  (bind AddAttribute_2 (Add%20Attribute :InputSource GeneratorType :OutputTarget GeneratorType))
  (connect FilterDataByAttribute_0 InsideFilter AddAttribute_0 Attributes)
  (connect AddAttribute_6 Out AddAttribute_0 In)
  (connect AttributeStringOp_1 Out AddAttribute_2 Attributes)
  (connect Gather_2 Out AddAttribute_2 In)
  (connect Select_8 Out AddAttribute_6 Attributes)
  (connect Select_10 Out AddAttribute_6 In)
  (connect AttributeReduce_17 Out AttributeCompareOp_5 InA)
  (connect CreateAttribute_7 Out AttributeCompareOp_5 InB)
  (connect Branch_1 Output%20A AttributeFilter_1 In)
  (connect GetEntriesCount_16 Out AttributeReduce_17 In)
  (connect CreateAttribute_2 Out AttributeStringOp_1 InA)
  (connect MergeAttributes_4 Out AttributeStringOp_1 InB)
  (connect AddAttribute_0 Out Branch_1 In)
  (connect Select_8 Out FilterDataByAttribute_0 In)
  (connect AttributeFilter_1 InsideFilter Gather_2 In)
  (connect Branch_1 Output%20B Gather_2 In)
  (connect DefaultInputNode BiomeAssets GetEntriesCount_16 In)
  (connect DefaultInputNode BiomeDefinitions GetEntriesCount_16 In)
  (connect Gather_2 Out GetPropertyFromObjectPath_0 In)
  (connect GetPropertyFromObjectPath_0 Out MergeAttributes_4 In)
  (connect AttributeCompareOp_5 Out Select_8 bUseInputB)
  (connect DefaultInputNode BiomeDefinitions Select_8 Input%20A)
  (connect AttributeCompareOp_5 Out Select_10 bUseInputB)
  (connect DefaultInputNode BiomeAssets Select_10 Input%20A)
  (output AssetsWithDefinitions AddAttribute_2))
