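; PCG subgraph for Biome Core that widens each entry's BoundsMin and BoundsMax to include the bounds of its Assembly.
; Splits off entries with a non-empty Assembly, loads those data assets and merges their Root-tagged data as points to read bounds.
; Combines these with existing bounds by Min/Max, gathers them with entries lacking assemblies, then applies BoundsOffset to all.
; A count-driven branch bypasses loading when no assemblies exist.
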(graph BiomeCore_BoundsFromAssembly
  (bind AttributeFilter_2 (Filter%20Attribute%20Elements :AttributeTypes (struct :Type SoftObjectPath) :bUseConstantThreshold true :bUseSpatialQuery true :Operator NotEqual :TargetAttribute Assembly))
  (bind GetEntriesCount_3 (Get%20Element%20Count))
  (bind Branch_1 (Branch))
  (bind Gather_28 (Gather))
  (bind PointToAttributeSet_0 (Point%20To%20Attribute%20Set))
  (bind PointToAttributeSet_1 (Point%20To%20Attribute%20Set))
  (bind Gather_0 (Gather))
  ;[0 0 0] Assembly Bounds from Root point data
  (bind PCGLoadDataAsset_0 (Load%20PCG%20Data%20Asset :AssetReferenceSelector Assembly :bLoadFromInput true :bWarnIfNoAsset false))
  (bind FilterDataByTag_0 (Filter%20Data%20By%20Tag :bTokenizeOnWhiteSpace true :SelectedTags "Root"))
  (bind ToPoint_0 (To%20Point))
  (bind MergePoints_0 (Merge%20Points :bMergeMetadata false))
  (bind CopyAttributes_0 (Copy%20Attributes :InputSource $BoundsMin :OutputTarget BoundsMin))
  (bind CopyAttributes_1 (Copy%20Attributes :InputSource $BoundsMax :OutputTarget BoundsMax))
  ;[0 0 0] Combine with Mesh Bounds
  (bind AttributeMathsOp_4 (Attribute%20Maths%20Op :InputSource1 BoundsMin :InputSource2 BoundsMin :Operation Min :OutputTarget BoundsMin))
  (bind AttributeMathsOp_5 (Attribute%20Maths%20Op :InputSource1 BoundsMax :InputSource2 BoundsMax :Operation Max :OutputTarget BoundsMax))
  ;[0 0 0] Apply Bounds Offset (Applies to Bounds from assemblies as well as bounds from meshes)
  (bind AttributeMathsOp_6 (Attribute%20Maths%20Op :InputSource1 BoundsMin :InputSource2 BoundsOffset :OutputTarget BoundsMin))
  (bind AttributeMathsOp_7 (Attribute%20Maths%20Op :InputSource1 BoundsMax :InputSource2 BoundsOffset :OutputTarget BoundsMax))
  (connect DefaultInputNode In AttributeFilter_2 In)
  (connect Branch_1 Output%20B AttributeMathsOp_4 InA)
  (connect PointToAttributeSet_0 Out AttributeMathsOp_4 InB)
  (connect AttributeMathsOp_4 Out AttributeMathsOp_5 InA)
  (connect PointToAttributeSet_1 Out AttributeMathsOp_5 InB)
  (connect Gather_0 Out AttributeMathsOp_6 InA)
  (connect Gather_0 Out AttributeMathsOp_6 InB)
  (connect AttributeMathsOp_6 Out AttributeMathsOp_7 InA)
  (connect AttributeMathsOp_6 Out AttributeMathsOp_7 InB)
  (connect GetEntriesCount_3 Out Branch_1 bOutputToB)
  (connect AttributeFilter_2 InsideFilter Branch_1 In)
  (connect MergePoints_0 Out CopyAttributes_0 Source)
  (connect MergePoints_0 Out CopyAttributes_0 Target)
  (connect MergePoints_0 Out CopyAttributes_1 Source)
  (connect MergePoints_0 Out CopyAttributes_1 Target)
  (connect PCGLoadDataAsset_0 Out FilterDataByTag_0 In)
  (connect AttributeMathsOp_5 Out Gather_0 In)
  (connect Gather_28 Out Gather_0 In)
  (connect AttributeFilter_2 OutsideFilter Gather_28 In)
  (connect Branch_1 Output%20A Gather_28 In)
  (connect AttributeFilter_2 InsideFilter GetEntriesCount_3 In)
  (connect ToPoint_0 Out MergePoints_0 In)
  (connect Branch_1 Output%20B PCGLoadDataAsset_0 In)
  (connect CopyAttributes_0 Out PointToAttributeSet_0 In)
  (connect CopyAttributes_1 Out PointToAttributeSet_1 In)
  (connect FilterDataByTag_0 InsideFilter ToPoint_0 In)
  (output Out AttributeMathsOp_7))
