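; PCG Biome Core loop subgraph that adds BoundsMin and BoundsMax attributes to biome data, derived from its mesh references.
; Builds points from the Mesh and SkeletalMesh attributes and counts the results.
; When both resolve, it keeps the one with larger extents length, favoring skeletal on ties.
; Bounds come from the chosen point, or fall back to a constant 50-unit half-extent box when neither mesh yields one.
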(graph BiomeCore_BoundsFromMeshLoop
  (bind GetAttributeFromIndex_0 (Get%20Attribute%20Set%20from%20Index))
  (bind DeleteAttributes_3 (Delete%20Attributes :bTokenizeOnWhiteSpace true :Operation KeepSelectedAttributes :SelectedAttributes "Mesh"))
  (bind PointFromMesh_2 (Point%20From%20Mesh))
  (bind DeleteAttributes_0 (Delete%20Attributes :bTokenizeOnWhiteSpace true :Operation KeepSelectedAttributes :SelectedAttributes "SkeletalMesh"))
  (bind PointFromMesh_0 (Point%20From%20Mesh))
  (bind DataNum_3 (Data%20Count))
  (bind Branch_2 (Branch :bOutputToB true))
  (bind AttributeCompareOp_1 (Attribute%20Compare%20Op :Operation Greater))
  (bind Branch_1 (Branch))
  (bind Branch_3 (Branch))
  (bind AttributeCompareOp_0 (Attribute%20Compare%20Op :InputSource1 $Extents.length :InputSource2 $Extents.length :Operation LessOrEqual :OutputTarget UseSkeletalMeshBounds))
  (bind DeleteAttributes_1 (Delete%20Attributes :Operation KeepSelectedAttributes :SelectedAttributes "UseSkeletalMeshBounds"))
  (bind PointToAttributeSet_0 (Point%20To%20Attribute%20Set))
  (bind Select_1 (Select))
  (bind GetAttributeFromPointIndex_6 (Get%20Attribute%20From%20Point%20Index :bForceOutputAttributeToBeInElementsDomain true :InputSource $BoundsMin :OutputAttributeName None))
  (bind AddAttribute_4 (Add%20Attribute :OutputTarget BoundsMin))
  (bind GetAttributeFromPointIndex_8 (Get%20Attribute%20From%20Point%20Index :bForceOutputAttributeToBeInElementsDomain true :InputSource $BoundsMax :OutputAttributeName None))
  (bind AddAttribute_10 (Add%20Attribute :OutputTarget BoundsMax))
  (bind CreateAttribute_8 (Create%20Constant :AttributeTypes (struct :Type Vector :DoubleValue -5e+01 :VectorValue (-5e+01 -5e+01 -5e+01))))
  (bind AddAttribute_5 (Add%20Attribute :OutputTarget BoundsMin))
  (bind CreateAttribute_10 (Create%20Constant :AttributeTypes (struct :Type Vector :DoubleValue 5e+01 :VectorValue (5e+01 5e+01 5e+01))))
  (bind AddAttribute_7 (Add%20Attribute :OutputTarget BoundsMax))
  (connect GetAttributeFromPointIndex_6 Attribute AddAttribute_4 Attributes)
  (connect Branch_2 Output%20B AddAttribute_4 In)
  (connect CreateAttribute_8 Out AddAttribute_5 Attributes)
  (connect Branch_2 Output%20A AddAttribute_5 In)
  (connect CreateAttribute_10 Out AddAttribute_7 Attributes)
  (connect AddAttribute_5 Out AddAttribute_7 In)
  (connect GetAttributeFromPointIndex_8 Attribute AddAttribute_10 Attributes)
  (connect AddAttribute_4 Out AddAttribute_10 In)
  (connect Branch_1 Output%20B AttributeCompareOp_0 InA)
  (connect Branch_3 Output%20B AttributeCompareOp_0 InB)
  (connect DataNum_3 Out AttributeCompareOp_1 InA)
  (connect AttributeCompareOp_1 Out Branch_1 bOutputToB)
  (connect PointFromMesh_2 Out Branch_1 In)
  (connect DataNum_3 Out Branch_2 bOutputToB)
  (connect DefaultInputNode In Branch_2 In)
  (connect AttributeCompareOp_1 Out Branch_3 bOutputToB)
  (connect PointFromMesh_0 Out Branch_3 In)
  (connect PointFromMesh_0 Out DataNum_3 In)
  (connect PointFromMesh_2 Out DataNum_3 In)
  (connect GetAttributeFromIndex_0 Out DeleteAttributes_0 In)
  (connect AttributeCompareOp_0 Out DeleteAttributes_1 In)
  (connect GetAttributeFromIndex_0 Out DeleteAttributes_3 In)
  (connect DefaultInputNode In GetAttributeFromIndex_0 In)
  (connect Branch_1 Output%20A GetAttributeFromPointIndex_6 In)
  (connect Branch_3 Output%20A GetAttributeFromPointIndex_6 In)
  (connect Select_1 Out GetAttributeFromPointIndex_6 In)
  (connect Branch_1 Output%20A GetAttributeFromPointIndex_8 In)
  (connect Branch_3 Output%20A GetAttributeFromPointIndex_8 In)
  (connect Select_1 Out GetAttributeFromPointIndex_8 In)
  (connect DeleteAttributes_0 Out PointFromMesh_0 Mesh)
  (connect DeleteAttributes_3 Out PointFromMesh_2 Mesh)
  (connect DeleteAttributes_1 Out PointToAttributeSet_0 In)
  (connect PointToAttributeSet_0 Out Select_1 bUseInputB)
  (connect Branch_1 Output%20B Select_1 Input%20A)
  (connect Branch_3 Output%20B Select_1 Input%20B)
  (output Out AddAttribute_7)
  (output Out AddAttribute_10))
