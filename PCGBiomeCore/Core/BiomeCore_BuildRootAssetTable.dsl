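; PCG graph that builds the root biome asset table for Biome Core, pairing assets with biome definitions.
; It resolves assets from class-selected actors and actor references, then assigns them enabled definitions, selecting between Definition and DefaultDefinition properties.
; Data-count branches gate both inputs to that assignment.
; Assets then receive index-based AssetIDs, mesh bounds via partition and loop, and assembly bounds, before an isolation pass and AssetID sort.
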(graph BiomeCore_BuildRootAssetTable
  (outputs (AssetTable :type Param))
  ;[0 0 0] Build Local Root Asset Table
  (bind GetActorProperty_3 (Get%20Actor%20Property :ActorSelector (struct :ActorSelection ByClass :bSelectMultiple true) :bAlwaysRequeryActors false :bProcessAllComponents false :bSanitizeOutputAttributeName false :OutputAttributeName Assets :PropertyName Assets))
  (bind Subgraph_1 (Subgraph :Graph /PCGBiomeCore/Core/BiomeCore_ResolveAssetProperties.BiomeCore_ResolveAssetProperties))
  (bind GetActorData_0 (Get%20Actor%20Data :Mode GetActorReference))
  (bind Subgraph_2 (Subgraph :Graph /PCGBiomeCore/Core/BiomeCore_ResolveAssetProperties.BiomeCore_ResolveAssetProperties))
  (bind MergeAttributes_1 (Merge%20Attributes))
  (bind DataNum_0 (Data%20Count))
  (bind Branch_0 (Branch))
  (bind GetActorProperty_5 (Get%20Actor%20Property :bAlwaysRequeryActors false :bSanitizeOutputAttributeName false :PropertyName Definition))
  (bind AttributeCompareOp_2 (Attribute%20Compare%20Op))
  (bind GetActorProperty_6 (Get%20Actor%20Property :bAlwaysRequeryActors false :bSanitizeOutputAttributeName false :PropertyName DefaultDefinition))
  (bind GetPropertyFromObjectPath_2 (Get%20Property%20From%20Object%20Path :bSanitizeOutputAttributeName false :ObjectExtractorBehavior Extract :PropertyName BiomeDefinition :StructExtractorBehavior Extract))
  (bind Select_1 (Select))
  (bind FilterDataByTag_0 (Filter%20Data%20By%20Tag :SelectedTags "Enabled"))
  (bind DataNum_1 (Data%20Count))
  (bind Branch_1 (Branch))
  (bind Subgraph_0 (Subgraph :Graph /PCGBiomeCore/Core/BiomeCore_AssignDefinitionsToAssets.BiomeCore_AssignDefinitionsToAssets))
  (bind MergeAttributes_0 (Merge%20Attributes))
  (bind SortAttributes_8 (Sort%20Attributes :bUseStableSort false :SortAttributes [(struct :InputSource AssetID)]))
  ;;[0 0 0] AssetID from Index
  (bind AddAttribute_12 (Add%20Attribute :AttributeTypes (struct :Type Integer32) :OutputTarget AssetID))
  (bind AttributeMathsOp_11 (Attribute%20Maths%20Op :InputSource1 AssetID :InputSource2 $Index :Operation Set))
  ;;[0 0 0] Store Meshes Bounds to assets tables
  (bind AttributePartition_9 (Attribute%20Partition :bTokenizeOnWhiteSpace true :PartitionAttributeSelectors [Mesh SkeletalMesh]))
  (bind Loop_8 (Loop :bTokenizeOnWhiteSpace true :Graph /PCGBiomeCore/Core/BiomeCore_BoundsFromMeshLoop.BiomeCore_BoundsFromMeshLoop))
  ;;[0 0 0] Store Assemblies Bounds to assets tables
  (bind MergeAttributes_3 (Merge%20Attributes))
  (bind Subgraph_3 (Subgraph :Graph /PCGBiomeCore/Core/BiomeCore_BoundsFromAssembly.BiomeCore_BoundsFromAssembly))
  ;;[0 0 0] Reset meshes/actors/assemblies/children of non isolated assets
  (bind MergeAttributes_6 (Merge%20Attributes))
  (bind Subgraph_26 (Subgraph :Graph /PCGBiomeCore/Core/BiomeCore_IsolateAssets.BiomeCore_IsolateAssets))
  (connect MergeAttributes_0 Out AddAttribute_12 In)
  (connect GetActorProperty_5 Out AttributeCompareOp_2 InA)
  (connect AddAttribute_12 Out AttributeMathsOp_11 InA)
  (connect AddAttribute_12 Out AttributeMathsOp_11 InB)
  (connect AttributeMathsOp_11 Out AttributePartition_9 In)
  (connect DataNum_0 Out Branch_0 bOutputToB)
  (connect MergeAttributes_1 Out Branch_0 In)
  (connect DataNum_1 Out Branch_1 bOutputToB)
  (connect FilterDataByTag_0 InsideFilter Branch_1 In)
  (connect MergeAttributes_1 Out DataNum_0 In)
  (connect FilterDataByTag_0 InsideFilter DataNum_1 In)
  (connect Select_1 Out FilterDataByTag_0 In)
  (connect GetActorProperty_5 Out GetPropertyFromObjectPath_2 In)
  (connect AttributePartition_9 Out Loop_8 In)
  (connect Subgraph_0 AssetsWithDefinitions MergeAttributes_0 In)
  (connect Subgraph_1 Out MergeAttributes_1 In)
  (connect Subgraph_2 Out MergeAttributes_1 In2)
  (connect Loop_8 Out MergeAttributes_3 In)
  (connect Subgraph_3 Out MergeAttributes_6 In)
  (connect AttributeCompareOp_2 Out Select_1 bUseInputB)
  (connect GetPropertyFromObjectPath_2 Out Select_1 Input%20A)
  (connect GetActorProperty_6 Out Select_1 Input%20B)
  (connect Subgraph_26 Merge%20Attributes%20Out SortAttributes_8 In)
  (connect Branch_0 Output%20B Subgraph_0 BiomeAssets)
  (connect Branch_1 Output%20B Subgraph_0 BiomeDefinitions)
  (connect GetActorProperty_3 Out Subgraph_1 In)
  (connect GetActorData_0 Out Subgraph_2 In)
  (connect MergeAttributes_3 Out Subgraph_3 In)
  (connect MergeAttributes_6 Out Subgraph_26 Attribute%20Filter%20In)
  (output AssetTable SortAttributes_8))
