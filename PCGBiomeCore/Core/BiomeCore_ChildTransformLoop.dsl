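; PCG Biome Core recursive loop body that spawns a child asset's points from parent points.
; Points are transformed, scaled, oriented, run through the Filters subgraph, weight-selected by noise, differenced against sibling Feedback and self-pruned.
; Survivors receive asset attributes and recursion tags, and join Feedback for later siblings.
; Below MaxIterations they also re-enter this graph as parents, selected parents are subtracted from returned children, and runtime asset tables are gathered.
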(graph BiomeCore_ChildTransformLoop
  (params (CurrentIteration :type Int) (MaxIterations :type Int :default 4) (Filters :type (Object /Script/PCG.PCGGraphInterface)))
  (inputs (ChildAssetsLoop :type Param) (MaxInputPoints :type Param) (Feedback :type Point))
  (outputs (RuntimeAssetTable :type Param) (Feedback :type Point))
  (bind Gather_10 (Gather))
  (bind AddTags_0 (Add%20Tags :bTokenizeOnWhiteSpace true :TagsToAdd "Child"))
  (bind AddTags_15 (Add%20Tags :bTokenizeOnWhiteSpace true :Prefix "RecursionLevel:" :TagsToAdd-param CurrentIteration))
  (bind DeleteAttributes_9 (Delete%20Attributes :bTokenizeOnWhiteSpace true :Operation KeepSelectedAttributes :SelectedAttributes "Assembly"))
  (bind AttributeCompareOp_0 (Attribute%20Compare%20Op))
  (bind Branch_1 (Branch))
  (bind AddTags_1 (Add%20Tags :TagsToAdd "Assembly"))
  (bind DeleteAttributes_5 (Delete%20Attributes :bTokenizeOnWhiteSpace true :Operation KeepSelectedAttributes :SelectedAttributes "Mesh"))
  (bind AttributeCompareOp_2 (Attribute%20Compare%20Op :InputSource1 Mesh))
  (bind DeleteAttributes_10 (Delete%20Attributes :bTokenizeOnWhiteSpace true :Operation KeepSelectedAttributes :SelectedAttributes "Actor"))
  (bind AttributeCompareOp_3 (Attribute%20Compare%20Op :InputSource1 Actor))
  (bind AttributeBooleanOp_1 (Attribute%20Boolean%20Op))
  (bind Branch_2 (Branch))
  (bind Gather_1 (Gather))
  ;[0 0 0 1 32] Children Recursion
  (bind AttributeMathsOp_2 (Attribute%20Maths%20Op :InA CurrentIteration))
  (bind AttributeCompareOp_1 (Attribute%20Compare%20Op :Operation Less :InB MaxIterations))
  (bind Branch_0 (Branch))
  (bind GetLoopIndex_0 (Get%20Loop%20Index))
  (bind AttributeRename_18 (Attribute%20Rename :AttributeToRename LoopIndex :NewAttributeName AssetID))
  (bind Subgraph_1 (Subgraph :Graph /PCGBiomeCore/Core/BiomeCore_PrepareChildRecursion.BiomeCore_PrepareChildRecursion))
  ;;[1 0 0] Recursive Loop
  (bind Loop_19 (Loop :bTokenizeOnWhiteSpace true :Graph /PCGBiomeCore/Core/BiomeCore_ChildTransformLoop.BiomeCore_ChildTransformLoop :Filters Filters :MaxIterations MaxIterations))
  ;[0 0 0 1 32] Child Transform
  (bind DeleteAttributes_6 (Delete%20Attributes :bTokenizeOnWhiteSpace true :Operation KeepSelectedAttributes :SelectedAttributes "OrientUpward"))
  (bind DeleteAttributes_11 (Delete%20Attributes :bTokenizeOnWhiteSpace true :Operation KeepSelectedAttributes :SelectedAttributes "ForceAssetScale"))
  (bind Subgraph_0 (Subgraph :Graph /PCGBiomeCore/Core/BiomeCore_RunTransformGraph.BiomeCore_RunTransformGraph))
  (bind Subgraph_4 (Subgraph :Graph /PCGBiomeCore/Core/BiomeCore_ApplyScaleAndBounds.BiomeCore_ApplyScaleAndBounds))
  (bind MakeRotatorAttribute_3 (Make%20Rotator%20Attribute :InputSource2 $Rotation.right :Operation MakeRotFromZY :OutputTarget $Rotation))
  (bind AttributeRotatorOp_4 (Attribute%20Rotator%20Op :InputSource1 $Rotation :InputSource2 $Rotation :InputSource3 OrientUpward :Operation Lerp))
  ;[0 0 0 1 30] Apply Filters
  (bind DeleteAttributes_7 (Delete%20Attributes :bTokenizeOnWhiteSpace true :Operation KeepSelectedAttributes :SelectedAttributes "DensityMin,DensityMax,WaterDistanceMin,WaterDistanceMax,SunExposureMin,SunExposureMax,FlowMin,FlowMax,HeightMin,HeightMax"))
  (bind AddAttribute_2 (Add%20Attribute :bCopyAllAttributes true))
  (bind Subgraph_5 (Subgraph :SubgraphOverride-param Filters))
  (bind FilterDataByType_8 (Filter%20Data%20By%20Type :TargetType (struct :Ids [(struct :Struct /Script/PCG.PCGDataTypeInfoPoint)])))
  (bind DeleteAttributes_4 (Delete%20Attributes :bTokenizeOnWhiteSpace true :SelectedAttributes "DensityMin,DensityMax,WaterDistanceMin,WaterDistanceMax,SunExposureMin,SunExposureMax,FlowMin,FlowMax,HeightMin,HeightMax"))
  ;[0 0 0 1 32] Apply Weight
  (bind AttributeNoise_1 (Attribute%20Noise :InputSource $Color.a))
  (bind DeleteAttributes_0 (Delete%20Attributes :bTokenizeOnWhiteSpace true :Operation KeepSelectedAttributes :SelectedAttributes "FilterUpperBound"))
  (bind DeleteAttributes_1 (Delete%20Attributes :bTokenizeOnWhiteSpace true :Operation KeepSelectedAttributes :SelectedAttributes "FilterLowerBound"))
  (bind AttributeFilterRange_8 (Filter%20Attribute%20Elements%20by%20Range :MaxThreshold (struct :ThresholdAttribute FilterUpperBound :bUseSpatialQuery false) :MinThreshold (struct :ThresholdAttribute FilterLowerBound :bUseSpatialQuery false) :TargetAttribute $Color.a))
  ;[0 0 0] Diff with siblings & self pruning
  (bind Difference_4 (Difference :bDiffMetadata false :DensityFunction Binary))
  (bind ToPoint_5 (To%20Point))
  (bind Loop_6 (Loop :bTokenizeOnWhiteSpace true :Graph /PCGBiomeCore/Core/BiomeCore_SelfPruningGraph.BiomeCore_SelfPruningGraph))
  ;[0 0 0] Add Mesh, Material, Actor attributes to points
  (bind DeleteAttributes_2 (Delete%20Attributes :bTokenizeOnWhiteSpace true :Operation KeepSelectedAttributes :SelectedAttributes "Mesh,SkeletalMesh,Material,Assembly,Actor,AllowCollision,AssemblyAllowCollision,Visible,CastShadow,CastHiddenShadow,CastDynamicShadow,AffectDistanceFieldLighting,DetailMode,StartCullDistance,EndCullDistance,EvaluateWorldPositionOffset,WorldPositionOffsetDisableDistance,IncludeInHLOD,HLODLayer,RuntimeAssemblyHashValue,InfluenceType,InfluenceRadius,AffectDynamicIndirectLighting,VisibleInRayTracing,CanEverAffectNavigation,RasterizeAsFilledConvexVolume,IsSpatiallyLoaded,ComponentTags"))
  (bind AddAttribute_12 (Add%20Attribute :OutputTarget RecursionLevel :Attributes CurrentIteration))
  (bind CopyAttribute_3 (Copy%20Attributes :bCopyAllAttributes true :InputSource $Density :OutputTarget None))
  ;[0 0 0] Diff selected parents from children
  (bind Subgraph_3 (Subgraph :Graph /PCGBiomeCore/Core/BiomeCore_OverlapWithChildrenGraph.BiomeCore_OverlapWithChildrenGraph))
  (bind Difference_1 (Difference :bDiffMetadata false :DensityFunction Binary))
  (bind ToPoint_2 (To%20Point))
  ;[1 0.2122 0 0.1] Show Bounds
  (bind DeleteAttributes_8 (Delete%20Attributes :bTokenizeOnWhiteSpace true :Operation KeepSelectedAttributes :SelectedAttributes "ShowBounds"))
  (bind Select_9 (Select))
  (bind Debug_10 (Debug))
  ;[0 0 0] Runtime Asset Table
  (bind DeleteAttributes_3 (Delete%20Attributes :Operation KeepSelectedAttributes :SelectedAttributes "RuntimeAssembly,RuntimeAssemblyHashValue"))
  (bind Gather_0 (Gather))
  (reroute Asset ChildAssetsLoop)
  (reroute NamedRerouteDeclaration_0 MaxInputPoints)
  (reroute NamedRerouteDeclaration_2 Feedback)
  (connect DeleteAttributes_7 Out AddAttribute_2 Attributes)
  (connect AttributeRotatorOp_4 Out AddAttribute_2 In)
  (connect DeleteAttributes_2 Out AddAttribute_12 In)
  (connect CopyAttribute_3 Out AddTags_0 In)
  (connect Branch_1 Output%20A AddTags_1 In)
  (connect AddTags_0 Out AddTags_15 In)
  (connect AttributeCompareOp_2 Out AttributeBooleanOp_1 InA)
  (connect AttributeCompareOp_3 Out AttributeBooleanOp_1 InB)
  (connect DeleteAttributes_9 Out AttributeCompareOp_0 InA)
  (connect AttributeMathsOp_2 Out AttributeCompareOp_1 InA)
  (connect DeleteAttributes_5 Out AttributeCompareOp_2 InA)
  (connect DeleteAttributes_10 Out AttributeCompareOp_3 InA)
  (connect DeleteAttributes_0 Out AttributeFilterRange_8 FilterMax)
  (connect DeleteAttributes_1 Out AttributeFilterRange_8 FilterMin)
  (connect AttributeNoise_1 Out AttributeFilterRange_8 In)
  (connect DeleteAttributes_4 Out AttributeNoise_1 In)
  (connect GetLoopIndex_0 Out AttributeRename_18 In)
  (connect Subgraph_4 Out AttributeRotatorOp_4 InA)
  (connect MakeRotatorAttribute_3 Out AttributeRotatorOp_4 InB)
  (connect DeleteAttributes_6 Out AttributeRotatorOp_4 Ratio)
  (connect AttributeCompareOp_1 Out Branch_0 bOutputToB)
  (connect Loop_6 Out Branch_0 In)
  (connect AttributeCompareOp_0 Out Branch_1 bOutputToB)
  (connect AddTags_15 Out Branch_1 In)
  (connect AttributeBooleanOp_1 Out Branch_2 bOutputToB)
  (connect Branch_1 Output%20A Branch_2 In)
  (connect AddAttribute_12 Out CopyAttribute_3 Source)
  (connect Loop_6 Out CopyAttribute_3 Target)
  (connect Select_9 Out Debug_10 In)
  (connect Asset Out DeleteAttributes_0 In)
  (connect Asset Out DeleteAttributes_1 In)
  (connect Asset Out DeleteAttributes_2 In)
  (connect Asset Out DeleteAttributes_3 In)
  (connect FilterDataByType_8 InsideFilter DeleteAttributes_4 In)
  (connect AddAttribute_12 Out DeleteAttributes_5 In)
  (connect Asset Out DeleteAttributes_6 In)
  (connect Asset Out DeleteAttributes_7 In)
  (connect Asset Out DeleteAttributes_8 In)
  (connect AddAttribute_12 Out DeleteAttributes_9 In)
  (connect AddAttribute_12 Out DeleteAttributes_10 In)
  (connect Asset Out DeleteAttributes_11 In)
  (connect Subgraph_3 Out Difference_1 Differences)
  (connect Loop_19 Out Difference_1 Source)
  (connect NamedRerouteDeclaration_2 Out Difference_4 Differences)
  (connect AttributeFilterRange_8 InsideFilter Difference_4 Source)
  (connect Subgraph_5 Out FilterDataByType_8 In)
  (connect DeleteAttributes_3 Out Gather_0 In)
  (connect Loop_19 RuntimeAssetTable Gather_0 In)
  (connect AddTags_1 Out Gather_1 In)
  (connect Branch_2 Output%20A Gather_1 In)
  (connect Branch_1 Output%20B Gather_1 In2)
  (connect NamedRerouteDeclaration_2 Out Gather_10 In)
  (connect Loop_6 Out Gather_10 In2)
  (connect Asset Out Loop_6 Asset)
  (connect ToPoint_5 Out Loop_6 In)
  (connect Subgraph_1 Out Loop_19 ChildAssetsLoop)
  (connect AttributeMathsOp_2 Out Loop_19 CurrentIteration)
  (connect Branch_0 Output%20B Loop_19 In)
  (connect NamedRerouteDeclaration_0 Out Loop_19 MaxInputPoints)
  (connect Subgraph_4 Out MakeRotatorAttribute_3 Y)
  (connect DeleteAttributes_8 Out Select_9 bUseInputB)
  (connect Loop_6 Out Select_9 Input%20B)
  (connect Asset Out Subgraph_0 Asset)
  (connect DefaultInputNode In Subgraph_0 In)
  (connect NamedRerouteDeclaration_0 Out Subgraph_0 MaxInputPoints)
  (connect AttributeRename_18 Out Subgraph_1 AssetID)
  (connect Asset Out Subgraph_1 AssetTable)
  (connect Asset Out Subgraph_3 Asset)
  (connect Loop_6 Out Subgraph_3 In)
  (connect Asset Out Subgraph_4 Asset)
  (connect Subgraph_0 Out Subgraph_4 In)
  (connect DeleteAttributes_11 Out Subgraph_4 SetScale)
  (connect AddAttribute_2 Out Subgraph_5 In)
  (connect Difference_1 Out ToPoint_2 In)
  (connect Difference_4 Out ToPoint_5 In)
  (output Feedback Gather_10)
  (output Out Gather_1)
  (output Out ToPoint_2)
  (output RuntimeAssetTable Gather_0))
