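; PCG subgraph from PCGBiomeCore that trims incoming points to partition actor bounds when the PCG component is partitioned.
; It reads bIsComponentPartitioned from the original actor's PCG component and uses it to switch a Branch.
; Partitioned runs pass through Cull Points Outside Actor Bounds, padded by the BoundsExpansion parameter, while non-partitioned runs skip culling.
; A Gather merges both routes into PointsOut.
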(graph BiomeCore_CullPointsOutsidePartitionActors
  (params (BoundsExpansion :type Double))
  (inputs (PointsIn :type Point))
  (outputs (PointsOut :type Point))
  (bind GetActorProperty_1 (Get%20Actor%20Property :ActorSelector (struct :ActorFilter Original) :bAlwaysRequeryActors false :bProcessAllComponents false :bSanitizeOutputAttributeName false :bSelectComponent true :ComponentClass /Script/PCG.PCGComponent :OutputAttributeName None :PropertyName bIsComponentPartitioned))
  (bind Branch_2 (Branch))
  (bind CullPointsOutsideActorBounds_0 (Cull%20Points%20Outside%20Actor%20Bounds :BoundsExpansion-param BoundsExpansion))
  (bind Gather_3 (Gather))
  (connect GetActorProperty_1 Out Branch_2 bOutputToB)
  (connect DefaultInputNode PointsIn Branch_2 In)
  (connect Branch_2 Output%20B CullPointsOutsideActorBounds_0 In)
  (connect Branch_2 Output%20A Gather_3 In)
  (connect CullPointsOutsideActorBounds_0 Out Gather_3 In)
  (output PointsOut Gather_3))
