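; PCG subgraph in Biome Core that culls unused generator subtypes from its input data against an AllowedSubtypes parameter input.
; It strips attributes down to GeneratorSubType, then filters data by that attribute, sending data outside the filter straight to output.
; Matching data is partitioned by GeneratorSubType and looped through the CullUnusedSubTypesInnerLoop graph with the allowed list.
; Loop results join the pass-through data on the output.
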(graph BiomeCore_CullUnusedSubTypes
  (inputs (AllowedSubtypes :type Param))
  (bind DeleteAttributes_0 (Delete%20Attributes :Operation KeepSelectedAttributes :SelectedAttributes "GeneratorSubType"))
  (bind FilterDataByAttribute_2 (Filter%20Data%20By%20Attribute :Attribute GeneratorSubType))
  (bind AttributePartition_1 (Attribute%20Partition :PartitionAttributeSelectors [GeneratorSubType]))
  (bind Loop_3 (Loop :Graph /PCGBiomeCore/Core/BiomeCore_CullUnusedSubTypesInnerLoop.BiomeCore_CullUnusedSubTypesInnerLoop))
  (connect FilterDataByAttribute_2 InsideFilter AttributePartition_1 In)
  (connect DefaultInputNode In DeleteAttributes_0 In)
  (connect DeleteAttributes_0 Out FilterDataByAttribute_2 In)
  (connect DefaultInputNode AllowedSubtypes Loop_3 AllowedSubtypes)
  (connect AttributePartition_1 Out Loop_3 In)
  (output Out FilterDataByAttribute_2.OutsideFilter)
  (output Out Loop_3))
