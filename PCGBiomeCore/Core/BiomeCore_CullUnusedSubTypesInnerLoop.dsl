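; PCG loop-body subgraph for BiomeCore that culls a point set whose generator subtype is absent from the allowed subtypes.
; It reads GeneratorSubType from the first point and substring-filters the AllowedSubtypes GeneratorSubTypes attribute against it.
; Element counts drive Branch nodes as booleans, so empty inputs are skipped and points pass to Out only when a match exists.
; Substring matching means a partial name overlap also counts as allowed.
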(graph BiomeCore_CullUnusedSubTypesInnerLoop
  (inputs (AllowedSubtypes :type Param))
  (bind GetPointsCount_6 (Get%20Element%20Count))
  (bind Branch_9 (Branch))
  (bind Branch_7 (Branch))
  (bind GetAttributeFromPointIndex_2 (Get%20Attribute%20From%20Point%20Index :bForceOutputAttributeToBeInElementsDomain true :InputSource GeneratorSubType :OutputAttributeName GeneratorSubType))
  (bind AttributeFilter_1 (Filter%20Attribute%20Elements :Operator Substring :TargetAttribute GeneratorSubTypes :ThresholdAttribute GeneratorSubType))
  (bind GetEntriesCount_5 (Get%20Element%20Count))
  (bind Branch_12 (Branch))
  (connect GetAttributeFromPointIndex_2 Attribute AttributeFilter_1 Filter)
  (connect Branch_9 Output%20B AttributeFilter_1 In)
  (connect GetPointsCount_6 Out Branch_7 bOutputToB)
  (connect DefaultInputNode In Branch_7 In)
  (connect GetPointsCount_6 Out Branch_9 bOutputToB)
  (connect DefaultInputNode AllowedSubtypes Branch_9 In)
  (connect GetEntriesCount_5 Out Branch_12 bOutputToB)
  (connect Branch_7 Output%20B Branch_12 In)
  (connect Branch_7 Output%20B GetAttributeFromPointIndex_2 In)
  (connect AttributeFilter_1 InsideFilter GetEntriesCount_5 In)
  (connect DefaultInputNode In GetPointsCount_6 In)
  (output Out Branch_12.Output%20B))
