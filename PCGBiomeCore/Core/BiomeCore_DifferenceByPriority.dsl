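; PCG loop subgraph for Biome Core that trims each source point set against points kept by earlier iterations, using priority.
; Data tags become attribute tables, filtered by biome priority, priority and recursion level to pick data in matching groups for Difference.
; Identical biome data also loses overlapping duplicates through a Difference on temporarily fixed unit bounds, restored afterward.
; The first iteration passes through; survivors join the accumulated RemainingPoints feedback.
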(graph BiomeCore_DifferenceByPriority
  (params (PriorityAttribute :type (Struct /Script/PCG.PCGAttributePropertyInputSelector) :default GeneratorPriority) (DiffMatchingBiomeData :type Bool :default true))
  (inputs (SourcePoints :type Point) (RemainingPoints :type Point))
  (outputs (RemainingPoints :type Point))
  (bind GetLoopIndex_0 (Get%20Loop%20Index))
  (bind Branch_38 (Branch))
  (bind TagsToAttributeSet_4 (Data%20Tags%20To%20Attribute%20Set))
  (bind MergeAttributes_7 (Merge%20Attributes))
  (bind CopyAttribute_9 (Copy%20Attributes :InputSource $Index :OutputTarget DataIndex))
  (bind Branch_4 (Branch :bOutputToB-param DiffMatchingBiomeData))
  (bind Branch_0 (Branch))
  (bind TagsToAttributeSet_13 (Data%20Tags%20To%20Attribute%20Set))
  (bind DataNum_0 (Data%20Count))
  (bind AttributeReduce_20 (Attribute%20Reduce :InputSource DataIndex :JoinDelimiter "," :Operation Join :OutputAttributeName SelectedIndices))
  (bind FilterDataByIndex_17 (Filter%20Data%20By%20Index))
  (bind DeleteTags_6 (Delete%20Tags :Operation KeepOnlySelectedTags :Operator Substring :SelectedTags "GROUP:"))
  (bind GetTags_1 (Get%20Tags :bExtractTagValues true))
  (bind AttributeStringOp_8 (Attribute%20String%20Op :InputSource1 Attribute))
  (bind AttributeStringOp_10 (Attribute%20String%20Op :InputSource1 Attribute :InputSource2 Values))
  (bind AttributeReduce_2 (Attribute%20Reduce :InputSource Attribute :Operation Join :OutputAttributeName Groups))
  (bind FilterDataByTag_7 (Filter%20Data%20By%20Tag :Operator Matches))
  (bind ToPoint_24 (To%20Point))
  (bind GetPointsCount_26 (Get%20Element%20Count))
  (bind Branch_28 (Branch))
  (bind Gather_0 (Gather))
  ;[0 0 0] Filter data indices for difference from BIOME, BIOMEPRIO, PRIO and RecursionLevel value
  (bind AttributeFilter_0 (Filter%20Attribute%20Elements :AttributeTypes (struct :DoubleValue 1.0) :Operator NotEqual :TargetAttribute BIOME :ThresholdAttribute BIOME))
  (bind AttributeFilter_2 (Filter%20Attribute%20Elements :AttributeTypes (struct :DoubleValue 1.0) :Operator Lesser :TargetAttribute BIOMEPRIO :ThresholdAttribute BIOMEPRIO))
  (bind AttributeFilter_11 (Filter%20Attribute%20Elements :AttributeTypes (struct :DoubleValue 1.0) :Operator Lesser :TargetAttribute PRIO :ThresholdAttribute PRIO))
  (bind AttributeFilter_3 (Filter%20Attribute%20Elements :AttributeTypes (struct :DoubleValue 1.0) :Operator LesserOrEqual :TargetAttribute PRIO :ThresholdAttribute PRIO))
  (bind AttributeFilter_1 (Filter%20Attribute%20Elements :AttributeTypes (struct :DoubleValue 1.0) :Operator LesserOrEqual :TargetAttribute RecursionLevel :ThresholdAttribute RecursionLevel))
  (bind MergeAttributes_0 (Merge%20Attributes))
  (bind DeleteAttributes_19 (Delete%20Attributes :Operation KeepSelectedAttributes :SelectedAttributes "DataIndex"))
  ;[0 0 0] Conditonal Branching
  (bind Branch_35 (Branch :bOutputToB true))
  (bind Branch_2 (Branch :bOutputToB true))
  (bind Branch_1 (Branch :bOutputToB true))
  (bind Branch_3 (Branch :bOutputToB true))
  ;[0 0 0 1 32] Difference
  (bind Difference_22 (Difference :bDiffMetadata false :DensityFunction Binary :Mode Discrete))
  ;[0 0 0 0.5] Filter Data Indices to find identical biome data and remove overlapping duplicates
  (bind AttributeFilter_6 (Filter%20Attribute%20Elements :AttributeTypes (struct :DoubleValue 1.0) :Operator Equal :TargetAttribute BIOME :ThresholdAttribute BIOME))
  (bind AttributeFilter_4 (Filter%20Attribute%20Elements :AttributeTypes (struct :DoubleValue 1.0) :Operator Equal :TargetAttribute BIOMEPRIO :ThresholdAttribute BIOMEPRIO))
  (bind AttributeFilter_5 (Filter%20Attribute%20Elements :AttributeTypes (struct :DoubleValue 1.0) :Operator Equal :TargetAttribute PRIO :ThresholdAttribute PRIO))
  (bind AttributeFilter_7 (Filter%20Attribute%20Elements :AttributeTypes (struct :DoubleValue 1.0) :Operator Equal :TargetAttribute RecursionLevel :ThresholdAttribute RecursionLevel))
  (bind AttributeFilter_8 (Filter%20Attribute%20Elements :AttributeTypes (struct :DoubleValue 1.0) :Operator Equal :TargetAttribute GEN :ThresholdAttribute GEN))
  (bind AttributeFilter_10 (Filter%20Attribute%20Elements :AttributeTypes (struct :DoubleValue 1.0) :Operator Equal :TargetAttribute GROUP :ThresholdAttribute GROUP))
  (bind MergeAttributes_1 (Merge%20Attributes))
  (bind DeleteAttributes_0 (Delete%20Attributes :Operation KeepSelectedAttributes :SelectedAttributes "DataIndex"))
  (bind DataNum_1 (Data%20Count))
  ;[0 0 0 0.5] Remove Overlapping Duplicated Biomes Points
  (bind Branch_9 (Branch :bOutputToB true))
  (bind CopyAttributes_1 (Copy%20Attributes :InputSource $BoundsMin :OutputTarget BoundsMinBCK))
  (bind CopyAttributes_0 (Copy%20Attributes :InputSource $BoundsMax :OutputTarget BoundsMaxBCK))
  (bind BoundsModifier_0 (Bounds%20Modifier :bAffectSteepness true :BoundsMax (0.5 0.5 0.5) :BoundsMin (-0.5 -0.5 -0.5) :Mode Set))
  (bind Branch_10 (Branch :bOutputToB true))
  (bind AttributeReduce_0 (Attribute%20Reduce :InputSource DataIndex :JoinDelimiter "," :Operation Join :OutputAttributeName SelectedIndices))
  (bind Branch_11 (Branch :bOutputToB true))
  (bind BoundsModifier_1 (Bounds%20Modifier :bAffectSteepness true :BoundsMax (0.5 0.5 0.5) :BoundsMin (-0.5 -0.5 -0.5) :Mode Set))
  (bind FilterDataByIndex_0 (Filter%20Data%20By%20Index))
  (bind GetElementsCount_0 (Get%20Element%20Count))
  (bind Branch_12 (Branch))
  (bind AttributeMathsOp_3 (Attribute%20Maths%20Op :InputSource1 $BoundsMin :InputSource2 BoundsMinBCK :Operation Set))
  (bind AttributeMathsOp_0 (Attribute%20Maths%20Op :InputSource1 $BoundsMax :InputSource2 BoundsMaxBCK :Operation Set))
  (bind DeleteAttributes_1 (Delete%20Attributes :SelectedAttributes "BoundsMinBCK, BoundsMaxBCK"))
  ;;[0 0 0 1 32] Remove Duplicates
  (bind Difference_0 (Difference :bDiffMetadata false :DensityFunction Binary :Mode Discrete))
  (bind ToPoint_0 (To%20Point))
  (reroute IncomingDataTagsTables TagsToAttributeSet_13)
  (reroute RemainingDataTagsTable Branch_4.Output%20B)
  (connect IncomingDataTagsTables Out AttributeFilter_0 Filter)
  (connect Branch_4 Output%20A AttributeFilter_0 In)
  (connect IncomingDataTagsTables Out AttributeFilter_1 Filter)
  (connect AttributeFilter_3 InsideFilter AttributeFilter_1 In)
  (connect AttributeFilter_11 InsideFilter AttributeFilter_1 In)
  (connect IncomingDataTagsTables Out AttributeFilter_2 Filter)
  (connect AttributeFilter_0 InsideFilter AttributeFilter_2 In)
  (connect RemainingDataTagsTable Out AttributeFilter_2 In)
  (connect IncomingDataTagsTables Out AttributeFilter_3 Filter)
  (connect AttributeFilter_2 InsideFilter AttributeFilter_3 In)
  (connect IncomingDataTagsTables Out AttributeFilter_4 Filter)
  (connect AttributeFilter_6 InsideFilter AttributeFilter_4 In)
  (connect IncomingDataTagsTables Out AttributeFilter_5 Filter)
  (connect AttributeFilter_4 InsideFilter AttributeFilter_5 In)
  (connect IncomingDataTagsTables Out AttributeFilter_6 Filter)
  (connect RemainingDataTagsTable Out AttributeFilter_6 In)
  (connect IncomingDataTagsTables Out AttributeFilter_7 Filter)
  (connect AttributeFilter_5 InsideFilter AttributeFilter_7 In)
  (connect IncomingDataTagsTables Out AttributeFilter_8 Filter)
  (connect AttributeFilter_7 InsideFilter AttributeFilter_8 In)
  (connect IncomingDataTagsTables Out AttributeFilter_10 Filter)
  (connect AttributeFilter_8 InsideFilter AttributeFilter_10 In)
  (connect IncomingDataTagsTables Out AttributeFilter_11 Filter)
  (connect AttributeFilter_2 OutsideFilter AttributeFilter_11 In)
  (connect AttributeMathsOp_3 Out AttributeMathsOp_0 InA)
  (connect AttributeMathsOp_3 Out AttributeMathsOp_0 InB)
  (connect Branch_12 Output%20B AttributeMathsOp_3 InA)
  (connect Branch_12 Output%20B AttributeMathsOp_3 InB)
  (connect Branch_10 Output%20B AttributeReduce_0 In)
  (connect AttributeStringOp_10 Out AttributeReduce_2 In)
  (connect Branch_35 Output%20B AttributeReduce_20 In)
  (connect GetTags_1 Out AttributeStringOp_8 InA)
  (connect AttributeStringOp_8 Out AttributeStringOp_10 InA)
  (connect AttributeStringOp_8 Out AttributeStringOp_10 InB)
  (connect CopyAttributes_0 Out BoundsModifier_0 In)
  (connect Branch_11 Output%20B BoundsModifier_1 In)
  (connect GetLoopIndex_0 Out Branch_0 bOutputToB)
  (connect DefaultInputNode SourcePoints Branch_0 In)
  (connect DataNum_0 Out Branch_1 bOutputToB)
  (connect Branch_0 Output%20B Branch_1 In)
  (connect DataNum_0 Out Branch_2 bOutputToB)
  (connect Branch_38 Output%20B Branch_2 In)
  (connect DataNum_0 Out Branch_3 bOutputToB)
  (connect IncomingDataTagsTables Out Branch_3 In)
  (connect CopyAttribute_9 Out Branch_4 In)
  (connect DataNum_1 Out Branch_9 bOutputToB)
  (connect Branch_1 Output%20A Branch_9 In)
  (connect Branch_28 Output%20B Branch_9 In)
  (connect DataNum_1 Out Branch_10 bOutputToB)
  (connect DeleteAttributes_0 Out Branch_10 In)
  (connect DataNum_1 Out Branch_11 bOutputToB)
  (connect Branch_38 Output%20B Branch_11 In)
  (connect GetElementsCount_0 Out Branch_12 bOutputToB)
  (connect ToPoint_0 Out Branch_12 In)
  (connect GetPointsCount_26 Out Branch_28 bOutputToB)
  (connect ToPoint_24 Out Branch_28 In)
  (connect DataNum_0 Out Branch_35 bOutputToB)
  (connect DeleteAttributes_19 Out Branch_35 In)
  (connect GetLoopIndex_0 Out Branch_38 bOutputToB)
  (connect DefaultInputNode RemainingPoints Branch_38 In)
  (connect MergeAttributes_7 Out CopyAttribute_9 Source)
  (connect MergeAttributes_7 Out CopyAttribute_9 Target)
  (connect CopyAttributes_1 Out CopyAttributes_0 Source)
  (connect CopyAttributes_1 Out CopyAttributes_0 Target)
  (connect Branch_9 Output%20B CopyAttributes_1 Source)
  (connect Branch_9 Output%20B CopyAttributes_1 Target)
  (connect DeleteAttributes_19 Out DataNum_0 In)
  (connect DeleteAttributes_0 Out DataNum_1 In)
  (connect MergeAttributes_1 Out DeleteAttributes_0 In)
  (connect AttributeMathsOp_0 Out DeleteAttributes_1 In)
  (connect MergeAttributes_0 Out DeleteAttributes_19 In)
  (connect Branch_3 Output%20B DeleteTags_6 In)
  (connect FilterDataByIndex_0 InsideFilter Difference_0 Differences)
  (connect BoundsModifier_0 Out Difference_0 Source)
  (connect FilterDataByTag_7 InsideFilter Difference_22 Differences)
  (connect Branch_1 Output%20B Difference_22 Source)
  (connect BoundsModifier_1 Out FilterDataByIndex_0 In)
  (connect AttributeReduce_0 Out FilterDataByIndex_0 SelectedIndices)
  (connect Branch_2 Output%20B FilterDataByIndex_17 In)
  (connect AttributeReduce_20 Out FilterDataByIndex_17 SelectedIndices)
  (connect FilterDataByIndex_17 InsideFilter FilterDataByTag_7 In)
  (connect AttributeReduce_2 Out FilterDataByTag_7 SelectedTags)
  (connect Branch_0 Output%20A Gather_0 In)
  (connect Branch_38 Output%20B Gather_0 In2)
  (connect Branch_9 Output%20A Gather_0 In3)
  (connect DeleteAttributes_1 Out Gather_0 In4)
  (connect ToPoint_0 Out GetElementsCount_0 In)
  (connect ToPoint_24 Out GetPointsCount_26 In)
  (connect DeleteTags_6 Out GetTags_1 In)
  (connect AttributeFilter_1 InsideFilter MergeAttributes_0 In)
  (connect AttributeFilter_10 InsideFilter MergeAttributes_1 In)
  (connect TagsToAttributeSet_4 Out MergeAttributes_7 In)
  (connect Branch_38 Output%20B TagsToAttributeSet_4 In)
  (connect Branch_0 Output%20B TagsToAttributeSet_13 In)
  (connect Difference_0 Out ToPoint_0 In)
  (connect Difference_22 Out ToPoint_24 In)
  (output RemainingPoints Gather_0))
