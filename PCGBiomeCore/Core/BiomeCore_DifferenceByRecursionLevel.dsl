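; PCG loop subgraph for Biome Core that trims one set of child points against the other child sets according to RecursionLevel.
; It excludes the current loop index from ChildsData, then keeps points whose RecursionLevel is at or below that of the current set.
; A binary Difference subtracts those points from the current set and converts the result to points.
; Element-count branches drop empty inputs and results, outputting only non-empty RemainingChilds.
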(graph BiomeCore_DifferenceByRecursionLevel
  (inputs (ChildsData :type Point))
  (outputs (RemainingChilds :type Point :advanced))
  (bind GetLoopIndex_7 (Get%20Loop%20Index))
  (bind FilterDataByIndex_8 (Filter%20Data%20By%20Index))
  (bind GetPointsCount_0 (Get%20Element%20Count))
  (bind Branch_1 (Branch))
  (bind GetAttributeFromPointIndex_3 (Get%20Attribute%20From%20Point%20Index :bForceOutputAttributeToBeInElementsDomain true :InputSource RecursionLevel :OutputAttributeName RecursionLevel))
  (bind AttributeFilter_1 (Filter%20Attribute%20Elements :Operator LesserOrEqual :TargetAttribute RecursionLevel :ThresholdAttribute @Last))
  (bind Difference_0 (Difference :bDiffMetadata false :DensityFunction Binary))
  (bind ToPoint_2 (To%20Point))
  (bind GetPointsCount_5 (Get%20Element%20Count))
  (bind Branch_6 (Branch))
  (connect GetAttributeFromPointIndex_3 Attribute AttributeFilter_1 Filter)
  (connect FilterDataByIndex_8 OutsideFilter AttributeFilter_1 In)
  (connect GetPointsCount_0 Out Branch_1 bOutputToB)
  (connect DefaultInputNode In Branch_1 In)
  (connect GetPointsCount_5 Out Branch_6 bOutputToB)
  (connect ToPoint_2 Out Branch_6 In)
  (connect AttributeFilter_1 InsideFilter Difference_0 Differences)
  (connect Branch_1 Output%20B Difference_0 Source)
  (connect DefaultInputNode ChildsData FilterDataByIndex_8 In)
  (connect GetLoopIndex_7 Out FilterDataByIndex_8 SelectedIndices)
  (connect Branch_1 Output%20B GetAttributeFromPointIndex_3 In)
  (connect DefaultInputNode In GetPointsCount_0 In)
  (connect ToPoint_2 Out GetPointsCount_5 In)
  (connect Difference_0 Out ToPoint_2 In)
  (output RemainingChilds Branch_6.Output%20B))
