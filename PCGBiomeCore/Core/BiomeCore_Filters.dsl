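; PCG subgraph of Biome Core that applies the biome global filters from its Filters parameter to incoming points.
; It backs up point color, then branches on whether filter data exists, either bypassing or running BiomeCore_RunFiltersLoop over filters partitioned by index.
; Both paths restore the color from the backup, keep only the AssetID and DataIndex attributes, and clear data tags before output.
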(graph BiomeCore_Filters
  (params (Filters :type (Array (Struct /PCGBiomeCore/Filters/Setup/BiomeFilter.BiomeFilter))))
  (bind DeleteAttributes_5 (Delete%20Attributes :Operation KeepSelectedAttributes :SelectedAttributes "AssetID, DataIndex"))
  (bind DeleteTags_7 (Delete%20Tags :bTokenizeOnWhiteSpace true :Operation KeepOnlySelectedTags))
  ;[0 0 0 1 32] Biome Global Filters
  (bind AttributePartition_2 (Attribute%20Partition :bTokenizeOnWhiteSpace true :PartitionAttributeSelectors [$Index] :In Filters))
  (bind FilterDataByAttribute_5 (Filter%20Data%20By%20Attribute :Attribute FilterAttribute :In Filters))
  (bind DataNum_1 (Data%20Count))
  (bind Branch_4 (Branch))
  ;;[1 0 0] Color Backup
  (bind CopyAttribute_1 (Copy%20Attributes :InputSource $Color :OutputTarget ColorBackup))
  ;;[0 0 0] Filters Loop
  (bind Loop_0 (Loop :bTokenizeOnWhiteSpace true :Graph /PCGBiomeCore/Core/BiomeCore_RunFiltersLoop.BiomeCore_RunFiltersLoop))
  ;;[1 0 0] Color Restore
  (bind CopyAttribute_3 (Copy%20Attributes :InputSource ColorBackup :OutputTarget $Color))
  (connect DataNum_1 Out Branch_4 bOutputToB)
  (connect CopyAttribute_1 Out Branch_4 In)
  (connect DefaultInputNode In CopyAttribute_1 Source)
  (connect DefaultInputNode In CopyAttribute_1 Target)
  (connect Branch_4 Output%20A CopyAttribute_3 Source)
  (connect Loop_0 Points CopyAttribute_3 Source)
  (connect Branch_4 Output%20A CopyAttribute_3 Target)
  (connect Loop_0 Points CopyAttribute_3 Target)
  (connect FilterDataByAttribute_5 InsideFilter DataNum_1 In)
  (connect CopyAttribute_3 Out DeleteAttributes_5 In)
  (connect DeleteAttributes_5 Out DeleteTags_7 In)
  (connect AttributePartition_2 Out Loop_0 Filters)
  (connect Branch_4 Output%20B Loop_0 Points)
  (output Out DeleteTags_7))
