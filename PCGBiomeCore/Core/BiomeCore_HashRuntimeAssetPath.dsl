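; PCG subgraph of Biome Core that resolves each runtime option's RuntimeAssembly asset path and hashes it into RuntimeAssemblyHashValue.
; Entries with GenerationMethod 0 get an empty path, written back by index before a cast to string.
; Entries with GenerationMethod 1 derive it from the base option's Mesh path, using string append and replace to suffix package and asset names with "_MS".
; These paths are merged by index before hashing.
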(graph BiomeCore_HashRuntimeAssetPath
  (inputs (RuntimeOptions :type Param) (BaseAssetOptions :type Param))
  (bind GetEntriesCount_1 (Get%20Element%20Count))
  (bind Branch_1 (Branch))
  (bind CopyAttributes_2 (Copy%20Attributes :InputSource $Index :OutputTarget index))
  (bind AttributeFilter_1 (Filter%20Attribute%20Elements :AttributeTypes (struct :Type Integer32) :bUseConstantThreshold true :Operator Equal :TargetAttribute GenerationMethod :ThresholdAttribute @Last))
  (bind GetEntriesCount_0 (Get%20Element%20Count))
  (bind Branch_0 (Branch))
  (bind AddAttribute_0 (Add%20Attribute :AttributeTypes (struct :Type SoftObjectPath) :OutputTarget RuntimeAssembly))
  (bind DeleteAttributes_2 (Delete%20Attributes :Operation KeepSelectedAttributes :SelectedAttributes "RuntimeAssembly,index"))
  (bind MatchAndSetAttributes_1 (Match%20And%20Set%20Attributes :bMatchAttributes true :InputAttribute $Index :MatchAttribute index))
  (bind AttributeCast_0 (Attribute%20Cast :InputSource RuntimeAssembly :OutputType String))
  (bind DeleteAttributes_0 (Delete%20Attributes :Operation KeepSelectedAttributes :SelectedAttributes "Mesh"))
  (bind CopyAttributes_0 (Copy%20Attributes :InputSource GenerationMethod :OutputTarget GenerationMethod))
  (bind CopyAttributes_1 (Copy%20Attributes :InputSource $Index :OutputTarget index))
  (bind AttributeFilter_0 (Filter%20Attribute%20Elements :AttributeTypes (struct :Type Integer32 :Int32Value 1) :bUseConstantThreshold true :Operator Equal :TargetAttribute GenerationMethod :ThresholdAttribute @Last))
  (bind CreateAttribute_0 (Create%20Constant :AttributeTypes (struct :Type String :StringValue "_MS")))
  (bind AttributeStringOp_0 (Attribute%20String%20Op :InputSource1 Mesh :OutputTarget RuntimeAssembly))
  (bind CreateAttribute_1 (Create%20Constant :AttributeTypes (struct :Type String :StringValue ".")))
  (bind AttributeStringOp_1 (Attribute%20String%20Op))
  (bind AttributeStringOp_2 (Attribute%20String%20Op :InputSource1 RuntimeAssembly :Operation Replace))
  (bind DeleteAttributes_1 (Delete%20Attributes :Operation KeepSelectedAttributes :SelectedAttributes "RuntimeAssembly,index"))
  (bind MatchAndSetAttributes_0 (Match%20And%20Set%20Attributes :bMatchAttributes true :InputAttribute $Index :MatchAttribute index))
  (bind HashAttribute_0 (Hash%20Attribute :InputSource1 RuntimeAssembly :OutputTarget RuntimeAssemblyHashValue))
  (connect Branch_0 Output%20B AddAttribute_0 In)
  (connect MatchAndSetAttributes_1 Out AttributeCast_0 In)
  (connect CopyAttributes_1 Out AttributeFilter_0 In)
  (connect CopyAttributes_2 Out AttributeFilter_1 In)
  (connect AttributeFilter_0 InsideFilter AttributeStringOp_0 InA)
  (connect CreateAttribute_0 Out AttributeStringOp_0 InB)
  (connect CreateAttribute_0 Out AttributeStringOp_1 InA)
  (connect CreateAttribute_1 Out AttributeStringOp_1 InB)
  (connect AttributeStringOp_0 Out AttributeStringOp_2 In)
  (connect AttributeStringOp_1 Out AttributeStringOp_2 Replace)
  (connect CreateAttribute_1 Out AttributeStringOp_2 Search)
  (connect GetEntriesCount_0 Out Branch_0 bOutputToB)
  (connect AttributeFilter_1 InsideFilter Branch_0 In)
  (connect GetEntriesCount_1 Out Branch_1 bOutputToB)
  (connect DefaultInputNode RuntimeOptions Branch_1 In)
  (connect Branch_1 Output%20B CopyAttributes_0 Source)
  (connect DeleteAttributes_0 Out CopyAttributes_0 Target)
  (connect CopyAttributes_0 Out CopyAttributes_1 Source)
  (connect CopyAttributes_0 Out CopyAttributes_1 Target)
  (connect Branch_1 Output%20B CopyAttributes_2 Source)
  (connect Branch_1 Output%20B CopyAttributes_2 Target)
  (connect DefaultInputNode BaseAssetOptions DeleteAttributes_0 In)
  (connect AttributeStringOp_2 Out DeleteAttributes_1 In)
  (connect AddAttribute_0 Out DeleteAttributes_2 In)
  (connect AttributeFilter_1 InsideFilter GetEntriesCount_0 In)
  (connect DefaultInputNode RuntimeOptions GetEntriesCount_1 In)
  (connect MatchAndSetAttributes_0 Out HashAttribute_0 In)
  (connect AttributeCast_0 Out MatchAndSetAttributes_0 In)
  (connect DeleteAttributes_1 Out MatchAndSetAttributes_0 Match%20Data)
  (connect Branch_1 Output%20B MatchAndSetAttributes_1 In)
  (connect DeleteAttributes_2 Out MatchAndSetAttributes_1 Match%20Data)
  (output Out HashAttribute_0))
