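; PCG Biome Core graph that applies a "solo" effect to asset entries flagged with the Isolate attribute.
; Entries are split on Isolate, and the isolated count drives a Branch on the remaining entries.
; If any are isolated, the others get empty Mesh, Actor, Assembly and ChildAssets paths, keeping their entries but clearing their assets.
; Otherwise they pass through unchanged, and all paths merge into one attribute set.
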(graph BiomeCore_IsolateAssets
  (inputs (Attribute%20Filter%20In :type Param))
  (outputs (Merge%20Attributes%20Out :type Param))
  (bind AttributeFilter_0 (Filter%20Attribute%20Elements :AttributeTypes (struct :Type Boolean :BoolValue true) :bUseConstantThreshold true :bUseSpatialQuery true :Operator Equal :TargetAttribute Isolate))
  (bind GetEntriesCount_3 (Get%20Element%20Count))
  (bind Branch_1 (Branch))
  (bind CreateAttribute_5 (Create%20Constant :AttributeTypes (struct :Type SoftObjectPath)))
  (bind AddAttribute_4 (Add%20Attribute :OutputTarget Mesh))
  (bind CreateAttribute_1 (Create%20Constant :AttributeTypes (struct :Type SoftClassPath)))
  (bind AddAttribute_7 (Add%20Attribute :OutputTarget Actor))
  (bind AddAttribute_8 (Add%20Attribute :OutputTarget Assembly))
  (bind AddAttribute_6 (Add%20Attribute :OutputTarget ChildAssets))
  (bind MergeAttributes_2 (Merge%20Attributes))
  (connect CreateAttribute_5 Out AddAttribute_4 Attributes)
  (connect Branch_1 Output%20B AddAttribute_4 In)
  (connect CreateAttribute_5 Out AddAttribute_6 Attributes)
  (connect AddAttribute_8 Out AddAttribute_6 In)
  (connect CreateAttribute_1 Out AddAttribute_7 Attributes)
  (connect AddAttribute_4 Out AddAttribute_7 In)
  (connect CreateAttribute_5 Out AddAttribute_8 Attributes)
  (connect AddAttribute_7 Out AddAttribute_8 In)
  (connect DefaultInputNode Attribute%20Filter%20In AttributeFilter_0 In)
  (connect GetEntriesCount_3 Out Branch_1 bOutputToB)
  (connect AttributeFilter_0 OutsideFilter Branch_1 In)
  (connect AttributeFilter_0 InsideFilter GetEntriesCount_3 In)
  (connect AddAttribute_6 Out MergeAttributes_2 In)
  (connect AttributeFilter_0 InsideFilter MergeAttributes_2 In)
  (connect Branch_1 Output%20A MergeAttributes_2 In)
  (output Merge%20Attributes%20Out MergeAttributes_2))
