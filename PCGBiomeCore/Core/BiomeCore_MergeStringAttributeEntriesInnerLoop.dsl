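; PCG Biome Core loop-body subgraph that merges string attribute entries into one comma-separated string.
; The loop index drives a Branch, so the first iteration passes In through unchanged and later iterations append a "," to it.
; A final string op joins that result with the Out param input, placing In first.
; The result returns on Out, matching the input pin name so the string can accumulate across iterations.
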(graph BiomeCore_MergeStringAttributeEntriesInnerLoop
  (inputs (Out :type Param))
  (bind GetLoopIndex_6 (Get%20Loop%20Index))
  (bind Branch_9 (Branch))
  (bind CreateAttribute_3 (Create%20Constant :AttributeTypes (struct :Type String :StringValue ",")))
  (bind AttributeStringOp_2 (Attribute%20String%20Op))
  (bind AttributeStringOp_0 (Attribute%20String%20Op))
  (connect AttributeStringOp_2 Out AttributeStringOp_0 InA)
  (connect Branch_9 Output%20A AttributeStringOp_0 InA)
  (connect DefaultInputNode Out AttributeStringOp_0 InB)
  (connect Branch_9 Output%20B AttributeStringOp_2 InA)
  (connect CreateAttribute_3 Out AttributeStringOp_2 InB)
  (connect GetLoopIndex_6 Out Branch_9 bOutputToB)
  (connect DefaultInputNode In Branch_9 In)
  (output Out AttributeStringOp_0))
