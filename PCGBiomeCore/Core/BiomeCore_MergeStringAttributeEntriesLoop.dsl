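; PCG subgraph in Biome Core that reduces a named string attribute to its distinct values and hands them to a merge loop.
; It keeps only the attribute named by the StringAttribute parameter, then partitions the data by that attribute.
; Get Attribute Set from Index takes the default-index entry of each partition, collapsing duplicates.
; These entries feed a Loop over BiomeCore_MergeStringAttributeEntriesInnerLoop, whose result is the graph output.
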(graph BiomeCore_MergeStringAttributeEntriesLoop
  (params (StringAttribute :type String))
  (bind DeleteAttributes_2 (Delete%20Attributes :Operation KeepSelectedAttributes :SelectedAttributes-param StringAttribute))
  (bind AttributePartition_0 (Attribute%20Partition :PartitionAttributeSelectors [] :PartitionAttributeNames-param StringAttribute))
  (bind GetAttributeFromIndex_1 (Get%20Attribute%20Set%20from%20Index))
  (bind Loop_4 (Loop :Graph /PCGBiomeCore/Core/BiomeCore_MergeStringAttributeEntriesInnerLoop.BiomeCore_MergeStringAttributeEntriesInnerLoop))
  (connect DeleteAttributes_2 Out AttributePartition_0 In)
  (connect DefaultInputNode In DeleteAttributes_2 In)
  (connect AttributePartition_0 Out GetAttributeFromIndex_1 In)
  (connect GetAttributeFromIndex_1 Out Loop_4 In)
  (output Out Loop_4))
