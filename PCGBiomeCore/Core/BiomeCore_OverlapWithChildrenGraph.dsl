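; PCG subgraph in Biome Core that gates data on an asset's OverlapWithChildren setting.
; A Delete Attributes node keeps only that attribute from the Asset param and feeds it into a Branch as the output-to-B override.
; Input data reaches the graph output through Output A only when the flag is false; when true it routes to the unconnected Output B and is dropped.
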(graph BiomeCore_OverlapWithChildrenGraph
  (inputs (Asset :type Param))
  (bind DeleteAttributes_0 (Delete%20Attributes :bTokenizeOnWhiteSpace true :Operation KeepSelectedAttributes :SelectedAttributes "OverlapWithChildren"))
  (bind Branch_4 (Branch))
  (connect DeleteAttributes_0 Out Branch_4 bOutputToB)
  (connect DefaultInputNode In Branch_4 In)
  (connect DefaultInputNode Asset DeleteAttributes_0 In)
  (output Out Branch_4.Output%20A))
