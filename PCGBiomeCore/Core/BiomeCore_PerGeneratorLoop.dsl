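; PCG loop subgraph of the biome core that assigns weighted assets to one generator's points by BiomeColor.
; Asset sets are filtered by generator and subtype tags, or by subtype and AssetType for "Typed" data, warning when none match.
; A Proxy node optionally runs the generator's spatial noise settings to weight the match.
; Matched points get bounds, transform, extents scale and upward orientation, plus the loop index as DataIndex for restoring order later.
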(graph BiomeCore_PerGeneratorLoop
  (inputs (PointDataByGenerator :type Point) (TaggedAttributeSets :type Param))
  (outputs (DefinedPoints :type Point))
  (bind FilterDataByTag_1 (Filter%20Data%20By%20Tag :Operator Matches :SelectedTags "Typed"))
  (bind DataNum_3 (Data%20Count))
  (bind Branch_7 (Branch))
  (bind Branch_4 (Branch))
  (bind GetAttributeFromPointIndex_4 (Get%20Attribute%20From%20Point%20Index :bForceOutputAttributeToBeInElementsDomain true :InputSource Generator :OutputAttributeName None))
  (bind FilterByTag_2 (Filter%20Data%20By%20Tag :bTokenizeOnWhiteSpace true :Operator Matches))
  (bind GetAttributeFromPointIndex_1 (Get%20Attribute%20From%20Point%20Index :bForceOutputAttributeToBeInElementsDomain true :InputSource GeneratorSubType :OutputAttributeName None))
  (bind FilterDataByTag_3 (Filter%20Data%20By%20Tag :bTokenizeOnWhiteSpace true))
  (bind GetAttributeFromPointIndex_9 (Get%20Attribute%20From%20Point%20Index :bForceOutputAttributeToBeInElementsDomain true :InputSource GeneratorSubType :OutputAttributeName None))
  (bind FilterDataByTag_11 (Filter%20Data%20By%20Tag :bTokenizeOnWhiteSpace true))
  (bind DataNum_4 (Data%20Count))
  (bind Branch_5 (Branch))
  (bind MergeAttributes_5 (Merge%20Attributes))
  (bind GetAttributeFromIndex_2 (Get%20Attribute%20Set%20from%20Index))
  (bind GetPropertyFromObjectPath_4 (Get%20Property%20From%20Object%20Path :bSanitizeOutputAttributeName false :InputSource Generator :OutputAttributeName SpatialNoiseSettings :PropertyName Generator.GeneratorSpatialNoiseSettings))
  (bind AttributeFilter_6 (Filter%20Attribute%20Elements :AttributeTypes (struct :Type SoftObjectPath) :bUseConstantThreshold true :bUseSpatialQuery true :Operator NotEqual :TargetAttribute SpatialNoiseSettings))
  (bind GetEntriesCount_11 (Get%20Element%20Count))
  (bind Branch_10 (Branch))
  (bind GetAttributeFromPointIndex_23 (Get%20Attribute%20From%20Point%20Index :bForceOutputAttributeToBeInElementsDomain true :InputSource AssetType :OutputAttributeName None))
  (bind MergeAttributes_19 (Merge%20Attributes))
  (bind AttributeFilter_22 (Filter%20Attribute%20Elements :Operator Matches :TargetAttribute AssetType :ThresholdAttribute @Last))
  (bind DeleteAttributes_12 (Delete%20Attributes :SelectedAttributes "AssetType,Generator,GeneratorType,GeneratorSubType"))
  (bind CreateAttribute_26 (Create%20Constant :AttributeTypes (struct :Type Boolean)))
  (bind Select_29 (Select))
  (bind GetPointsCount_4 (Get%20Element%20Count))
  (bind Branch_6 (Branch :bOutputToB true))
  ;[0 0 0] (Optional) Spatial Noise Weighted Distribution
  (bind Proxy_0 (Proxy :bTagOutputsBasedOnOutputPins false :ProxyInterfaceMode ByNativeElement :SettingsClass /Script/PCG.PCGSpatialNoiseSettings))
  ;[0 0 0 1 32] Assign Assets to Root Points
  (bind MatchAndSetAttributes_1 (Match%20And%20Set%20Attributes :bKeepUnmatched false :bMatchAttributes true :bUseInputWeightAttribute true :bUseWeightAttribute true :InputAttribute $Color.xyz :InputWeightAttribute SpatialNoise :MatchAttribute BiomeColor :WeightAttribute Weight))
  ;[0 0 0] Set Mesh Bounds and Apply Scale and Upward Orientation
  (bind AttributeMathsOp_0 (Attribute%20Maths%20Op :InputSource1 $BoundsMin :InputSource2 BoundsMin :Operation Set))
  (bind AttributeMathsOp_4 (Attribute%20Maths%20Op :InputSource1 $BoundsMax :InputSource2 BoundsMax :Operation Set))
  (bind AttributeTransformOp_0 (Attribute%20Transform%20Op :InputSource1 Transform :InputSource2 $Transform))
  (bind AttributeMathsOp_6 (Attribute%20Maths%20Op :InputSource1 $Extents :InputSource2 ExtentsMultiplier :Operation Multiply))
  (bind CreateAttribute_0 (Create%20Constant :AttributeTypes (struct :Type Vector :VectorValue (0.0 0.0 1.0))))
  (bind MakeRotatorAttribute_1 (Make%20Rotator%20Attribute :InputSource2 $Rotation.right :Operation MakeRotFromZY :OutputTarget $Rotation))
  (bind AttributeRotatorOp_2 (Attribute%20Rotator%20Op :InputSource1 $Rotation :InputSource2 $Rotation :InputSource3 OrientUpward :Operation Lerp))
  ;[1 0.3333 0] Debug Spatial Noise
  (bind CopyAttribute_0 (Copy%20Attributes :InputSource SpatialNoise :OutputTarget $Color))
  (bind Debug_1 (Debug))
  ;[1 0.3929 0 0.1] LogWarning: No Asset associated with Generator Sub Type: [name]
  (bind CreateAttribute_2 (Create%20Constant :AttributeTypes (struct :Type String :StringValue "No Asset associated with Generator Sub Type:  ")))
  (bind GetAttributeFromPointIndex_5 (Get%20Attribute%20From%20Point%20Index :bForceOutputAttributeToBeInElementsDomain true :InputSource GeneratorSubType :OutputAttributeName None))
  (bind AttributeStringOp_1 (Attribute%20String%20Op))
  (bind Print%20String_0 (Print%20String :bDisplayOnNode true :bPrefixWithNode false :bPrefixWithOwner true :PrintString "Generator SubType " :PrintToScreenColor (struct :G 157 :R 255 :A 255) :PrintToScreenDuration 0.0 :Verbosity Warning))
  ;[0 0 0] Backup DataIndex to restore data ordering and tags by Generators at a later stage
  (bind GetLoopIndex_0 (Get%20Loop%20Index))
  (bind AddAttribute_0 (Add%20Attribute :OutputTarget DataIndex))
  (connect GetLoopIndex_0 Out AddAttribute_0 Attributes)
  (connect AttributeRotatorOp_2 Out AddAttribute_0 In)
  (connect GetPropertyFromObjectPath_4 Out AttributeFilter_6 In)
  (connect GetAttributeFromPointIndex_23 Attribute AttributeFilter_22 Filter)
  (connect MergeAttributes_19 Out AttributeFilter_22 In)
  (connect Branch_6 Output%20B AttributeMathsOp_0 InA)
  (connect Branch_6 Output%20B AttributeMathsOp_0 InB)
  (connect AttributeMathsOp_0 Out AttributeMathsOp_4 InA)
  (connect AttributeMathsOp_0 Out AttributeMathsOp_4 InB)
  (connect AttributeTransformOp_0 Out AttributeMathsOp_6 InA)
  (connect AttributeTransformOp_0 Out AttributeMathsOp_6 InB)
  (connect AttributeMathsOp_6 Out AttributeRotatorOp_2 InA)
  (connect MakeRotatorAttribute_1 Out AttributeRotatorOp_2 InB)
  (connect AttributeMathsOp_6 Out AttributeRotatorOp_2 Ratio)
  (connect CreateAttribute_2 Out AttributeStringOp_1 InA)
  (connect GetAttributeFromPointIndex_5 Attribute AttributeStringOp_1 InB)
  (connect AttributeMathsOp_4 Out AttributeTransformOp_0 InA)
  (connect AttributeMathsOp_4 Out AttributeTransformOp_0 InB)
  (connect DataNum_3 Out Branch_4 bOutputToB)
  (connect DefaultInputNode PointDataByGenerator Branch_4 In)
  (connect DataNum_4 Out Branch_5 bOutputToB)
  (connect DefaultInputNode PointDataByGenerator Branch_5 In)
  (connect GetPointsCount_4 Out Branch_6 bOutputToB)
  (connect MatchAndSetAttributes_1 Out Branch_6 In)
  (connect DataNum_3 Out Branch_7 bOutputToB)
  (connect DefaultInputNode TaggedAttributeSets Branch_7 In)
  (connect GetEntriesCount_11 Out Branch_10 bOutputToB)
  (connect Branch_5 Output%20B Branch_10 In)
  (connect Proxy_0 Out CopyAttribute_0 Source)
  (connect Proxy_0 Out CopyAttribute_0 Target)
  (connect FilterDataByTag_1 InsideFilter DataNum_3 In)
  (connect FilterDataByTag_3 InsideFilter DataNum_4 In)
  (connect FilterDataByTag_11 InsideFilter DataNum_4 In)
  (connect CopyAttribute_0 Out Debug_1 In)
  (connect AttributeFilter_22 InsideFilter DeleteAttributes_12 In)
  (connect MergeAttributes_5 Out DeleteAttributes_12 In)
  (connect Branch_7 Output%20A FilterByTag_2 In)
  (connect GetAttributeFromPointIndex_4 Attribute FilterByTag_2 SelectedTags)
  (connect DefaultInputNode PointDataByGenerator FilterDataByTag_1 In)
  (connect FilterByTag_2 InsideFilter FilterDataByTag_3 In)
  (connect GetAttributeFromPointIndex_1 Attribute FilterDataByTag_3 SelectedTags)
  (connect Branch_7 Output%20B FilterDataByTag_11 In)
  (connect GetAttributeFromPointIndex_9 Attribute FilterDataByTag_11 SelectedTags)
  (connect MergeAttributes_5 Out GetAttributeFromIndex_2 In)
  (connect Branch_4 Output%20A GetAttributeFromPointIndex_1 In)
  (connect Branch_4 Output%20A GetAttributeFromPointIndex_4 In)
  (connect Branch_5 Output%20A GetAttributeFromPointIndex_5 In)
  (connect Branch_4 Output%20B GetAttributeFromPointIndex_9 In)
  (connect Branch_4 Output%20B GetAttributeFromPointIndex_23 In)
  (connect AttributeFilter_6 InsideFilter GetEntriesCount_11 In)
  (connect MatchAndSetAttributes_1 Out GetPointsCount_4 In)
  (connect GetAttributeFromIndex_2 Out GetPropertyFromObjectPath_4 In)
  (connect AttributeMathsOp_6 Out MakeRotatorAttribute_1 Y)
  (connect CreateAttribute_0 Out MakeRotatorAttribute_1 Z)
  (connect Select_29 Out MatchAndSetAttributes_1 bUseInputWeightAttribute)
  (connect Branch_10 Output%20A MatchAndSetAttributes_1 In)
  (connect Proxy_0 Out MatchAndSetAttributes_1 In)
  (connect DeleteAttributes_12 Out MatchAndSetAttributes_1 Match%20Data)
  (connect FilterDataByTag_3 InsideFilter MergeAttributes_5 In)
  (connect FilterDataByTag_11 InsideFilter MergeAttributes_19 In)
  (connect Branch_5 Output%20A Print%20String_0 In)
  (connect AttributeStringOp_1 Out Print%20String_0 PrintString)
  (connect Branch_10 Output%20B Proxy_0 In)
  (connect GetPropertyFromObjectPath_4 Out Proxy_0 Settings)
  (connect DataNum_3 Out Select_29 bUseInputB)
  (connect GetEntriesCount_11 Out Select_29 Input%20A)
  (connect CreateAttribute_26 Out Select_29 Input%20B)
  (output DefinedPoints AddAttribute_0))
