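; PCG Biome Core subgraph that filters assembly points after the difference step, keeping only those whose source point remains.
; Assembly source points from both inputs yield a deduplicated AssemblyHashValue set, which is matched against assembly points from SkipFiltersPoints, dropping unmatched ones.
; Regular points from both inputs join the matched points in the output; source points are excluded.
; When Enabled is false, both inputs pass through unfiltered.
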(graph BiomeCore_PostDifference_FilterAssemblies
  (params (Enabled :type Bool :default true))
  (inputs (DifferencePoints :type Point) (SkipFiltersPoints :type Point))
  (bind Branch_1 (Branch :bOutputToB true :bOutputToB-param Enabled))
  (bind Branch_2 (Branch :bOutputToB true :bOutputToB-param Enabled))
  (bind Gather_2 (Gather))
  (bind AttributeFilter_7 (Filter%20Attribute%20Elements :AttributeTypes (struct :Type Integer32 :BoolValue true) :bUseConstantThreshold true :Operator NotEqual :TargetAttribute Assembly :ThresholdAttribute @Last))
  (bind AttributeFilter_4 (Filter%20Attribute%20Elements :AttributeTypes (struct :Type Integer32 :Int32Value 1 :BoolValue true) :bUseConstantThreshold true :Operator Equal :TargetAttribute Assembly :ThresholdAttribute @Last)) ; Assmbly Src Points
  (bind PointToAttributeSet_0 (Point%20To%20Attribute%20Set))
  (bind DataNum_1 (Data%20Count))
  (bind Branch_3 (Branch))
  (bind DeleteAttributes_0 (Delete%20Attributes :Operation KeepSelectedAttributes :SelectedAttributes "AssemblyHashValue"))
  (bind MergeAttributes_0 (Merge%20Attributes))
  (bind AttributeRemoveDuplicates_0 (Attribute%20Remove%20Duplicates))
  (bind GetElementsCount_1 (Get%20Element%20Count))
  (bind Branch_0 (Branch))
  (bind AttributeFilter_2 (Filter%20Attribute%20Elements :AttributeTypes (struct :Type Integer32 :Int32Value 2 :BoolValue true) :bUseConstantThreshold true :Operator Equal :TargetAttribute Assembly :ThresholdAttribute @Last)) ; Assmbly Pts
  (bind RemoveEmptyData_0 (Remove%20Empty%20Data))
  (bind Branch_4 (Branch))
  (bind MatchAndSetAttributes_0 (Match%20And%20Set%20Attributes :bKeepUnmatched false :bMatchAttributes true :InputAttribute AssemblyHashValue :MatchAttribute AssemblyHashValue))
  (bind AttributeFilter_3 (Filter%20Attribute%20Elements :AttributeTypes (struct :Type Integer32 :BoolValue true) :bUseConstantThreshold true :Operator Equal :TargetAttribute Assembly :ThresholdAttribute @Last)) ; Regular Points
  (bind RemoveEmptyData_3 (Remove%20Empty%20Data))
  (bind RemoveEmptyData_4 (Remove%20Empty%20Data))
  (bind Gather_0 (Gather))
  (connect AttributeFilter_7 InsideFilter AttributeFilter_2 In)
  (connect Branch_1 Output%20B AttributeFilter_3 In)
  (connect AttributeFilter_7 InsideFilter AttributeFilter_4 In)
  (connect Branch_1 Output%20B AttributeFilter_4 In)
  (connect Branch_2 Output%20B AttributeFilter_7 In)
  (connect MergeAttributes_0 Out AttributeRemoveDuplicates_0 In)
  (connect GetElementsCount_1 Out Branch_0 bOutputToB)
  (connect AttributeRemoveDuplicates_0 Out Branch_0 In)
  (connect DefaultInputNode DifferencePoints Branch_1 In)
  (connect DefaultInputNode SkipFiltersPoints Branch_2 In)
  (connect DataNum_1 Out Branch_3 bOutputToB)
  (connect PointToAttributeSet_0 Out Branch_3 In)
  (connect GetElementsCount_1 Out Branch_4 bOutputToB)
  (connect RemoveEmptyData_0 Out Branch_4 In)
  (connect PointToAttributeSet_0 Out DataNum_1 In)
  (connect Branch_3 Output%20B DeleteAttributes_0 In)
  (connect Gather_2 Out Gather_0 In)
  (connect RemoveEmptyData_3 Out Gather_0 In2)
  (connect MatchAndSetAttributes_0 Out Gather_0 In3)
  (connect RemoveEmptyData_4 Out Gather_0 In4)
  (connect Branch_1 Output%20A Gather_2 In)
  (connect Branch_2 Output%20A Gather_2 In2)
  (connect AttributeRemoveDuplicates_0 Out GetElementsCount_1 In)
  (connect Branch_4 Output%20B MatchAndSetAttributes_0 In)
  (connect Branch_0 Output%20B MatchAndSetAttributes_0 Match%20Data)
  (connect DeleteAttributes_0 Out MergeAttributes_0 In)
  (connect AttributeFilter_4 InsideFilter PointToAttributeSet_0 In)
  (connect AttributeFilter_2 InsideFilter RemoveEmptyData_0 In)
  (connect AttributeFilter_3 InsideFilter RemoveEmptyData_3 In)
  (connect AttributeFilter_7 OutsideFilter RemoveEmptyData_4 In)
  (output Out Gather_0))
