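; PCG Biome Core subgraph that prepares a given asset's child assets for recursion.
; It selects asset table rows matching AssetID with non-empty ChildAssets, resolves their properties in a loop, and resets AssetID to the element index.
; It then adds bounds from meshes and assemblies, and converts weights to filter bounds per TransformGraph.
; Partition, loop and merge stages handle grouping, with final output sorted and partitioned by AssetID.
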(graph BiomeCore_PrepareChildRecursion
  (inputs (AssetTable :type Param :advanced) (AssetID :type Param))
  ;[0 0 0 1 32] Prepare Child Asset for Recursion
  (bind AttributeFilter_4 (Filter%20Attribute%20Elements :AttributeTypes (struct :Type SoftObjectPath) :bUseConstantThreshold true :Operator NotEqual :TargetAttribute ChildAssets))
  (bind AttributeFilter_7 (Filter%20Attribute%20Elements :bGenerateOutputDataEvenIfEmpty false :Operator Equal :TargetAttribute AssetID :ThresholdAttribute AssetID))
  (bind Loop_3 (Loop :bTokenizeOnWhiteSpace true :Graph /PCGBiomeCore/Core/BiomeCore_ResolveAssetProperties.BiomeCore_ResolveAssetProperties))
  (bind AttributeMathsOp_10 (Attribute%20Maths%20Op :InputSource1 $Index :InputSource2 $Index :Operation Set :OutputTarget AssetID))
  (bind MergeAttributes_2 (Merge%20Attributes))
  (bind SortAttributes_3 (Sort%20Attributes :bUseStableSort false :SortAttributes [(struct :InputSource AssetID)]))
  (bind AttributePartition_9 (Attribute%20Partition :bTokenizeOnWhiteSpace true :PartitionAttributeSelectors [AssetID]))
  ;;[0 0 0] Store Meshes Bounds to assets tables
  (bind AttributePartition_1 (Attribute%20Partition :bTokenizeOnWhiteSpace true :PartitionAttributeSelectors [Mesh SkeletalMesh]))
  (bind Loop_0 (Loop :bTokenizeOnWhiteSpace true :Graph /PCGBiomeCore/Core/BiomeCore_BoundsFromMeshLoop.BiomeCore_BoundsFromMeshLoop))
  ;;[0 0 0] Store Assemblies Bounds to assets tables
  (bind MergeAttributes_0 (Merge%20Attributes))
  (bind Subgraph_0 (Subgraph :Graph /PCGBiomeCore/Core/BiomeCore_BoundsFromAssembly.BiomeCore_BoundsFromAssembly))
  ;;[0 0 0] Convert weights to filter bounds - per transform graph
  (bind AttributePartition_24 (Attribute%20Partition :bTokenizeOnWhiteSpace true :PartitionAttributeSelectors [TransformGraph]))
  (bind Loop_25 (Loop :bTokenizeOnWhiteSpace true :Graph /PCGBiomeCore/Core/BiomeCore_PrepareChildWeights.BiomeCore_PrepareChildWeights))
  (bind MergeAttributes_28 (Merge%20Attributes))
  (connect DefaultInputNode AssetTable AttributeFilter_4 In)
  (connect DefaultInputNode AssetID AttributeFilter_7 Filter)
  (connect AttributeFilter_4 InsideFilter AttributeFilter_7 In)
  (connect Loop_3 Out AttributeMathsOp_10 InA)
  (connect Loop_3 Out AttributeMathsOp_10 InB)
  (connect AttributeMathsOp_10 Out AttributePartition_1 In)
  (connect SortAttributes_3 Out AttributePartition_9 In)
  (connect MergeAttributes_2 Out AttributePartition_24 In)
  (connect AttributePartition_1 Out Loop_0 In)
  (connect AttributeFilter_7 InsideFilter Loop_3 In)
  (connect AttributePartition_24 Out Loop_25 In)
  (connect Loop_0 Out MergeAttributes_0 In)
  (connect Subgraph_0 Out MergeAttributes_2 In)
  (connect Loop_25 Out MergeAttributes_28 In)
  (connect MergeAttributes_28 Out SortAttributes_3 In)
  (connect MergeAttributes_0 Out Subgraph_0 In)
  (output Out AttributePartition_9))
