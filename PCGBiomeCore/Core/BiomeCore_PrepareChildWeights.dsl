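; PCG subgraph for Biome Core that prepares child entries for weighted selection by giving each FilterLowerBound and FilterUpperBound attributes.
; A Branch passes zero or one child straight out with default 0–1 bounds.
; For multiple children, it divides 1 by the count, multiplies by weight and divides by the sum to normalize shares.
; It then partitions by index and loops BiomeCore_SetChildFilterBoundsFromWeights over each child.
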(graph BiomeCore_PrepareChildWeights
  (bind AddAttribute_14 (Add%20Attribute :OutputTarget FilterLowerBound))
  (bind AddAttribute_15 (Add%20Attribute :AttributeTypes (struct :DoubleValue 1.0) :OutputTarget FilterUpperBound))
  (bind CreateAttribute_12 (Create%20Constant :AttributeTypes (struct :Type Integer32 :Int32Value 1 :DoubleValue 1.0)))
  (bind GetEntriesCount_10 (Get%20Element%20Count))
  (bind AttributeCompareOp_11 (Attribute%20Compare%20Op :Operation Greater))
  (bind Branch_13 (Branch))
  (bind GetEntriesCount_0 (Get%20Element%20Count))
  (bind CreateAttribute_1 (Create%20Constant :AttributeTypes (struct :DoubleValue 1.0)))
  (bind AddAttribute_3 (Add%20Attribute :OutputTarget Quantity))
  (bind AttributeMathsOp_2 (Attribute%20Maths%20Op :Operation Divide))
  (bind AttributeMathsOp_4 (Attribute%20Maths%20Op :InputSource2 weight :Operation Multiply))
  (bind AttributeReduce_5 (Attribute%20Reduce :Operation Sum))
  (bind AttributeMathsOp_6 (Attribute%20Maths%20Op :Operation Divide))
  (bind AttributePartition_7 (Attribute%20Partition :bTokenizeOnWhiteSpace true :PartitionAttributeSelectors [$Index]))
  (bind Loop_8 (Loop :bTokenizeOnWhiteSpace true :Graph /PCGBiomeCore/Core/BiomeCore_SetChildFilterBoundsFromWeights.BiomeCore_SetChildFilterBoundsFromWeights))
  (connect CreateAttribute_1 Out AddAttribute_3 Attributes)
  (connect Branch_13 Output%20B AddAttribute_3 In)
  (connect DefaultInputNode In AddAttribute_14 In)
  (connect AddAttribute_14 Out AddAttribute_15 In)
  (connect GetEntriesCount_10 Out AttributeCompareOp_11 InA)
  (connect CreateAttribute_12 Out AttributeCompareOp_11 InB)
  (connect AddAttribute_3 Out AttributeMathsOp_2 InA)
  (connect GetEntriesCount_0 Out AttributeMathsOp_2 InB)
  (connect AttributeMathsOp_2 Out AttributeMathsOp_4 InA)
  (connect AttributeMathsOp_2 Out AttributeMathsOp_4 InB)
  (connect AttributeMathsOp_4 Out AttributeMathsOp_6 InA)
  (connect AttributeReduce_5 Out AttributeMathsOp_6 InB)
  (connect AttributeMathsOp_6 Out AttributePartition_7 In)
  (connect AttributeMathsOp_4 Out AttributeReduce_5 In)
  (connect AttributeCompareOp_11 Out Branch_13 bOutputToB)
  (connect AddAttribute_15 Out Branch_13 In)
  (connect Branch_13 Output%20B GetEntriesCount_0 In)
  (connect DefaultInputNode In GetEntriesCount_10 In)
  (connect AttributePartition_7 Out Loop_8 CurrentChild)
  (output Out Branch_13.Output%20A)
  (output Out Loop_8))
