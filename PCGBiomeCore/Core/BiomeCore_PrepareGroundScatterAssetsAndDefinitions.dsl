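; PCG graph that builds the ground scatter list of assets and biome definitions from enabled actors tagged PCG_BiomeActor.
; Each actor's DefaultDefinition is overridden by its external Definition's BiomeDefinition, if present, then gets a BiomeID hashed from biome color and actor reference.
; RuntimeAssets pass through a loop resolving properties and generator graphs, then a subgraph assigns definitions to them.
; Only entries with a non-empty GeneratorGraph are output.
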(graph BiomeCore_PrepareGroundScatterAssetsAndDefinitions
  (outputs (AssetsAndDefinitions :type Param))
  (bind GetActorData_0 (Get%20Actor%20Data :ActorSelector (struct :ActorSelectionTag PCG_BiomeActor) :Mode GetActorReference))
  (bind FilterDataByTag_1 (Filter%20Data%20By%20Tag :SelectedTags "Enabled"))
  (bind GetPropertyFromObjectPath_0 (Get%20Property%20From%20Object%20Path :bSanitizeOutputAttributeName false :OutputAttributeName Definition :PropertyName Definition))
  (bind CopyAttributes_0 (Copy%20Attributes :InputSource ActorReference))
  (bind GetPropertyFromObjectPath_1 (Get%20Property%20From%20Object%20Path :bSanitizeOutputAttributeName false :InputSource Definition :OutputAttributeName None :PropertyName BiomeDefinition))
  (bind GetPropertyFromObjectPath_2 (Get%20Property%20From%20Object%20Path :bSanitizeOutputAttributeName false :InputSource ActorReference :OutputAttributeName Definition :PropertyName DefaultDefinition))
  (bind GetPropertyFromObjectPath_3 (Get%20Property%20From%20Object%20Path :bSanitizeOutputAttributeName false :InputSource ActorReference :OutputAttributeName RuntimeAssets :PropertyName RuntimeAssets))
  (bind Loop_0 (Loop :Graph /PCGBiomeCore/Runtime/GroundScatter/BiomeCore_ResolveAssetPropertiesAndGeneratorGraphLoop.BiomeCore_ResolveAssetPropertiesAndGeneratorGraphLoop))
  (bind CreateConstant_1 (Create%20Constant :AttributeTypes (struct :Type Boolean :BoolValue true)))
  (bind Subgraph_1 (Subgraph :Graph /PCGBiomeCore/Core/BiomeCore_AssignDefinitionsToAssets.BiomeCore_AssignDefinitionsToAssets))
  (bind AttributeFilter_0 (Filter%20Attribute%20Elements :AttributeTypes (struct :Type SoftObjectPath) :bUseConstantThreshold true :bWarnOnDataMissingAttribute false :Operator NotEqual :TargetAttribute GeneratorGraph :ThresholdAttribute @Last))
  ;[0.15 0.15 0.15 0.5] Copy Local definition first, Then override with External if it exists
  (bind CopyAttributes_1 (Copy%20Attributes :bCopyAllAttributes true :InputSource @Last))
  (bind CopyAttributes_2 (Copy%20Attributes :bCopyAllAttributes true :InputSource @Last))
  ;[0.15 0.15 0.15 0.5] Generate per BiomeActor and Biome Color unique Biome IDs
  (bind AttributeCast_0 (Attribute%20Cast :InputSource BiomeColor :OutputTarget BiomeID :OutputType String))
  (bind AttributeStringOp_0 (Attribute%20String%20Op :InputSource1 BiomeID :InputSource2 ActorReference))
  (bind HashAttribute_0 (Hash%20Attribute :InputSource1 BiomeID :OutputTarget BiomeID))
  (connect CopyAttributes_2 Out AttributeCast_0 In)
  (connect Subgraph_1 AssetsWithDefinitions AttributeFilter_0 In)
  (connect AttributeCast_0 Out AttributeStringOp_0 InA)
  (connect CopyAttributes_2 Out AttributeStringOp_0 InB)
  (connect FilterDataByTag_1 InsideFilter CopyAttributes_0 Source)
  (connect GetPropertyFromObjectPath_0 Out CopyAttributes_0 Target)
  (connect GetPropertyFromObjectPath_2 Out CopyAttributes_1 Source)
  (connect CopyAttributes_0 Out CopyAttributes_1 Target)
  (connect GetPropertyFromObjectPath_1 Out CopyAttributes_2 Source)
  (connect CopyAttributes_1 Out CopyAttributes_2 Target)
  (connect GetActorData_0 Out FilterDataByTag_1 In)
  (connect FilterDataByTag_1 InsideFilter GetPropertyFromObjectPath_0 In)
  (connect CopyAttributes_0 Out GetPropertyFromObjectPath_1 In)
  (connect CopyAttributes_0 Out GetPropertyFromObjectPath_2 In)
  (connect CopyAttributes_0 Out GetPropertyFromObjectPath_3 In)
  (connect AttributeStringOp_0 Out HashAttribute_0 In)
  (connect GetPropertyFromObjectPath_3 Out Loop_0 In)
  (connect CreateConstant_1 Out Subgraph_1 AllowNoGenerator)
  (connect Loop_0 Out Subgraph_1 BiomeAssets)
  (connect HashAttribute_0 Out Subgraph_1 BiomeDefinitions)
  (output AssetsAndDefinitions AttributeFilter_0.InsideFilter))
