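; PCG Biome Core helper graph that converts the tags of its input into a single Tags string attribute.
; It extracts tag names and values, then rebuilds each as a name:value string using two chained string operations.
; A join reduce merges these strings into Tags.
; A final replace substitutes an empty string for ":true", so tags with a true value appear as bare names.
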(graph BiomeCore_PrepareTagsData
  (bind GetTags_0 (Get%20Tags :bExtractTagValues true))
  (bind CreateAttribute_3 (Create%20Constant :AttributeTypes (struct :Type String :StringValue ":")))
  (bind AttributeStringOp_2 (Attribute%20String%20Op :InputSource1 Attribute))
  (bind AttributeStringOp_4 (Attribute%20String%20Op :InputSource1 Attribute :InputSource2 Values))
  (bind AttributeReduce_1 (Attribute%20Reduce :InputSource Attribute :Operation Join :OutputAttributeName Tags))
  (bind CreateAttribute_7 (Create%20Constant :AttributeTypes (struct :Type String :StringValue ":true")))
  (bind CreateAttribute_9 (Create%20Constant :AttributeTypes (struct :Type String)))
  (bind AttributeStringOp_6 (Attribute%20String%20Op :InputSource1 Tags :Operation Replace))
  (connect AttributeStringOp_4 Out AttributeReduce_1 In)
  (connect GetTags_0 Out AttributeStringOp_2 InA)
  (connect CreateAttribute_3 Out AttributeStringOp_2 InB)
  (connect AttributeStringOp_2 Out AttributeStringOp_4 InA)
  (connect AttributeStringOp_2 Out AttributeStringOp_4 InB)
  (connect AttributeReduce_1 Out AttributeStringOp_6 In)
  (connect CreateAttribute_9 Out AttributeStringOp_6 Replace)
  (connect CreateAttribute_7 Out AttributeStringOp_6 Search)
  (connect DefaultInputNode In GetTags_0 In)
  (output Out AttributeStringOp_6))
