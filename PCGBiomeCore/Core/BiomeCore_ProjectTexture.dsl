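; PCG Biome Core subgraph that projects selected texture channels onto a target point attribute, Color by default.
; The world origin is TextureWorldOrigin or, optionally, a Landscape's XY position when one exists.
; For multi-tile grids, points are merged, split by BiomeCore_TilePartitioner and tagged TilePartitioned, unless already-tagged inputs match the tile count.
; Partitioned or bypassed points feed the BiomeCore_ProjectionLoop loop, which outputs the projected points.
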(graph BiomeCore_ProjectTexture
  (params (TileSize :type Double :default 2.56e+04) (TileCountX :type Int :default 1) (TileCountY :type Int :default 1) (FlipTilesVerticalOrdering :type Bool) (UseLandscapeAsWorldOrigins :type Bool) (TextureToProjectOn :type (Object /Script/Engine.Texture)) (TextureSize :type Vector2 :default (2.56e+04 2.56e+04)) (TextureWorldOrigin :type Vector) (TextureFiltering :type (Enum /Script/PCG.EPCGTextureFilter)) (TargetAttributeName :type (Struct /Script/PCG.PCGAttributePropertySelector) :default $Color) (ChannelsToProject :type String :default "x") (Debug-Projection :type Bool))
  (bind Merge_2 (Merge%20Points))
  (bind GetActorData_2 (Get%20Actor%20Data :ActorSelector (struct :ActorFilter AllWorldActors :ActorSelection ByClass :ActorSelectionClass /Script/Landscape.Landscape) :bComponentsMustOverlapSelf false :bTrackActorsOnlyWithinBounds false :Mode GetSinglePoint))
  (bind DataNum_1 (Data%20Count))
  (bind GetAttributeFromPointIndex_1 (Get%20Attribute%20From%20Point%20Index :bForceOutputAttributeToBeInElementsDomain true :InputSource $Position.xy :OutputAttributeName Origin))
  (bind Select_2 (Select :Input%20A TextureWorldOrigin))
  (bind Select_6 (Select :bUseInputB-param UseLandscapeAsWorldOrigins :Input%20A TextureWorldOrigin))
  (bind Subgraph_0 (Subgraph :Graph /PCGBiomeCore/Core/BiomeCore_TilePartitioner.BiomeCore_TilePartitioner :FlipTilesVerticalOrdering FlipTilesVerticalOrdering :TileCountX TileCountX :TileCountY TileCountY :TileSize TileSize))
  (bind AddTags_22 (Add%20Tags :bTokenizeOnWhiteSpace true :TagsToAdd "TilePartitioned"))
  (bind Loop_3 (Loop :bTokenizeOnWhiteSpace true :Graph /PCGBiomeCore/Core/BiomeCore_ProjectionLoop.BiomeCore_ProjectionLoop :ChannelsToProject ChannelsToProject :Debug-Projection Debug-Projection :TargetAttributeName TargetAttributeName :TextureFiltering TextureFiltering :TextureSize TextureSize :TextureToProjectOn TextureToProjectOn))
  ;[0 0 0] Test for tile and data count combination to bypass or not the merge and partitioning
  (bind CreateAttribute_5 (Create%20Constant :AttributeTypes (struct :Type Integer32 :Int32Value 1 :DoubleValue 1.0) :OutputTarget Partition))
  (bind AttributeCompareOp_4 (Attribute%20Compare%20Op :Operation Greater :OutputTarget Partition :InA TileCountX))
  (bind AttributeCompareOp_8 (Attribute%20Compare%20Op :Operation Greater :OutputTarget Partition :InA TileCountY))
  (bind AttributeBooleanOp_6 (Attribute%20Boolean%20Op :OutputTarget Partition))
  (bind AttributeMathsOp_23 (Attribute%20Maths%20Op :Operation Multiply :InA TileCountX :InB TileCountY))
  (bind FilterByTag_21 (Filter%20Data%20By%20Tag :bTokenizeOnWhiteSpace true :SelectedTags "TilePartitioned"))
  (bind DataNum_12 (Data%20Count :OutputAttributeName Partition))
  (bind AttributeCompareOp_31 (Attribute%20Compare%20Op :Operation NotEqual :OutputTarget Partition))
  (bind AttributeBooleanOp_1 (Attribute%20Boolean%20Op :InputSource1 Partition :InputSource2 Partition :OutputTarget Partition))
  (bind Branch_9 (Branch))
  (connect Subgraph_0 Tile%20Partitioned%20Points AddTags_22 In)
  (connect AttributeCompareOp_31 Out AttributeBooleanOp_1 InA)
  (connect AttributeBooleanOp_6 Out AttributeBooleanOp_1 InB)
  (connect AttributeCompareOp_4 Out AttributeBooleanOp_6 InA)
  (connect AttributeCompareOp_8 Out AttributeBooleanOp_6 InB)
  (connect CreateAttribute_5 Out AttributeCompareOp_4 InB)
  (connect CreateAttribute_5 Out AttributeCompareOp_8 InB)
  (connect DataNum_12 Out AttributeCompareOp_31 InA)
  (connect AttributeMathsOp_23 Out AttributeCompareOp_31 InB)
  (connect AttributeBooleanOp_1 Out Branch_9 bOutputToB)
  (connect DefaultInputNode In Branch_9 In)
  (connect GetActorData_2 Out DataNum_1 In)
  (connect FilterByTag_21 InsideFilter DataNum_12 In)
  (connect DefaultInputNode In FilterByTag_21 In)
  (connect GetActorData_2 Out GetAttributeFromPointIndex_1 In)
  (connect AddTags_22 Out Loop_3 PointsToProject)
  (connect Branch_9 Output%20A Loop_3 PointsToProject)
  (connect Select_6 Out Loop_3 TextureWorldOrigin)
  (connect Branch_9 Output%20B Merge_2 In)
  (connect DataNum_1 Out Select_2 bUseInputB)
  (connect GetAttributeFromPointIndex_1 Attribute Select_2 Input%20B)
  (connect Select_2 Out Select_6 Input%20B)
  (connect Select_6 Out Subgraph_0 PartitionWorldOrigin)
  (connect Merge_2 Out Subgraph_0 Points%20To%20Partition)
  (output Out Loop_3.ProjectedPoints))
