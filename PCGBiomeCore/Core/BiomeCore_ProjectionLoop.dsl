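; PCG loop body for Biome Core that projects a world-placed texture onto points and stores selected color channels in a target attribute.
; The points' TileIndex picks the texture array slice, and size and origin params build the sampling transform.
; Projection colors are copied to the target attribute, then Color resets if it was not the target.
; Empty inputs still gain the attribute to keep data count stable, and a flag enables debug display.
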(graph BiomeCore_ProjectionLoop
  (params (TextureToProjectOn :type (Object /Script/Engine.Texture)) (TextureSize :type Vector2 :default (2.56e+04 2.56e+04)) (TextureWorldOrigin :type Vector) (TextureFiltering :type (Enum /Script/PCG.EPCGTextureFilter)) (TargetAttributeName :type (Struct /Script/PCG.PCGAttributePropertySelector) :default $Color) (ChannelsToProject :type String :default "x") (Debug-Projection :type Bool))
  (inputs (PointsToProject :type Point))
  (outputs (ProjectedPoints :type Point))
  (bind Branch_1 (Branch))
  (bind FilterDataByAttribute_0 (Filter%20Data%20By%20Attribute :Attribute TileIndex))
  (bind GetAttributeFromPointIndex_4 (Get%20Attribute%20From%20Point%20Index :bForceOutputAttributeToBeInElementsDomain true :InputSource TileIndex :OutputAttributeName Tile))
  (bind CreateAttribute_3 (Create%20Constant :AttributeTypes (struct :Type Vector2 :DoubleValue -2.0 :Vector2Value (2.0 2.0))))
  (bind AttributeMathsOp_2 (Attribute%20Maths%20Op :Operation Divide :InA TextureSize))
  (bind MakeTransformAttribute_0 (Make%20Transform%20Attribute :Translation TextureWorldOrigin))
  (bind GetTextureData_0 (Get%20Texture%20Data :bForceEditorOnlyCPUSampling true :bUseAbsoluteTransform true :bUseAdvancedTiling true :bUseDensitySourceChannel false :CenterOffset (0.5 0.5) :TexelSize 1e+02 :Filter-param TextureFiltering :Texture-param TextureToProjectOn))
  (bind Projection_1 (Projection :ProjectionParams (struct :bProjectPositions false :bProjectRotations false :ColorBlendMode TargetValue :AttributeMode IncludeAttributes)))
  (bind ToPoint_2 (To%20Point))
  (bind CreateAttribute_2 (Create%20Constant :AttributeTypes (struct :Type String :StringValue "$Color.")))
  (bind AttributeStringOp_1 (Attribute%20String%20Op :InB ChannelsToProject))
  (bind CopyAttribute_1 (Copy%20Attributes :InputSource $Color.x :OutputTarget ProjectedData :OutputTarget-param TargetAttributeName))
  ;[0 0 0] True if >0
  (bind GetPointsCount_0 (Get%20Element%20Count))
  ;[0 0 0] Reset Color if Target is not Color
  (bind CreateAttribute_4 (Create%20Constant :AttributeTypes (struct :Type String :StringValue "$Color")))
  (bind AttributeCompareOp_2 (Attribute%20Compare%20Op :Operation NotEqual :InA TargetAttributeName))
  (bind Branch_5 (Branch))
  (bind CreateAttribute_8 (Create%20Constant :AttributeTypes (struct :DoubleValue 1.0)))
  (bind AttributeMathsOp_7 (Attribute%20Maths%20Op :InputSource1 $Color :Operation Set))
  ;[0 0 0] To keep data count stable and avoid warnings in subsequent point filters
  (bind AddAttribute_0 (Add%20Attribute :OutputTarget-param TargetAttributeName))
  ;[1 0.3333 0] Debug Projection
  (bind Branch_0 (Branch :bOutputToB-param Debug-Projection))
  (bind TransformPoints_0 (Transform%20Points :bAbsoluteRotation true))
  (bind Debug_1 (Debug))
  (connect Branch_1 Output%20A AddAttribute_0 In)
  (connect CreateAttribute_4 Out AttributeCompareOp_2 InB)
  (connect CreateAttribute_3 Out AttributeMathsOp_2 InB)
  (connect Branch_5 Output%20B AttributeMathsOp_7 InA)
  (connect CreateAttribute_8 Out AttributeMathsOp_7 InB)
  (connect CreateAttribute_2 Out AttributeStringOp_1 InA)
  (connect ToPoint_2 Out Branch_0 In)
  (connect GetPointsCount_0 Out Branch_1 bOutputToB)
  (connect DefaultInputNode PointsToProject Branch_1 In)
  (connect AttributeCompareOp_2 Out Branch_5 bOutputToB)
  (connect CopyAttribute_1 Out Branch_5 In)
  (connect AttributeStringOp_1 Out CopyAttribute_1 InputSource)
  (connect ToPoint_2 Out CopyAttribute_1 Source)
  (connect ToPoint_2 Out CopyAttribute_1 Target)
  (connect TransformPoints_0 Out Debug_1 In)
  (connect Branch_1 Output%20B FilterDataByAttribute_0 In)
  (connect FilterDataByAttribute_0 InsideFilter GetAttributeFromPointIndex_4 In)
  (connect DefaultInputNode PointsToProject GetPointsCount_0 In)
  (connect GetAttributeFromPointIndex_4 Attribute GetTextureData_0 TextureArrayIndex)
  (connect MakeTransformAttribute_0 Out GetTextureData_0 Transform)
  (connect AttributeMathsOp_2 Out MakeTransformAttribute_0 Scale)
  (connect Branch_1 Output%20B Projection_1 In)
  (connect GetTextureData_0 Out Projection_1 Projection%20Target)
  (connect Projection_1 Out ToPoint_2 In)
  (connect Branch_0 Output%20B TransformPoints_0 In)
  (output ProjectedPoints AddAttribute_0)
  (output ProjectedPoints AttributeMathsOp_7)
  (output ProjectedPoints Branch_5.Output%20A))
