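; PCG subgraph of Biome Core that resolves biome asset entries from objects in Enabled-tagged input data into one flat attribute set of enabled entries.
; It reads the named property and its optional debug, asset, mesh, assembly, filter and runtime sub-structures, building a Transform from asset options.
; MeshOptionsPreset values override mesh options by matching backed-up indices.
; Runtime options pass through a path-hashing subgraph before everything is gathered onto the base entries.
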(graph BiomeCore_ResolveAssetProperties
  (params (InputSource :type Name :default Assets) (PropertyName :type Name :default BiomeAssets) (ExtractDebugOptions :type Bool :default true) (ExtractAssetOptions :type Bool :default true) (ExtractMeshOptions :type Bool :default true) (ExtractAssemblyOptions :type Bool :default true) (ExtractFilterOptions :type Bool :default true) (ExtractRuntimeOptions :type Bool :default true))
  (bind FilterDataByTag_0 (Filter%20Data%20By%20Tag :SelectedTags "Enabled"))
  (bind GetEntriesCount_32 (Get%20Element%20Count))
  (bind Branch_30 (Branch))
  (bind AttributeStringOp_1 (Attribute%20String%20Op :InA PropertyName))
  (bind Branch_5 (Branch :bOutputToB-param ExtractAssetOptions))
  (bind AttributeStringOp_0 (Attribute%20String%20Op :InA PropertyName))
  (bind Branch_1 (Branch :bOutputToB-param ExtractDebugOptions))
  (bind AttributeStringOp_2 (Attribute%20String%20Op :InA PropertyName))
  (bind Branch_18 (Branch :bOutputToB-param ExtractMeshOptions))
  (bind AttributeStringOp_4 (Attribute%20String%20Op :InA PropertyName))
  (bind Branch_21 (Branch :bOutputToB-param ExtractFilterOptions))
  (bind AttributeStringOp_5 (Attribute%20String%20Op :InA PropertyName))
  (bind Branch_9 (Branch :bOutputToB-param ExtractRuntimeOptions))
  (bind AttributeStringOp_3 (Attribute%20String%20Op :InA PropertyName))
  (bind Branch_7 (Branch :bOutputToB-param ExtractAssemblyOptions))
  (bind AttributeRename_3 (Attribute%20Rename :AttributeToRename AllowCollision :NewAttributeName AssemblyAllowCollision))
  (bind DeleteAttributes_1 (Delete%20Attributes :SelectedAttributes "MeshOptionsPreset"))
  (bind MakeTransformAttribute_3 (Make%20Transform%20Attribute :InputSource1 Translation :InputSource2 Rotation :InputSource3 Scale :OutputTarget Transform))
  (bind DeleteAttributes_4 (Delete%20Attributes :bTokenizeOnWhiteSpace true :SelectedAttributes "Translation,Rotation,Scale"))
  (bind MatchAndSetAttributes_0 (Match%20And%20Set%20Attributes :bMatchAttributes true :InputAttribute $Index :MatchAttribute IndexBck))
  (bind Subgraph_1 (Subgraph :Graph /PCGBiomeCore/Core/BiomeCore_HashRuntimeAssetPath.BiomeCore_HashRuntimeAssetPath))
  (bind Gather_0 (Gather))
  (bind AddAttribute_41 (Add%20Attribute :bCopyAllAttributes true))
  (bind AttributeFilter_45 (Filter%20Attribute%20Elements :AttributeTypes (struct :Type Boolean :BoolValue true) :bUseConstantThreshold true :bUseSpatialQuery true :Operator Equal :TargetAttribute Enabled))
  ;[0 0 0] Assets Base Structures
  (bind GetPropertyFromObjectPath_0 (Get%20Property%20From%20Object%20Path :bSanitizeOutputAttributeName false :OutputAttributeName None :PropertyName BiomeAssets :InputSource-param InputSource :PropertyName-param PropertyName))
  (bind MergeAttributes_16 (Merge%20Attributes))
  ;[0 0 0] Asset Options Sub-Structures
  (bind GetPropertyFromObjectPath_1 (Get%20Property%20From%20Object%20Path :bSanitizeOutputAttributeName false :OutputAttributeName None :PropertyName BiomeAssets.AssetOptions :InputSource-param InputSource))
  (bind GetPropertyFromObjectPath_6 (Get%20Property%20From%20Object%20Path :bSanitizeOutputAttributeName false :OutputAttributeName None :PropertyName BiomeAssets.DebugOptions :InputSource-param InputSource))
  (bind GetPropertyFromObjectPath_8 (Get%20Property%20From%20Object%20Path :bSanitizeOutputAttributeName false :OutputAttributeName None :PropertyName BiomeAssets.MeshOptions :InputSource-param InputSource))
  (bind GetPropertyFromObjectPath_10 (Get%20Property%20From%20Object%20Path :bSanitizeOutputAttributeName false :OutputAttributeName None :PropertyName BiomeAssets.FilterOptions :InputSource-param InputSource))
  (bind GetPropertyFromObjectPath_12 (Get%20Property%20From%20Object%20Path :bSanitizeOutputAttributeName false :OutputAttributeName None :PropertyName BiomeAssets.RuntimeOptions :InputSource-param InputSource))
  (bind GetPropertyFromObjectPath_2 (Get%20Property%20From%20Object%20Path :bSanitizeOutputAttributeName false :OutputAttributeName None :PropertyName BiomeAssets.AssemblyOptions :InputSource-param InputSource))
  (bind MergeAttributes_28 (Merge%20Attributes))
  (bind MergeAttributes_18 (Merge%20Attributes))
  (bind MergeAttributes_20 (Merge%20Attributes))
  (bind MergeAttributes_22 (Merge%20Attributes))
  (bind MergeAttributes_24 (Merge%20Attributes))
  (bind MergeAttributes_0 (Merge%20Attributes))
  ;[0 0 0 0.5] MeshOptionsPreset override
  (bind Branch_0 (Branch :bOutputToB-param ExtractMeshOptions))
  (bind AttributeCompareOp_0 (Attribute%20Compare%20Op :InputSource1 MeshOptionsPreset :Operation NotEqual))
  (bind AttributePartition_0 (Attribute%20Partition :PartitionAttributeSelectors [$Index]))
  (bind GetPropertyFromObjectPath_4 (Get%20Property%20From%20Object%20Path :bSanitizeOutputAttributeName false :InputSource MeshOptionsPreset :OutputAttributeName None :PropertyName MeshOptionsPreset))
  (bind CopyAttributes_0 (Copy%20Attributes :InputSource MeshOptionsPreset))
  (bind MergeAttributes_1 (Merge%20Attributes))
  (bind CopyAttributes_1 (Copy%20Attributes :InputSource $Index :OutputTarget IndexBck))
  (bind AttributeFilter_0 (Filter%20Attribute%20Elements :AttributeTypes (struct :Type Boolean :BoolValue true) :bUseConstantThreshold true :Operator Equal :TargetAttribute MeshOptionsPreset :ThresholdAttribute @Last))
  (bind DeleteAttributes_0 (Delete%20Attributes :SelectedAttributes "MeshOptionsPreset"))
  (reroute Base%20Properties MergeAttributes_16)
  (connect Gather_0 Out AddAttribute_41 Attributes)
  (connect DeleteAttributes_1 Out AddAttribute_41 In)
  (connect Branch_0 Output%20B AttributeCompareOp_0 InA)
  (connect CopyAttributes_1 Out AttributeFilter_0 In)
  (connect AddAttribute_41 Out AttributeFilter_45 In)
  (connect AttributeCompareOp_0 Out AttributePartition_0 In)
  (connect MergeAttributes_28 Out AttributeRename_3 In)
  (connect MergeAttributes_16 Out Branch_0 In)
  (connect Branch_30 Output%20B Branch_1 In)
  (connect Branch_30 Output%20B Branch_5 In)
  (connect Branch_30 Output%20B Branch_7 In)
  (connect Branch_30 Output%20B Branch_9 In)
  (connect Branch_30 Output%20B Branch_18 In)
  (connect Branch_30 Output%20B Branch_21 In)
  (connect GetEntriesCount_32 Out Branch_30 bOutputToB)
  (connect FilterDataByTag_0 InsideFilter Branch_30 In)
  (connect AttributePartition_0 Out CopyAttributes_0 Source)
  (connect GetPropertyFromObjectPath_4 Out CopyAttributes_0 Target)
  (connect MergeAttributes_1 Out CopyAttributes_1 Source)
  (connect MergeAttributes_1 Out CopyAttributes_1 Target)
  (connect AttributeFilter_0 InsideFilter DeleteAttributes_0 In)
  (connect MergeAttributes_16 Out DeleteAttributes_1 In)
  (connect MakeTransformAttribute_3 Out DeleteAttributes_4 In)
  (connect DefaultInputNode In FilterDataByTag_0 In)
  (connect MergeAttributes_20 Out Gather_0 In)
  (connect DeleteAttributes_4 Out Gather_0 In2)
  (connect MatchAndSetAttributes_0 Out Gather_0 In3)
  (connect AttributeRename_3 Out Gather_0 In4)
  (connect MergeAttributes_24 Out Gather_0 In5)
  (connect Subgraph_1 Out Gather_0 In6)
  (connect FilterDataByTag_0 InsideFilter GetEntriesCount_32 In)
  (connect Branch_30 Output%20B GetPropertyFromObjectPath_0 In)
  (connect Branch_5 Output%20B GetPropertyFromObjectPath_1 In)
  (connect AttributeStringOp_1 Out GetPropertyFromObjectPath_1 PropertyName)
  (connect Branch_7 Output%20B GetPropertyFromObjectPath_2 In)
  (connect AttributeStringOp_3 Out GetPropertyFromObjectPath_2 PropertyName)
  (connect Branch_0 Output%20B GetPropertyFromObjectPath_4 In)
  (connect Branch_1 Output%20B GetPropertyFromObjectPath_6 In)
  (connect AttributeStringOp_0 Out GetPropertyFromObjectPath_6 PropertyName)
  (connect Branch_18 Output%20B GetPropertyFromObjectPath_8 In)
  (connect AttributeStringOp_2 Out GetPropertyFromObjectPath_8 PropertyName)
  (connect Branch_21 Output%20B GetPropertyFromObjectPath_10 In)
  (connect AttributeStringOp_4 Out GetPropertyFromObjectPath_10 PropertyName)
  (connect Branch_9 Output%20B GetPropertyFromObjectPath_12 In)
  (connect AttributeStringOp_5 Out GetPropertyFromObjectPath_12 PropertyName)
  (connect MergeAttributes_18 Out MakeTransformAttribute_3 Rotation)
  (connect MergeAttributes_18 Out MakeTransformAttribute_3 Scale)
  (connect MergeAttributes_18 Out MakeTransformAttribute_3 Translation)
  (connect MergeAttributes_22 Out MatchAndSetAttributes_0 In)
  (connect DeleteAttributes_0 Out MatchAndSetAttributes_0 Match%20Data)
  (connect GetPropertyFromObjectPath_12 Out MergeAttributes_0 In)
  (connect CopyAttributes_0 Out MergeAttributes_1 In)
  (connect GetPropertyFromObjectPath_0 Out MergeAttributes_16 In)
  (connect GetPropertyFromObjectPath_1 Out MergeAttributes_18 In)
  (connect GetPropertyFromObjectPath_6 Out MergeAttributes_20 In)
  (connect GetPropertyFromObjectPath_8 Out MergeAttributes_22 In)
  (connect GetPropertyFromObjectPath_10 Out MergeAttributes_24 In)
  (connect GetPropertyFromObjectPath_2 Out MergeAttributes_28 In)
  (connect Base%20Properties Out Subgraph_1 BaseAssetOptions)
  (connect MergeAttributes_0 Out Subgraph_1 RuntimeOptions)
  (output Out AttributeFilter_45.InsideFilter)
  (output Out Branch_30.Output%20A))
