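; PCG Biome Core subgraph that builds a priority-sorted list of unique generators from biome Assets and LocalAssets.
; Assets pass through a property-resolving loop, merge with LocalAssets, and are reduced to non-empty Generator entries partitioned by generator path.
; Each unique generator's properties are read from its object, while a loop merges its subtype strings into GeneratorSubTypes.
; The joined attributes are sorted by GeneratorPriority and partitioned by index, outputting one set per generator.
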(graph BiomeCore_ResolveGeneratorsFromAssets
  (inputs (Assets :type Param) (LocalAssets :type Param))
  (outputs (Generators :type Param))
  (bind Loop_0 (Loop :bTokenizeOnWhiteSpace true :Graph /PCGBiomeCore/Core/BiomeCore_ResolveAssetProperties.BiomeCore_ResolveAssetProperties))
  (bind MergeAttributes_8 (Merge%20Attributes))
  (bind DeleteAttributes_3 (Delete%20Attributes :bTokenizeOnWhiteSpace true :Operation KeepSelectedAttributes :SelectedAttributes "Generator,GeneratorSubtype"))
  (bind AttributeFilter_1 (Filter%20Attribute%20Elements :AttributeTypes (struct :Type SoftObjectPath) :bUseConstantThreshold true :bUseSpatialQuery true :Operator NotEqual :TargetAttribute Generator))
  (bind AttributePartition_4 (Attribute%20Partition :bTokenizeOnWhiteSpace true :PartitionAttributeSelectors [Generator]))
  (bind AddAttribute_3 (Add%20Attribute :bCopyAllAttributes true))
  (bind SortAttributes_0 (Sort%20Attributes :bUseStableSort false :SortAttributes [(struct :InputSource GeneratorPriority)]))
  (bind AttributePartition_5 (Attribute%20Partition :PartitionAttributeSelectors [$Index]))
  ;[0 0 0] Get Generators Properties
  (bind GetAttributeFromIndex_6 (Get%20Attribute%20Set%20from%20Index))
  (bind MergeAttributes_7 (Merge%20Attributes))
  (bind GetPropertyFromObjectPath_5 (Get%20Property%20From%20Object%20Path :bSanitizeOutputAttributeName false :InputSource Generator :OutputAttributeName None :PropertyName Generator))
  (bind MergeAttributes_1 (Merge%20Attributes))
  (bind AddAttribute_2 (Add%20Attribute :bCopyAllAttributes true :InputSource Generator :OutputTarget Generator))
  ;[0 0 0] Resolve Generators Subtypes
  (bind CreateAttribute_4 (Create%20Constant :AttributeTypes (struct :Type String :StringValue "GeneratorSubType")))
  (bind Loop_3 (Loop :Graph /PCGBiomeCore/Core/BiomeCore_MergeStringAttributeEntriesLoop.BiomeCore_MergeStringAttributeEntriesLoop))
  (bind CreateAttribute_17 (Create%20Constant :AttributeTypes (struct :Type String :StringValue "GeneratorSubTypes")))
  (bind AttributeRename_15 (Attribute%20Rename :AttributeToRename @LastCreated))
  (bind MergeAttributes_2 (Merge%20Attributes))
  (connect MergeAttributes_1 Out AddAttribute_2 Attributes)
  (connect MergeAttributes_7 Out AddAttribute_2 In)
  (connect MergeAttributes_2 Out AddAttribute_3 Attributes)
  (connect AddAttribute_2 Out AddAttribute_3 In)
  (connect DeleteAttributes_3 Out AttributeFilter_1 In)
  (connect AttributeFilter_1 InsideFilter AttributePartition_4 In)
  (connect SortAttributes_0 Out AttributePartition_5 In)
  (connect CreateAttribute_4 Out AttributeRename_15 AttributeToRename)
  (connect Loop_3 Out AttributeRename_15 In)
  (connect CreateAttribute_17 Out AttributeRename_15 NewAttributeName)
  (connect MergeAttributes_8 Out DeleteAttributes_3 In)
  (connect AttributePartition_4 Out GetAttributeFromIndex_6 In)
  (connect MergeAttributes_7 Out GetPropertyFromObjectPath_5 In)
  (connect DefaultInputNode Assets Loop_0 In)
  (connect AttributePartition_4 Out Loop_3 In)
  (connect CreateAttribute_4 Out Loop_3 StringAttribute)
  (connect GetPropertyFromObjectPath_5 Out MergeAttributes_1 In)
  (connect AttributeRename_15 Out MergeAttributes_2 In)
  (connect GetAttributeFromIndex_6 Out MergeAttributes_7 In)
  (connect DefaultInputNode LocalAssets MergeAttributes_8 In)
  (connect Loop_0 Out MergeAttributes_8 In)
  (connect AddAttribute_3 Out SortAttributes_0 In)
  (output Generators AttributePartition_5))
