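; PCG Biome Core subgraph that regroups generated points into per-DataIndex data sets and restores their generator tags.
; It drops AssetID, then marks regular, assembly and assembly-source points with an Assembly attribute.
; Root and non-root points merge separately, non-root splitting by RecursionLevel, then partition by DataIndex for a RestoreTags loop fed by prepared GeneratorsOrderedData.
; UseAssembliesSourcePointsInDifference sends source points through that path, outputting assembly points directly tagged SkipDifference; otherwise assembly points take it.
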(graph BiomeCore_RestoreDataByGenerators
  (params (UseAssembliesSourcePointsInDifference :type Bool :default true))
  (inputs (GeneratorsOrderedData :type Point))
  (bind FilterDataByTag_0 (Filter%20Data%20By%20Tag :SelectedTags "Root"))
  (bind MergePoints_2 (Merge%20Points))
  (bind MergePoints_3 (Merge%20Points))
  (bind AttributePartition_3 (Attribute%20Partition :PartitionAttributeSelectors [RecursionLevel]))
  (bind SortAttributes_2 (Sort%20Attributes :bUseStableSort false :SortAttributes [(struct :InputSource DataIndex)]))
  (bind AttributePartition_2 (Attribute%20Partition :PartitionAttributeSelectors [DataIndex]))
  (bind Subgraph_1 (Subgraph :Graph /PCGBiomeCore/Core/BiomeCore_PrepareTagsData.BiomeCore_PrepareTagsData))
  (bind Loop_1 (Loop :Graph /PCGBiomeCore/Core/BiomeCore_RestoreTags.BiomeCore_RestoreTags))
  ;[0.15 0 0 0.5] Review Change
  (bind DeleteAttributes_2 (Delete%20Attributes :SelectedAttributes "AssetID"))
  (bind FilterDataByTag_1 (Filter%20Data%20By%20Tag :Operator Substring :SelectedTags "Assembly"))
  (bind AddAttribute_0 (Add%20Attribute :AttributeTypes (struct :Type Integer32) :OutputTarget Assembly)) ; Regular Pts
  (bind FilterDataByTag_7 (Filter%20Data%20By%20Tag :Operation RemoveTagged :Operator Substring :SelectedTags "Assembly:Source"))
  (bind AddAttribute_2 (Add%20Attribute :AttributeTypes (struct :Type Integer32 :Int32Value 2 :BoolValue true) :OutputTarget Assembly)) ; Assmbly Pts
  (bind AddAttribute_4 (Add%20Attribute :AttributeTypes (struct :Type Integer32 :Int32Value 1 :BoolValue true) :OutputTarget Assembly)) ; Assmbly Src Points
  (bind Select_1 (Select :bUseInputB-param UseAssembliesSourcePointsInDifference))
  (bind Branch_1 (Branch :bOutputToB-param UseAssembliesSourcePointsInDifference))
  (bind AddTags_1 (Add%20Tags :TagsToAdd "SkipDifference"))
  (connect FilterDataByTag_1 OutsideFilter AddAttribute_0 In)
  (connect FilterDataByTag_7 InsideFilter AddAttribute_2 In)
  (connect FilterDataByTag_7 OutsideFilter AddAttribute_4 In)
  (connect Branch_1 Output%20B AddTags_1 In)
  (connect SortAttributes_2 Out AttributePartition_2 In)
  (connect MergePoints_3 Out AttributePartition_3 In)
  (connect AddAttribute_2 Out Branch_1 In)
  (connect DefaultInputNode In DeleteAttributes_2 In)
  (connect AddAttribute_0 Out FilterDataByTag_0 In)
  (connect Select_1 Out FilterDataByTag_0 In)
  (connect DeleteAttributes_2 Out FilterDataByTag_1 In)
  (connect FilterDataByTag_1 InsideFilter FilterDataByTag_7 In)
  (connect AttributePartition_2 Out Loop_1 In)
  (connect Subgraph_1 Out Loop_1 Tags)
  (connect FilterDataByTag_0 InsideFilter MergePoints_2 In)
  (connect FilterDataByTag_0 OutsideFilter MergePoints_3 In)
  (connect AddAttribute_2 Out Select_1 Input%20A)
  (connect AddAttribute_4 Out Select_1 Input%20B)
  (connect AttributePartition_3 Out SortAttributes_2 In)
  (connect MergePoints_2 Out SortAttributes_2 In)
  (connect DefaultInputNode GeneratorsOrderedData Subgraph_1 In)
  (output Out AddTags_1)
  (output Out Loop_1))
