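; PCG graph in Biome Core that restores saved tags on incoming data, looked up from a Tags param by each data's DataIndex.
; It strips all tags except Root and Child, then adds the looked-up tags.
; Child data gains a RecursionLevel tag from its attribute, while other data gets RecursionLevel:-1.
; Final cleanup deletes DataIndex, but misspells RecursionLevel as "RecrusionLevel", so that attribute is not matched.
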(graph BiomeCore_RestoreTags
  (inputs (Tags :type Param))
  (bind DeleteTags_0 (Delete%20Tags :Operation KeepOnlySelectedTags :SelectedTags "Root, Child"))
  (bind GetAttributeFromPointIndex_1 (Get%20Attribute%20From%20Point%20Index :bForceOutputAttributeToBeInElementsDomain true :InputSource DataIndex :OutputAttributeName None))
  (bind FilterDataByIndex_2 (Filter%20Data%20By%20Index))
  (bind AddTags_0 (Add%20Tags))
  (bind FilterDataByTag_0 (Filter%20Data%20By%20Tag :SelectedTags "Child"))
  (bind GetAttributeFromPointIndex_0 (Get%20Attribute%20From%20Point%20Index :bForceOutputAttributeToBeInElementsDomain true :InputSource RecursionLevel :OutputAttributeName None))
  (bind AddTags_1 (Add%20Tags :Prefix "RecursionLevel:"))
  (bind AddTags_2 (Add%20Tags :TagsToAdd "RecursionLevel:-1"))
  (bind DeleteAttributes_3 (Delete%20Attributes :SelectedAttributes "DataIndex, RecrusionLevel"))
  (connect DeleteTags_0 Out AddTags_0 In)
  (connect FilterDataByIndex_2 InsideFilter AddTags_0 TagsToAdd)
  (connect FilterDataByTag_0 InsideFilter AddTags_1 In)
  (connect GetAttributeFromPointIndex_0 Attribute AddTags_1 TagsToAdd)
  (connect FilterDataByTag_0 OutsideFilter AddTags_2 In)
  (connect AddTags_1 Out DeleteAttributes_3 In)
  (connect AddTags_2 Out DeleteAttributes_3 In)
  (connect DefaultInputNode In DeleteTags_0 In)
  (connect DefaultInputNode Tags FilterDataByIndex_2 In)
  (connect GetAttributeFromPointIndex_1 Attribute FilterDataByIndex_2 SelectedIndices)
  (connect AddTags_0 Out FilterDataByTag_0 In)
  (connect FilterDataByTag_0 InsideFilter GetAttributeFromPointIndex_0 In)
  (connect DefaultInputNode In GetAttributeFromPointIndex_1 In)
  (output Out DeleteAttributes_3))
