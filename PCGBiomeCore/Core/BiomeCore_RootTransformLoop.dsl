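; Biome Core loop body that transforms root points for one AssetID, spawns their child assets and outputs tagged points plus runtime asset tables.
; The matching AssetTable row drives transform, scale/bounds and self-pruning, then supplies mesh, actor and assembly attributes copied onto roots.
; ChildTransformLoop recurses up to MaxIterations with input limits derived from point count.
; Children are culled against overlapping parents and by recursion level; ShowBounds enables debug display.
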(graph BiomeCore_RootTransformLoop
  (params (TransfomInputPointsRateMultiplier :type Int :default 1500) (MaxIterations :type Int :default 4) (Filters :type (Object /Script/PCG.PCGGraphInterface)))
  (inputs (InPoints :type Point :advanced) (AssetTable :type Param))
  (outputs (RuntimeAssetTables :type Param))
  (bind ToPoint_1 (To%20Point))
  (bind AddTags_1 (Add%20Tags :TagsToAdd "Root, RecursionLevel:-1"))
  (bind DeleteAttributes_5 (Delete%20Attributes :bTokenizeOnWhiteSpace true :Operation KeepSelectedAttributes :SelectedAttributes "Assembly"))
  (bind AttributeCompareOp_1 (Attribute%20Compare%20Op))
  (bind Branch_2 (Branch))
  (bind AddTags_2 (Add%20Tags :TagsToAdd "Assembly"))
  (bind DeleteAttributes_7 (Delete%20Attributes :bTokenizeOnWhiteSpace true :Operation KeepSelectedAttributes :SelectedAttributes "Mesh"))
  (bind AttributeCompareOp_0 (Attribute%20Compare%20Op :InputSource1 Mesh))
  (bind DeleteAttributes_8 (Delete%20Attributes :bTokenizeOnWhiteSpace true :Operation KeepSelectedAttributes :SelectedAttributes "Actor"))
  (bind AttributeCompareOp_2 (Attribute%20Compare%20Op :InputSource1 Actor))
  (bind AttributeBooleanOp_0 (Attribute%20Boolean%20Op))
  (bind Branch_1 (Branch))
  (bind Gather_1 (Gather))
  ;[0 0 0 1 32] Children
  (bind Branch_0 (Branch :bOutputToB-param MaxIterations))
  (bind DeleteAttributes_2 (Delete%20Attributes :Operation KeepSelectedAttributes :SelectedAttributes "AssetID"))
  (bind Subgraph_0 (Subgraph :Graph /PCGBiomeCore/Core/BiomeCore_PrepareChildRecursion.BiomeCore_PrepareChildRecursion))
  (bind Loop_19 (Loop :bTokenizeOnWhiteSpace true :Graph /PCGBiomeCore/Core/BiomeCore_ChildTransformLoop.BiomeCore_ChildTransformLoop :Filters Filters :MaxIterations MaxIterations))
  ;;[0 0 0] Limit Child Inputs
  (bind GetPointsCount_8 (Get%20Element%20Count))
  (bind AttributeMathsOp_3 (Attribute%20Maths%20Op :Operation Multiply :InB TransfomInputPointsRateMultiplier))
  (bind CreateAttribute_6 (Create%20Constant :AttributeTypes (struct :Type Integer32 :Int32Value 1500)))
  (bind AttributeMathsOp_4 (Attribute%20Maths%20Op :Operation Min))
  (bind CreateAttribute_7 (Create%20Constant :AttributeTypes (struct :Type Integer32 :Int32Value 5)))
  (bind AttributeMathsOp_5 (Attribute%20Maths%20Op :Operation Max))
  ;[0 0 0 1 32] Root Transform
  (bind DeleteAttributes_4 (Delete%20Attributes :bTokenizeOnWhiteSpace true :Operation KeepSelectedAttributes :SelectedAttributes "ForceAssetScale"))
  (bind Subgraph_1 (Subgraph :Graph /PCGBiomeCore/Core/BiomeCore_RunTransformGraph.BiomeCore_RunTransformGraph))
  (bind Subgraph_7 (Subgraph :Graph /PCGBiomeCore/Core/BiomeCore_ApplyScaleAndBounds.BiomeCore_ApplyScaleAndBounds))
  (bind Subgraph_13 (Subgraph :Graph /PCGBiomeCore/Core/BiomeCore_SelfPruningGraph.BiomeCore_SelfPruningGraph))
  ;[0 0 0 1 32] Asset
  (bind GetAttributeFromPointIndex_38 (Get%20Attribute%20From%20Point%20Index :bForceOutputAttributeToBeInElementsDomain true :InputSource AssetID :OutputAttributeName AssetID))
  (bind AttributeFilter_39 (Filter%20Attribute%20Elements :bUseSpatialQuery true :Operator Equal :TargetAttribute AssetID :ThresholdAttribute AssetID))
  ;[0 0 0] Add Mesh, Material, Actor attributes to points
  (bind DeleteAttributes_1 (Delete%20Attributes :bTokenizeOnWhiteSpace true :Operation KeepSelectedAttributes :SelectedAttributes "Mesh,SkeletalMesh,Material,Assembly,Actor,AllowCollision,AssemblyAllowCollision,Visible,CastShadow,CastHiddenShadow,CastDynamicShadow,AffectDistanceFieldLighting,DetailMode,StartCullDistance,EndCullDistance,EvaluateWorldPositionOffset,WorldPositionOffsetDisableDistance,IncludeInHLOD,HLODLayer,RuntimeAssemblyHashValue,InfluenceType,InfluenceRadius,AffectDynamicIndirectLighting,VisibleInRayTracing,CanEverAffectNavigation,RasterizeAsFilledConvexVolume,IsSpatiallyLoaded,ComponentTags"))
  (bind CopyAttribute_0 (Copy%20Attributes :bCopyAllAttributes true :InputSource $Density :OutputTarget None))
  ;[0 0 0] Diff selected parents from children
  (bind Subgraph_2 (Subgraph :Graph /PCGBiomeCore/Core/BiomeCore_OverlapWithChildrenGraph.BiomeCore_OverlapWithChildrenGraph))
  (bind Difference_0 (Difference :bDiffMetadata false :DensityFunction Binary))
  ;[0 0 0] Diff all children by recursion level
  (bind Loop_7 (Loop :bTokenizeOnWhiteSpace true :Graph /PCGBiomeCore/Core/BiomeCore_DifferenceByRecursionLevel.BiomeCore_DifferenceByRecursionLevel))
  (bind DeleteAttributes_9 (Delete%20Attributes :bTokenizeOnWhiteSpace true :SelectedAttributes "AssetID"))
  ;[1 0.3333 0 0.1] Show Bounds
  (bind DeleteAttributes_3 (Delete%20Attributes :bTokenizeOnWhiteSpace true :Operation KeepSelectedAttributes :SelectedAttributes "ShowBounds"))
  (bind Select_4 (Select))
  (bind Debug_5 (Debug))
  ;[0 0 0] Runtime Asset Table
  (bind DeleteAttributes_0 (Delete%20Attributes :Operation KeepSelectedAttributes :SelectedAttributes "RuntimeAssembly,RuntimeAssemblyHashValue"))
  (bind Gather_0 (Gather))
  (reroute Asset AttributeFilter_39.InsideFilter)
  (reroute NamedRerouteDeclaration_1 AssetTable)
  (reroute NamedRerouteDeclaration_2 InPoints)
  (connect CopyAttribute_0 Out AddTags_1 In)
  (connect Branch_2 Output%20A AddTags_2 In)
  (connect AttributeCompareOp_0 Out AttributeBooleanOp_0 InA)
  (connect AttributeCompareOp_2 Out AttributeBooleanOp_0 InB)
  (connect DeleteAttributes_7 Out AttributeCompareOp_0 InA)
  (connect DeleteAttributes_5 Out AttributeCompareOp_1 InA)
  (connect DeleteAttributes_8 Out AttributeCompareOp_2 InA)
  (connect GetAttributeFromPointIndex_38 Attribute AttributeFilter_39 Filter)
  (connect NamedRerouteDeclaration_1 Out AttributeFilter_39 In)
  (connect GetPointsCount_8 Out AttributeMathsOp_3 InA)
  (connect AttributeMathsOp_3 Out AttributeMathsOp_4 InA)
  (connect CreateAttribute_6 Out AttributeMathsOp_4 InB)
  (connect AttributeMathsOp_4 Out AttributeMathsOp_5 InA)
  (connect CreateAttribute_7 Out AttributeMathsOp_5 InB)
  (connect Subgraph_13 Out Branch_0 In)
  (connect AttributeBooleanOp_0 Out Branch_1 bOutputToB)
  (connect Branch_2 Output%20A Branch_1 In)
  (connect AttributeCompareOp_1 Out Branch_2 bOutputToB)
  (connect AddTags_1 Out Branch_2 In)
  (connect DeleteAttributes_1 Out CopyAttribute_0 Source)
  (connect Subgraph_13 Out CopyAttribute_0 Target)
  (connect Select_4 Out Debug_5 In)
  (connect Asset Out DeleteAttributes_0 In)
  (connect Asset Out DeleteAttributes_1 In)
  (connect Asset Out DeleteAttributes_2 In)
  (connect Asset Out DeleteAttributes_3 In)
  (connect Asset Out DeleteAttributes_4 In)
  (connect DeleteAttributes_1 Out DeleteAttributes_5 In)
  (connect DeleteAttributes_1 Out DeleteAttributes_7 In)
  (connect DeleteAttributes_1 Out DeleteAttributes_8 In)
  (connect Loop_7 RemainingChilds DeleteAttributes_9 In)
  (connect Subgraph_2 Out Difference_0 Differences)
  (connect Loop_19 Out Difference_0 Source)
  (connect DeleteAttributes_0 Out Gather_0 In)
  (connect Loop_19 RuntimeAssetTable Gather_0 In)
  (connect AddTags_2 Out Gather_1 In)
  (connect Branch_1 Output%20A Gather_1 In)
  (connect Branch_2 Output%20B Gather_1 In2)
  (connect NamedRerouteDeclaration_2 Out GetAttributeFromPointIndex_38 In)
  (connect NamedRerouteDeclaration_2 Out GetPointsCount_8 In)
  (connect ToPoint_1 Out Loop_7 ChildsData)
  (connect ToPoint_1 Out Loop_7 In)
  (connect Subgraph_0 Out Loop_19 ChildAssetsLoop)
  (connect Branch_0 Output%20B Loop_19 In)
  (connect AttributeMathsOp_5 Out Loop_19 MaxInputPoints)
  (connect DeleteAttributes_3 Out Select_4 bUseInputB)
  (connect Subgraph_13 Out Select_4 Input%20B)
  (connect DeleteAttributes_2 Out Subgraph_0 AssetID)
  (connect NamedRerouteDeclaration_1 Out Subgraph_0 AssetTable)
  (connect Asset Out Subgraph_1 Asset)
  (connect DefaultInputNode InPoints Subgraph_1 In)
  (connect Asset Out Subgraph_2 Asset)
  (connect Subgraph_13 Out Subgraph_2 In)
  (connect Asset Out Subgraph_7 Asset)
  (connect Subgraph_1 Out Subgraph_7 In)
  (connect DeleteAttributes_4 Out Subgraph_7 SetScale)
  (connect Asset Out Subgraph_13 Asset)
  (connect Subgraph_7 Out Subgraph_13 In)
  (connect Difference_0 Out ToPoint_1 In)
  (output Out DeleteAttributes_9)
  (output Out Gather_1)
  (output RuntimeAssetTables Gather_0))
