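; Biome Core PCG loop-body graph that applies one filter definition from the Filters params to incoming points.
; It runs the FilterGraph as a dynamic subgraph override, then range-filters data carrying the named filter attribute between named min and max attributes.
; Data lacking that attribute passes through unfiltered.
; All three named attributes are then stripped, and only point data is output.
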(graph BiomeCore_RunFiltersLoop
  (inputs (Filters :type Param) (Points :type Point :advanced))
  (outputs (Points :type Point :advanced))
  (bind DeleteAttributes_6 (Delete%20Attributes :bTokenizeOnWhiteSpace true :Operation KeepSelectedAttributes :SelectedAttributes "FilterAttribute"))
  (bind DeleteAttributes_8 (Delete%20Attributes :bTokenizeOnWhiteSpace true :Operation KeepSelectedAttributes :SelectedAttributes "FilterRangeMinAttribute"))
  (bind DeleteAttributes_10 (Delete%20Attributes :bTokenizeOnWhiteSpace true :Operation KeepSelectedAttributes :SelectedAttributes "FilterRangeMaxAttribute"))
  (bind DeleteAttributes_4 (Delete%20Attributes :bTokenizeOnWhiteSpace true :Operation KeepSelectedAttributes :SelectedAttributes "FilterGraph"))
  (bind Subgraph_1 (Subgraph))
  (bind FilterDataByAttribute_3 (Filter%20Data%20By%20Attribute))
  (bind AttributeFilterRange_0 (Filter%20Attribute%20Elements%20by%20Range))
  (bind DeleteAttributes_2 (Delete%20Attributes :bTokenizeOnWhiteSpace true))
  (bind DeleteAttributes_3 (Delete%20Attributes :bTokenizeOnWhiteSpace true))
  (bind DeleteAttributes_5 (Delete%20Attributes :bTokenizeOnWhiteSpace true))
  (bind FilterDataByType_6 (Filter%20Data%20By%20Type :TargetType (struct :Ids [(struct :Struct /Script/PCG.PCGDataTypeInfoPoint)])))
  (connect FilterDataByAttribute_3 InsideFilter AttributeFilterRange_0 In)
  (connect DeleteAttributes_10 Out AttributeFilterRange_0 MaxThreshold/ThresholdAttribute)
  (connect DeleteAttributes_8 Out AttributeFilterRange_0 MinThreshold/ThresholdAttribute)
  (connect DeleteAttributes_6 Out AttributeFilterRange_0 TargetAttribute)
  (connect AttributeFilterRange_0 InsideFilter DeleteAttributes_2 In)
  (connect FilterDataByAttribute_3 OutsideFilter DeleteAttributes_2 In)
  (connect DeleteAttributes_6 Out DeleteAttributes_2 SelectedAttributes)
  (connect DeleteAttributes_2 Out DeleteAttributes_3 In)
  (connect DeleteAttributes_8 Out DeleteAttributes_3 SelectedAttributes)
  (connect DefaultInputNode Filters DeleteAttributes_4 In)
  (connect DeleteAttributes_3 Out DeleteAttributes_5 In)
  (connect DeleteAttributes_10 Out DeleteAttributes_5 SelectedAttributes)
  (connect DefaultInputNode Filters DeleteAttributes_6 In)
  (connect DefaultInputNode Filters DeleteAttributes_8 In)
  (connect DefaultInputNode Filters DeleteAttributes_10 In)
  (connect DeleteAttributes_6 Out FilterDataByAttribute_3 Attribute)
  (connect Subgraph_1 Out FilterDataByAttribute_3 In)
  (connect DeleteAttributes_5 Out FilterDataByType_6 In)
  (connect DefaultInputNode Points Subgraph_1 In)
  (connect DeleteAttributes_4 Out Subgraph_1 SubgraphOverride)
  (output Points FilterDataByType_6.InsideFilter))
