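; PCG Biome Core subgraph that runs a biome generator's dynamic subgraph and outputs its points tagged for that biome.
; The generator samples LocalBiomeCache bounds expanded by BiomeBlendingRange, then unused subtypes are culled.
; An optional blending pass thins points by normalized distance to the cache bounds, shaped by density power and noise.
; Points receive generator attributes, biome color, type, priority, group and biome tags, plus SkipDifference when overlap is allowed.
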(graph BiomeCore_RunGenerators
  (params (BiomeBlendingNoise :type Double :default 1.0) (BiomeBlendingDensity :type Double :default 1.0))
  (inputs (Generators :type Param) (LocalBiomeCache :type Point) (BiomeBlendingRange :type Param))
  (outputs (SampledPoints :type Point))
  (bind FilterByType_5 (Filter%20Data%20By%20Type :TargetType (struct :Ids [(struct :Struct /Script/PCG.PCGDataTypeInfoPoint)])))
  (bind AttributeCompareOp_0 (Attribute%20Compare%20Op))
  (bind DeleteAttributes_8 (Delete%20Attributes :bTokenizeOnWhiteSpace true :Operation KeepSelectedAttributes :SelectedAttributes "GeneratorSubTypes"))
  (bind Subgraph_14 (Subgraph :Graph /PCGBiomeCore/Core/BiomeCore_CullUnusedSubTypes.BiomeCore_CullUnusedSubTypes))
  (bind Branch_0 (Branch))
  (bind AddAttribute_1 (Add%20Attribute :InputSource Generator :OutputTarget Generator))
  (bind FilterDataByAttribute_2 (Filter%20Data%20By%20Attribute :Attribute GeneratorSubType))
  (bind AddAttribute_3 (Add%20Attribute :AttributeTypes (struct :Type Name) :OutputTarget GeneratorSubType))
  (bind AddTags_6 (Add%20Tags :bTokenizeOnWhiteSpace true))
  (bind AddTags_1 (Add%20Tags :bTokenizeOnWhiteSpace true))
  (bind AddTags_5 (Add%20Tags :Prefix "GROUP:"))
  ;[0 0 0 0.5] Get Generator subgraph
  (bind FilterAttribute_3 (Delete%20Attributes :bTokenizeOnWhiteSpace true :Operation KeepSelectedAttributes :SelectedAttributes "GeneratorGraph"))
  ;[0 0 0 1 32] Generator Dynamic SG
  (bind Subgraph_2 (Subgraph))
  ;[0 0 0] Biome Blending Points Filtering
  (bind AttributeMathsOp_8 (Attribute%20Maths%20Op :Operation Max :InA BiomeBlendingNoise))
  (bind AttributeMathsOp_1 (Attribute%20Maths%20Op :Operation Min))
  (bind AttributeMathsOp_4 (Attribute%20Maths%20Op :Operation Max))
  (bind Distance_1 (Distance :bCheckSourceAgainstRespectiveTarget true :OutputAttribute BiomeBlendDistance :SourceShape Center :TargetShape BoxBounds))
  (bind AttributeMathsOp_2 (Attribute%20Maths%20Op :InputSource1 BiomeBlendDistance :Operation Divide))
  (bind AttributeMathsOp_6 (Attribute%20Maths%20Op :Operation Max :InA BiomeBlendingDensity))
  (bind AttributeMathsOp_5 (Attribute%20Maths%20Op :InputSource1 BiomeBlendDistance :Operation Pow))
  (bind AttributeNoise_1 (Attribute%20Noise :InputSource BiomeBlendDistance :Mode Add :OutputTarget BiomeBlendDistance))
  (bind AttributeFilterRange_1 (Filter%20Attribute%20Elements%20by%20Range :MaxThreshold (struct :bInclusive false :bUseConstantThreshold true :ThresholdAttribute @Last :bUseSpatialQuery false :AttributeTypes (struct :DoubleValue 1.0)) :MinThreshold (struct :bUseConstantThreshold true :ThresholdAttribute @Last :bUseSpatialQuery false) :TargetAttribute BiomeBlendDistance))
  (bind DeleteAttributes_3 (Delete%20Attributes :SelectedAttributes "BiomeBlendDistance"))
  ;[0 0 0] Prepare Data Tags
  (bind DeleteAttributes_5 (Delete%20Attributes :bTokenizeOnWhiteSpace true :Operation KeepSelectedAttributes :SelectedAttributes "GeneratorType"))
  (bind AttributeStringOp_3 (Attribute%20String%20Op :InputSource2 GeneratorType :OutputTarget GeneratorType))
  (bind DeleteAttributes_6 (Delete%20Attributes :bTokenizeOnWhiteSpace true :Operation KeepSelectedAttributes :SelectedAttributes "GeneratorPriority"))
  (bind AttributeStringOp_4 (Attribute%20String%20Op :InputSource2 GeneratorPriority :OutputTarget GeneratorPriority))
  (bind DeleteAttributes_7 (Delete%20Attributes :bTokenizeOnWhiteSpace true :Operation KeepSelectedAttributes :SelectedAttributes "GeneratorGroups"))
  ;[0 0 0] Bounding Shape adjusted with Blending Range
  (bind AttributeMathsOp_0 (Attribute%20Maths%20Op :InputSource1 $Extents))
  (bind AddTags_3 (Add%20Tags :bTokenizeOnWhiteSpace true :TagsToAdd "BoundingShape"))
  ;[0 0 0 0.5 32] Apply Biome from the Local Biome Cache - Bounding shape
  (bind GetAttributeFromPointIndex_0 (Get%20Attribute%20From%20Point%20Index :bForceOutputAttributeToBeInElementsDomain true :InputSource BiomeName))
  (bind AttributeMathsOp_3 (Attribute%20Maths%20Op :InputSource1 $Color :InputSource2 BiomeColor :Operation Set))
  (bind CreateAttribute_0 (Create%20Constant :AttributeTypes (struct :Type String :StringValue "BIOME:")))
  (bind AttributeStringOp_1 (Attribute%20String%20Op :InputSource2 BiomeName :OutputTarget BiomeName))
  (bind AddTags_0 (Add%20Tags :bTokenizeOnWhiteSpace true))
  (bind CreateAttribute_1 (Create%20Constant :AttributeTypes (struct :Type String :StringValue "BIOMEPRIO:")))
  (bind GetAttributeFromPointIndex_1 (Get%20Attribute%20From%20Point%20Index :bForceOutputAttributeToBeInElementsDomain true :InputSource BiomePriority))
  (bind AttributeStringOp_0 (Attribute%20String%20Op :InputSource2 BiomePriority :OutputTarget BiomePriority))
  (bind AddTags_2 (Add%20Tags :bTokenizeOnWhiteSpace true))
  ;[0 0 0] Apply Skip Difference Tag
  (bind DeleteAttributes_4 (Delete%20Attributes :bTokenizeOnWhiteSpace true :Operation KeepSelectedAttributes :SelectedAttributes "GeneratorAllowOverlap"))
  (bind Branch_1 (Branch))
  (bind AddTags_4 (Add%20Tags :bTokenizeOnWhiteSpace true :TagsToAdd "SkipDifference"))
  (reroute BoundingShape AddTags_3)
  (reroute NamedRerouteDeclaration_0 Generators)
  (reroute NamedRerouteDeclaration_3 BiomeBlendingRange)
  (reroute SourceBounds LocalBiomeCache)
  (connect NamedRerouteDeclaration_0 Out AddAttribute_1 Attributes)
  (connect Branch_0 Output%20B AddAttribute_1 In)
  (connect DeleteAttributes_3 Out AddAttribute_1 In)
  (connect FilterDataByAttribute_2 OutsideFilter AddAttribute_3 In)
  (connect AttributeMathsOp_3 Out AddTags_0 In)
  (connect AttributeStringOp_1 Out AddTags_0 TagsToAdd)
  (connect AddTags_6 Out AddTags_1 In)
  (connect AttributeStringOp_4 Out AddTags_1 TagsToAdd)
  (connect AddTags_0 Out AddTags_2 In)
  (connect AttributeStringOp_0 Out AddTags_2 TagsToAdd)
  (connect AttributeMathsOp_0 Out AddTags_3 In)
  (connect Branch_1 Output%20B AddTags_4 In)
  (connect AddTags_1 Out AddTags_5 In)
  (connect DeleteAttributes_7 Out AddTags_5 TagsToAdd)
  (connect AddAttribute_3 Out AddTags_6 In)
  (connect FilterDataByAttribute_2 InsideFilter AddTags_6 In)
  (connect AttributeStringOp_3 Out AddTags_6 TagsToAdd)
  (connect NamedRerouteDeclaration_3 Out AttributeCompareOp_0 InA)
  (connect AttributeNoise_1 Out AttributeFilterRange_1 In)
  (connect DefaultInputNode LocalBiomeCache AttributeMathsOp_0 InA)
  (connect DefaultInputNode BiomeBlendingRange AttributeMathsOp_0 InB)
  (connect AttributeMathsOp_8 Out AttributeMathsOp_1 InA)
  (connect Distance_1 Out AttributeMathsOp_2 InA)
  (connect AttributeMathsOp_4 Out AttributeMathsOp_2 InB)
  (connect AddTags_5 Out AttributeMathsOp_3 InA)
  (connect GetAttributeFromPointIndex_0 Point AttributeMathsOp_3 InB)
  (connect NamedRerouteDeclaration_3 Out AttributeMathsOp_4 InA)
  (connect AttributeMathsOp_2 Out AttributeMathsOp_5 InA)
  (connect AttributeMathsOp_6 Out AttributeMathsOp_5 InB)
  (connect AttributeMathsOp_5 Out AttributeNoise_1 In)
  (connect AttributeMathsOp_1 Out AttributeNoise_1 NoiseMax)
  (connect CreateAttribute_1 Out AttributeStringOp_0 InA)
  (connect GetAttributeFromPointIndex_1 Attribute AttributeStringOp_0 InB)
  (connect CreateAttribute_0 Out AttributeStringOp_1 InA)
  (connect GetAttributeFromPointIndex_0 Attribute AttributeStringOp_1 InB)
  (connect DeleteAttributes_5 Out AttributeStringOp_3 InB)
  (connect DeleteAttributes_6 Out AttributeStringOp_4 InB)
  (connect AttributeCompareOp_0 Out Branch_0 bOutputToB)
  (connect Subgraph_14 Out Branch_0 In)
  (connect DeleteAttributes_4 Out Branch_1 bOutputToB)
  (connect AddTags_2 Out Branch_1 In)
  (connect AttributeFilterRange_1 InsideFilter DeleteAttributes_3 In)
  (connect NamedRerouteDeclaration_0 Out DeleteAttributes_4 In)
  (connect NamedRerouteDeclaration_0 Out DeleteAttributes_5 In)
  (connect NamedRerouteDeclaration_0 Out DeleteAttributes_6 In)
  (connect NamedRerouteDeclaration_0 Out DeleteAttributes_7 In)
  (connect NamedRerouteDeclaration_0 Out DeleteAttributes_8 In)
  (connect NamedRerouteDeclaration_3 Out Distance_1 MaximumDistance)
  (connect Branch_0 Output%20A Distance_1 Source)
  (connect SourceBounds Out Distance_1 Target)
  (connect NamedRerouteDeclaration_0 Out FilterAttribute_3 In)
  (connect Subgraph_2 Out FilterByType_5 In)
  (connect AddAttribute_1 Out FilterDataByAttribute_2 In)
  (connect BoundingShape Out GetAttributeFromPointIndex_0 In)
  (connect GetAttributeFromPointIndex_0 Point GetAttributeFromPointIndex_1 In)
  (connect BoundingShape Out Subgraph_2 In)
  (connect FilterAttribute_3 Out Subgraph_2 SubgraphOverride)
  (connect DeleteAttributes_8 Out Subgraph_14 AllowedSubtypes)
  (connect FilterByType_5 InsideFilter Subgraph_14 In)
  (output SampledPoints AddTags_4)
  (output SampledPoints Branch_1.Output%20A))
