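; PCG subgraph in Biome Core that runs the transform graph named by the Asset input's TransformGraph attribute on incoming points.
; It first caps point count: MaxInputPoints, effectively unlimited by default, is divided by the element count of the first input data and clamped to form a Select Points ratio.
; The selected points feed a dynamically overridden Subgraph node, whose result is converted to points for output.
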(graph BiomeCore_RunTransformGraph
  (params (MaxInputPoints :type Int :default 2147483647))
  (inputs (Asset :type Param))
  (bind DeleteAttributes_11 (Delete%20Attributes :bTokenizeOnWhiteSpace true :Operation KeepSelectedAttributes :SelectedAttributes "TransformGraph"))
  (bind Subgraph_2 (Subgraph))
  (bind ToPoint_0 (To%20Point))
  ;[0 0 0] Limit input points count
  (bind FilterDataByIndex_3 (Filter%20Data%20By%20Index :SelectedIndices "0"))
  (bind GetPointsCount_0 (Get%20Element%20Count))
  (bind AttributeMathsOp_1 (Attribute%20Maths%20Op :bForceOpToDouble true :Operation Divide :InA MaxInputPoints))
  (bind AttributeMathsOp_4 (Attribute%20Maths%20Op :Operation Clamp))
  (bind SelectPoints_3 (Select%20Points :Ratio 0.5))
  (connect GetPointsCount_0 Out AttributeMathsOp_1 InB)
  (connect AttributeMathsOp_1 Out AttributeMathsOp_4 In)
  (connect DefaultInputNode Asset DeleteAttributes_11 In)
  (connect DefaultInputNode In FilterDataByIndex_3 In)
  (connect FilterDataByIndex_3 InsideFilter GetPointsCount_0 In)
  (connect DefaultInputNode In SelectPoints_3 In)
  (connect AttributeMathsOp_4 Out SelectPoints_3 Ratio)
  (connect SelectPoints_3 Out Subgraph_2 In)
  (connect DeleteAttributes_11 Out Subgraph_2 SubgraphOverride)
  (connect Subgraph_2 Out ToPoint_0 In)
  (output Out ToPoint_0))
