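; PCG subgraph in Biome Core that optionally self-prunes points according to settings on the incoming Asset parameter.
; The SelfPrune attribute drives a Branch that either passes points straight to the output or sends them to pruning.
; The pruning path multiplies extents by SelfPruningExtentsMultiplier, runs Self Pruning with all-equal priority, then multiplies extents by a Divide result of that multiplier, apparently to restore original size.
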(graph BiomeCore_SelfPruningGraph
  (inputs (Asset :type Param))
  (bind DeleteAttributes_0 (Delete%20Attributes :bTokenizeOnWhiteSpace true :Operation KeepSelectedAttributes :SelectedAttributes "SelfPrune"))
  (bind Branch_4 (Branch))
  (bind DeleteAttributes_1 (Delete%20Attributes :bTokenizeOnWhiteSpace true :Operation KeepSelectedAttributes :SelectedAttributes "SelfPruningExtentsMultiplier"))
  (bind ExtentsModifier_2 (Extents%20Modifier :Mode Multiply))
  (bind SelfPruning_3 (Self%20Pruning :Parameters (struct :PruningType AllEqual :ComparisonSource $Extents)))
  (bind AttributeMathsOp_4 (Attribute%20Maths%20Op :Operation Divide))
  (bind ExtentsModifier_6 (Extents%20Modifier :Mode Multiply))
  (connect DeleteAttributes_1 Out AttributeMathsOp_4 InB)
  (connect DeleteAttributes_0 Out Branch_4 bOutputToB)
  (connect DefaultInputNode In Branch_4 In)
  (connect DefaultInputNode Asset DeleteAttributes_0 In)
  (connect DefaultInputNode Asset DeleteAttributes_1 In)
  (connect DeleteAttributes_1 Out ExtentsModifier_2 Extents)
  (connect Branch_4 Output%20B ExtentsModifier_2 In)
  (connect AttributeMathsOp_4 Out ExtentsModifier_6 Extents)
  (connect SelfPruning_3 Out ExtentsModifier_6 In)
  (connect ExtentsModifier_2 Out SelfPruning_3 In)
  (output Out Branch_4.Output%20A)
  (output Out ExtentsModifier_6))
