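; PCG Biome Core loop subgraph that turns child weights into cumulative filter ranges, one child per iteration.
; It sets the current child's FilterLowerBound from the Feedback input's FilterUpperBound, then combines that bound with Quantity to produce FilterUpperBound.
; The bounded child returns through Feedback to seed the next range, while a copy with Quantity deleted goes to Out.
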(graph BiomeCore_SetChildFilterBoundsFromWeights
  (inputs (CurrentChild :type Param) (Feedback :type Param))
  (outputs (Feedback :type Param))
  (bind AttributeMathsOp_8 (Attribute%20Maths%20Op :InputSource1 FilterLowerBound :InputSource2 FilterUpperBound :Operation Set :OutputTarget FilterLowerBound))
  (bind AttributeMathsOp_6 (Attribute%20Maths%20Op :InputSource1 FilterLowerBound :InputSource2 Quantity :OutputTarget FilterUpperBound))
  (bind DeleteAttributes_9 (Delete%20Attributes :bTokenizeOnWhiteSpace true :SelectedAttributes "Quantity"))
  (connect AttributeMathsOp_8 Out AttributeMathsOp_6 InA)
  (connect AttributeMathsOp_8 Out AttributeMathsOp_6 InB)
  (connect DefaultInputNode CurrentChild AttributeMathsOp_8 InA)
  (connect DefaultInputNode Feedback AttributeMathsOp_8 InB)
  (connect AttributeMathsOp_6 Out DeleteAttributes_9 In)
  (output Feedback AttributeMathsOp_6)
  (output Out DeleteAttributes_9))
