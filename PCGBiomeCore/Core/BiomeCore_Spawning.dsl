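; PCG Biome Core spawning graph that turns biome points into actors, meshes and skeletal meshes.
; Points with a non-empty Actor attribute spawn actors by attribute.
; Points also split on IsSpatiallyLoaded: loaded ones feed a BiomeCore_SpawningLoop run per unique HLODLayer.
; Non-spatially-loaded ones feed a second loop with an empty HLOD layer, without the ForceCreateTargetActors override; all three branches merge into the output.
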(graph BiomeCore_Spawning
  (params (ForceCreateTargetActors :type Bool) (TargetActorsClass :type (Struct /Script/CoreUObject.SoftClassPath) :default (struct :AssetPath (struct :PackageName /Script/Engine :AssetName Actor))) (MergeDifferentDataInSameInstancedComponents :type Bool) (SkeletalMeshTransformProviderData :type (SoftObject /Script/Engine.TransformProviderData)))
  (bind RemoveEmptyData_0 (Remove%20Empty%20Data))
  (bind AttributeFilter_0 (Filter%20Attribute%20Elements :AttributeTypes (struct :Type Boolean) :bGenerateOutputDataEvenIfEmpty false :bUseConstantThreshold true :Operator Equal :TargetAttribute IsSpatiallyLoaded :ThresholdAttribute @Last))
  (bind DeleteAttributes_0 (Delete%20Attributes :Operation KeepSelectedAttributes :SelectedAttributes "HLODLayer"))
  (bind PointToAttributeSet_0 (Point%20To%20Attribute%20Set))
  (bind AttributeRemoveDuplicates_0 (Attribute%20Remove%20Duplicates :AttributeSelectors [HLODLayer]))
  (bind MergeAttributes_1 (Merge%20Attributes :MergeTagsMode None))
  (bind AttributeRemoveDuplicates_1 (Attribute%20Remove%20Duplicates :AttributeSelectors [HLODLayer]))
  ;[0 0 0 1 32] Spawning
  ;;[0 0 0] Actors
  (bind AttributeFilter_4 (Filter%20Attribute%20Elements :AttributeTypes (struct :Type SoftClassPath) :bGenerateOutputDataEvenIfEmpty false :bUseConstantThreshold true :Operator NotEqual :TargetAttribute Actor))
  (bind SpawnActor_0 (Spawn%20Actor :bSpawnByAttribute true :SpawnAttribute Actor))
  ;;[0 0 0] Meshes and Skeletal Meshes per unique HLOD Layers
  (bind AttributePartition_0 (Attribute%20Partition :PartitionAttributeSelectors [HLODLayer]))
  (bind Loop_0 (Loop :Graph /PCGBiomeCore/Core/BiomeCore_SpawningLoop.BiomeCore_SpawningLoop :ForceCreateTargetActors ForceCreateTargetActors :MergeDifferentDataInSameInstancedComponents MergeDifferentDataInSameInstancedComponents :SkeletalMeshTransformProviderData SkeletalMeshTransformProviderData :TargetActorsClass TargetActorsClass))
  ;;[0 0 0 0.5] Meshes and Skeletal Meshes not Spatially Loaded
  (bind CreateConstant_0 (Create%20Constant :AttributeTypes (struct :Type SoftObjectPath) :OutputTarget HLODLayer))
  (bind Loop_1 (Loop :Graph /PCGBiomeCore/Core/BiomeCore_SpawningLoop.BiomeCore_SpawningLoop :MergeDifferentDataInSameInstancedComponents MergeDifferentDataInSameInstancedComponents :SkeletalMeshTransformProviderData SkeletalMeshTransformProviderData :TargetActorsClass TargetActorsClass))
  (connect RemoveEmptyData_0 Out AttributeFilter_0 In)
  (connect RemoveEmptyData_0 Out AttributeFilter_4 In)
  (connect AttributeRemoveDuplicates_1 Out AttributePartition_0 In)
  (connect PointToAttributeSet_0 Out AttributeRemoveDuplicates_0 In)
  (connect MergeAttributes_1 Out AttributeRemoveDuplicates_1 In)
  (connect AttributeFilter_0 OutsideFilter DeleteAttributes_0 In)
  (connect AttributePartition_0 Out Loop_0 HLODLayers)
  (connect AttributeFilter_0 OutsideFilter Loop_0 Points)
  (connect CreateConstant_0 Out Loop_1 HLODLayers)
  (connect AttributeFilter_0 InsideFilter Loop_1 Points)
  (connect AttributeRemoveDuplicates_0 Out MergeAttributes_1 In)
  (connect DeleteAttributes_0 Out PointToAttributeSet_0 In)
  (connect DefaultInputNode In RemoveEmptyData_0 In)
  (connect AttributeFilter_4 InsideFilter SpawnActor_0 In)
  (output Out Loop_0)
  (output Out Loop_1)
  (output Out SpawnActor_0))
