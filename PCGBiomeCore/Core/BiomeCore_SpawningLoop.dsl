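; PCG Biome Core loop subgraph that spawns biome points as instanced static and skinned meshes with per-point component overrides.
; Spatially loaded points are filtered by the input HLOD layer; otherwise they are excluded from HLOD.
; A labeled target actor with that layer is created when points exist and either the layer comparison passes or creation is forced.
; Skeletal points receive a transform provider, then split by AllowCollision into no-collision or BlockAll spawners.
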(graph BiomeCore_SpawningLoop
  (params (ForceCreateTargetActors :type Bool) (TargetActorsClass :type (Struct /Script/CoreUObject.SoftClassPath) :default (struct :AssetPath (struct :PackageName /Script/Engine :AssetName Actor))) (TargetActorsAreSpatiallyLoaded :type Bool :default true) (MergeDifferentDataInSameInstancedComponents :type Bool) (SkeletalMeshTransformProviderData :type (SoftObject /Script/Engine.TransformProviderData)))
  (inputs (Points :type Point) (HLODLayers :type Param))
  (bind Branch_1 (Branch :bOutputToB-param TargetActorsAreSpatiallyLoaded))
  (bind AttributeFilter_0 (Filter%20Attribute%20Elements :bGenerateOutputDataEvenIfEmpty false :Operator Equal :TargetAttribute HLODLayer :ThresholdAttribute HLODLayer))
  (bind AttributeMathsOp_1 (Attribute%20Maths%20Op :InputSource1 IncludeInHLOD :Operation Set)) ; Set IncludeInHLOD false
  (bind AttributeFilter_3 (Filter%20Attribute%20Elements :AttributeTypes (struct :Type SoftObjectPath) :bGenerateOutputDataEvenIfEmpty false :bUseConstantThreshold true :Operator NotEqual :TargetAttribute Mesh :ThresholdAttribute $Density))
  (bind GetActorData_1 (Get%20Actor%20Data :Mode GetSinglePoint))
  (bind PointToAttributeSet_0 (Point%20To%20Attribute%20Set :PropertiesToConvert (map ($Position Pivot))))
  (bind AttributeStringOp_1 (Attribute%20String%20Op :InputSource1 ActorReference :Operation FindLast :SearchCase IgnoreCase))
  (bind AttributeMathsOp_0 (Attribute%20Maths%20Op :InputSource1 ActorReference))
  (bind AttributeStringOp_2 (Attribute%20String%20Op :InputSource1 ActorReference :InputSource2 ActorReference :Operation RightChop))
  (bind CreateConstant_0 (Create%20Constant :AttributeTypes (struct :Type String :StringValue "PCGBiomeCoreActor") :OutputTarget ActorLabel))
  (bind AttributeStringOp_3 (Attribute%20String%20Op :InputSource2 ActorReference))
  (bind AttributeCompareOp_0 (Attribute%20Compare%20Op :Operation NotEqual))
  (bind AttributeFilter_2 (Filter%20Attribute%20Elements :AttributeTypes (struct :Type SoftObjectPath) :bGenerateOutputDataEvenIfEmpty false :bUseConstantThreshold true :Operator NotEqual :TargetAttribute SkeletalMesh :ThresholdAttribute @Last))
  (bind DataNum_0 (Data%20Count))
  (bind AttributeCompareOp_1 (Attribute%20Compare%20Op :Operation Greater))
  (bind AttributeBooleanOp_0 (Attribute%20Boolean%20Op))
  (bind AttributeBooleanOp_2 (Attribute%20Boolean%20Op :InA ForceCreateTargetActors))
  (bind AttributeBooleanOp_1 (Attribute%20Boolean%20Op :Operation Or))
  (bind Branch_0 (Branch))
  (bind DeleteAttributes_0 (Delete%20Attributes :Operation KeepSelectedAttributes :SelectedAttributes "Pivot"))
  (bind CreateTargetActor_0 (Create%20Target%20Actor :AttachOptions InGeneratedFolder :HLODSettings (struct :HLODSourceType Reference) :PropertyOverrideDescriptions [(struct :InputSource TargetActorsAreSpatiallyLoaded :PropertyTarget "bIsSpatiallyLoaded")] :TemplateActorClass /Script/Engine.Actor :OverrideTemplateActorClass-param TargetActorsClass :Property%20Overrides TargetActorsAreSpatiallyLoaded))
  (bind AddAttribute_1 (Add%20Attribute :AttributeTypes (struct :Type SoftObjectPath) :OutputTarget AnimBank))
  (bind AddAttribute_2 (Add%20Attribute :AttributeTypes (struct :Type Integer32 :Int32Value -1) :OutputTarget SequenceIndex))
  (bind CopyAttributes_0 (Copy%20Attributes :InputSource SkeletalMeshTransformProviderData :OutputTarget TransformProvider :Source SkeletalMeshTransformProviderData))
  (bind AttributeFilter_1 (Filter%20Attribute%20Elements :AttributeTypes (struct :Type Boolean) :bGenerateOutputDataEvenIfEmpty false :bUseConstantThreshold true :Operator Equal :TargetAttribute AllowCollision :ThresholdAttribute @Last)) ; Branch NoColl/BlockAll
  ;[0 0 0 1 32] Spawners
  ;;[0 0 0] Meshes
  (bind StaticMeshSpawner_0 (Static%20Mesh%20Spawner :bApplyMeshBoundsToPoints false :MeshSelectorParameters (class:/Script/PCG.PCGMeshSelectorByAttribute :AttributeName Mesh :bUseAttributeMaterialOverrides true :MaterialOverrideAttributes [Material] :TemplateDescriptor (struct :bIncludeInHLOD false :bGenerateOverlapEvents true)) :MeshSelectorType /Script/PCG.PCGMeshSelectorByAttribute :StaticMeshComponentPropertyOverrides [(struct :InputSource AllowCollision :PropertyTarget "bUseDefaultCollision") (struct :InputSource CastShadow :PropertyTarget "bCastShadow") (struct :InputSource CastHiddenShadow :PropertyTarget "bCastHiddenShadow") (struct :InputSource DetailMode :PropertyTarget "DetailMode") (struct :InputSource StartCullDistance :PropertyTarget "InstanceStartCullDistance") (struct :InputSource EndCullDistance :PropertyTarget "InstanceEndCullDistance") (struct :InputSource WorldPositionOffsetDisableDistance :PropertyTarget "WorldPositionOffsetDisableDistance") (struct :InputSource IncludeInHLOD :PropertyTarget "bIncludeInHLOD") (struct :InputSource Visible :PropertyTarget "bVisible") (struct :InputSource AffectDistanceFieldLighting :PropertyTarget "bAffectDistanceFieldLighting") (struct :InputSource AffectDynamicIndirectLighting :PropertyTarget "bAffectDynamicIndirectLighting") (struct :InputSource VisibleInRayTracing :PropertyTarget "bVisibleInRayTracing") (struct :InputSource CanEverAffectNavigation :PropertyTarget "bCanEverAffectNavigation") (struct :InputSource RasterizeAsFilledConvexVolume :PropertyTarget "bRasterizeAsFilledConvexVolume") (struct :InputSource CastDynamicShadow :PropertyTarget "bCastDynamicShadow") (struct :InputSource EvaluateWorldPositionOffset :PropertyTarget "bEvaluateWorldPositionOffset") (struct :InputSource ComponentTags :PropertyTarget "AdditionalCommaSeparatedTags")] :bAllowMergeDifferentDataInSameInstancedComponents-param MergeDifferentDataInSameInstancedComponents))
  ;;[0 0 0] Skeletal Meshes
  (bind SkinnedMeshSpawner_1 (Instanced%20Skinned%20Mesh%20Spawner :MeshSelectorParameters (class:/Script/PCG.PCGSkinnedMeshSelector :bUseAttributeMaterialOverrides true :MaterialOverrideAttributes [Material] :MeshAttribute SkeletalMesh :TemplateDescriptor (struct :bRenderStatic true)) :SkinnedMeshComponentPropertyOverrides [(struct :InputSource CastShadow :PropertyTarget "bCastShadow") (struct :InputSource StartCullDistance :PropertyTarget "InstanceStartCullDistance") (struct :InputSource EndCullDistance :PropertyTarget "InstanceEndCullDistance") (struct :InputSource IncludeInHLOD :PropertyTarget "bIncludeInHLOD") (struct :InputSource AffectDistanceFieldLighting :PropertyTarget "bAffectDistanceFieldLighting") (struct :InputSource CastDynamicShadow :PropertyTarget "bCastDynamicShadow") (struct :InputSource AffectDynamicIndirectLighting :PropertyTarget "bAffectDynamicIndirectLighting") (struct :InputSource VisibleInRayTracing :PropertyTarget "bVisibleInRayTracing") (struct :InputSource CastHiddenShadow :PropertyTarget "bCastHiddenShadow") (struct :InputSource Visible :PropertyTarget "bVisible") (struct :InputSource EvaluateWorldPositionOffset :PropertyTarget "bEvaluateWorldPositionOffset") (struct :InputSource DetailMode :PropertyTarget "DetailMode") (struct :InputSource TransformProvider :PropertyTarget "TransformProvider") (struct :InputSource CanEverAffectNavigation :PropertyTarget "bCanEverAffectNavigation")]))
  (bind SkinnedMeshSpawner_0 (Instanced%20Skinned%20Mesh%20Spawner :MeshSelectorParameters (class:/Script/PCG.PCGSkinnedMeshSelector :bUseAttributeMaterialOverrides true :MaterialOverrideAttributes [Material] :MeshAttribute SkeletalMesh :TemplateDescriptor (struct :bRenderStatic true :BodyInstance (struct :CollisionEnabled QueryOnly :CollisionProfileName BlockAll :CollisionResponses (struct :ResponseArray [(struct :Channel FCC_Trace_WeaponCapsule) (struct :Channel FCC_Object_PawnAI) (struct :Channel FCC_Trace_WeaponMulti) (struct :Channel FCC_Trace_Weapon) (struct :Channel FCC_Object_Projectile) (struct :Channel FCC_Object_ProjectileIgnoreAI) (struct :Channel FCC_Object_ProjectileMesh) (struct :Channel FCC_Object_SpectatorDroneCamera) (struct :Channel FCC_Trace_WeaponPawnAndWorldMulti)])))) :SkinnedMeshComponentPropertyOverrides [(struct :InputSource CastShadow :PropertyTarget "bCastShadow") (struct :InputSource StartCullDistance :PropertyTarget "InstanceStartCullDistance") (struct :InputSource EndCullDistance :PropertyTarget "InstanceEndCullDistance") (struct :InputSource IncludeInHLOD :PropertyTarget "bIncludeInHLOD") (struct :InputSource AffectDistanceFieldLighting :PropertyTarget "bAffectDistanceFieldLighting") (struct :InputSource CastDynamicShadow :PropertyTarget "bCastDynamicShadow") (struct :InputSource AffectDynamicIndirectLighting :PropertyTarget "bAffectDynamicIndirectLighting") (struct :InputSource VisibleInRayTracing :PropertyTarget "bVisibleInRayTracing") (struct :InputSource CastHiddenShadow :PropertyTarget "bCastHiddenShadow") (struct :InputSource Visible :PropertyTarget "bVisible") (struct :InputSource EvaluateWorldPositionOffset :PropertyTarget "bEvaluateWorldPositionOffset") (struct :InputSource DetailMode :PropertyTarget "DetailMode") (struct :InputSource TransformProvider :PropertyTarget "TransformProvider") (struct :InputSource CanEverAffectNavigation :PropertyTarget "bCanEverAffectNavigation")]))
  (connect AttributeFilter_2 InsideFilter AddAttribute_1 In)
  (connect AddAttribute_1 Out AddAttribute_2 In)
  (connect AttributeCompareOp_1 Out AttributeBooleanOp_0 InA)
  (connect AttributeCompareOp_0 Out AttributeBooleanOp_0 InB)
  (connect AttributeBooleanOp_2 Out AttributeBooleanOp_1 InA)
  (connect AttributeBooleanOp_0 Out AttributeBooleanOp_1 InB)
  (connect AttributeCompareOp_1 Out AttributeBooleanOp_2 InB)
  (connect DefaultInputNode HLODLayers AttributeCompareOp_0 InA)
  (connect DataNum_0 Out AttributeCompareOp_1 InA)
  (connect DefaultInputNode HLODLayers AttributeFilter_0 Filter)
  (connect Branch_1 Output%20B AttributeFilter_0 In)
  (connect CopyAttributes_0 Out AttributeFilter_1 In)
  (connect AttributeFilter_0 InsideFilter AttributeFilter_2 In)
  (connect AttributeMathsOp_1 Out AttributeFilter_2 In)
  (connect AttributeFilter_0 InsideFilter AttributeFilter_3 In)
  (connect AttributeMathsOp_1 Out AttributeFilter_3 In)
  (connect AttributeStringOp_1 Out AttributeMathsOp_0 InA)
  (connect Branch_1 Output%20A AttributeMathsOp_1 InA)
  (connect PointToAttributeSet_0 Out AttributeStringOp_1 InA)
  (connect AttributeMathsOp_0 Out AttributeStringOp_2 Count)
  (connect PointToAttributeSet_0 Out AttributeStringOp_2 In)
  (connect CreateConstant_0 Out AttributeStringOp_3 InA)
  (connect AttributeStringOp_2 Out AttributeStringOp_3 InB)
  (connect AttributeBooleanOp_1 Out Branch_0 bOutputToB)
  (connect AttributeBooleanOp_1 Out Branch_0 In)
  (connect DefaultInputNode Points Branch_1 In)
  (connect AddAttribute_2 Out CopyAttributes_0 Target)
  (connect AttributeStringOp_3 Out CreateTargetActor_0 ActorLabel)
  (connect DeleteAttributes_0 Out CreateTargetActor_0 ActorPivot)
  (connect Branch_0 Output%20B CreateTargetActor_0 Execution%20Dependency)
  (connect DefaultInputNode HLODLayers CreateTargetActor_0 HLODLayer)
  (connect AttributeFilter_2 InsideFilter DataNum_0 In)
  (connect AttributeFilter_3 InsideFilter DataNum_0 In)
  (connect PointToAttributeSet_0 Out DeleteAttributes_0 In)
  (connect GetActorData_1 Out PointToAttributeSet_0 In)
  (connect AttributeFilter_1 OutsideFilter SkinnedMeshSpawner_0 In)
  (connect CreateTargetActor_0 Out SkinnedMeshSpawner_0 TargetActor)
  (connect AttributeFilter_1 InsideFilter SkinnedMeshSpawner_1 In)
  (connect CreateTargetActor_0 Out SkinnedMeshSpawner_1 TargetActor)
  (connect AttributeFilter_3 InsideFilter StaticMeshSpawner_0 In)
  (connect CreateTargetActor_0 Out StaticMeshSpawner_0 TargetActor)
  (output Out SkinnedMeshSpawner_0)
  (output Out SkinnedMeshSpawner_1)
  (output Out StaticMeshSpawner_0))
