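; PCG subgraph in Biome Core that tags incoming attribute sets with their generator information.
; It takes one set by the default index and isolates its Generator, GeneratorType and GeneratorSubType attributes.
; Three chained Add Tags nodes apply these values to all input sets.
; Generator and type values split on whitespace, while the subtype has comma-space separators collapsed to commas and skips whitespace tokenizing.
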(graph BiomeCore_TagAssetsWithGenerators
  (inputs (AttributeSetsToTag :type Param))
  (outputs (TaggedAttributeSets :type Param))
  (bind GetAttributeFromIndex_2 (Get%20Attribute%20Set%20from%20Index))
  (bind FilterAttribute_4 (Delete%20Attributes :bTokenizeOnWhiteSpace true :Operation KeepSelectedAttributes :SelectedAttributes "GeneratorType"))
  (bind DeleteAttributes_1 (Delete%20Attributes :bTokenizeOnWhiteSpace true :Operation KeepSelectedAttributes :SelectedAttributes "Generator"))
  (bind AddTags_4 (Add%20Tags :bTokenizeOnWhiteSpace true))
  (bind AddTags_5 (Add%20Tags :bTokenizeOnWhiteSpace true))
  (bind CreateAttribute_1 (Create%20Constant :AttributeTypes (struct :Type String :StringValue ", ")))
  (bind CreateAttribute_3 (Create%20Constant :AttributeTypes (struct :Type String :StringValue ",")))
  (bind DeleteAttributes_5 (Delete%20Attributes :bTokenizeOnWhiteSpace true :Operation KeepSelectedAttributes :SelectedAttributes "GeneratorSubType"))
  (bind AttributeStringOp_0 (Attribute%20String%20Op :Operation Replace))
  (bind AddTags_3 (Add%20Tags))
  (connect AddTags_5 Out AddTags_3 In)
  (connect AttributeStringOp_0 Out AddTags_3 TagsToAdd)
  (connect DefaultInputNode AttributeSetsToTag AddTags_4 In)
  (connect DeleteAttributes_1 Out AddTags_4 TagsToAdd)
  (connect AddTags_4 Out AddTags_5 In)
  (connect FilterAttribute_4 Out AddTags_5 TagsToAdd)
  (connect DeleteAttributes_5 Out AttributeStringOp_0 In)
  (connect CreateAttribute_3 Out AttributeStringOp_0 Replace)
  (connect CreateAttribute_1 Out AttributeStringOp_0 Search)
  (connect GetAttributeFromIndex_2 Out DeleteAttributes_1 In)
  (connect GetAttributeFromIndex_2 Out DeleteAttributes_5 In)
  (connect GetAttributeFromIndex_2 Out FilterAttribute_4 In)
  (connect DefaultInputNode AttributeSetsToTag GetAttributeFromIndex_2 In)
  (output TaggedAttributeSets AddTags_3))
