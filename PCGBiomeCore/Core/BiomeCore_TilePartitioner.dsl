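; Biome Core PCG subgraph that splits input points into separate data sets by the world-grid tile they fall in.
; It quantizes position XY relative to PartitionWorldOrigin by TileSize and floors it into a temporary Tile attribute.
; An optional branch flips Tile Y within TileCountY.
; TileIndex is computed as Tile.x * TileCountX + Tile.y, then the Tile attribute is deleted and points are partitioned by TileIndex.
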(graph BiomeCore_TilePartitioner
  (params (TileSize :type Double :default 2.56e+04) (TileCountX :type Int :default 1) (TileCountY :type Int :default 1) (PartitionWorldOrigin :type Vector) (FlipTilesVerticalOrdering :type Bool))
  (inputs (Points%20To%20Partition :type Point))
  (outputs (Tile%20Partitioned%20Points :type Point))
  (bind Branch_0 (Branch :bOutputToB-param FlipTilesVerticalOrdering))
  (bind FilterAttributesByName_4 (Delete%20Attributes :bTokenizeOnWhiteSpace true :SelectedAttributes "Tile"))
  (bind AttributePartition_4 (Attribute%20Partition :bTokenizeOnWhiteSpace true :PartitionAttributeSelectors [TileIndex]))
  ;[0 0 0] Quantization
  (bind AttributeMathsOp_0 (Attribute%20Maths%20Op :InputSource1 $Position.xy :Operation Subtract :OutputTarget Tile :InB PartitionWorldOrigin))
  (bind AttributeMathsOp_1 (Attribute%20Maths%20Op :InputSource1 Tile :Operation Divide :OutputTarget Tile :InB TileSize))
  (bind AttributeMathsOp_3 (Attribute%20Maths%20Op :InputSource1 Tile :Operation Floor :OutputTarget Tile))
  ;[1 0 0 0.15] Flip Tile Y
  (bind CreateAttribute_19 (Create%20Constant :AttributeTypes (struct :Type Integer32 :Int32Value 1 :DoubleValue 1.0)))
  (bind AttributeMathsOp_21 (Attribute%20Maths%20Op :Operation Subtract :InA TileCountY))
  (bind AttributeMathsOp_2 (Attribute%20Maths%20Op :Operation Max))
  (bind AttributeMathsOp_8 (Attribute%20Maths%20Op :InputSource1 Tile.y :Operation Divide :OutputTarget Tile.y))
  (bind AttributeMathsOp_22 (Attribute%20Maths%20Op :InputSource1 Tile.y :Operation OneMinus))
  (bind AttributeMathsOp_11 (Attribute%20Maths%20Op :InputSource1 Tile.y :Operation Multiply))
  ;[0 0 0] Tile Index
  (bind AttributeMathsOp_5 (Attribute%20Maths%20Op :InputSource1 Tile.x :Operation Multiply :OutputTarget TileIndex :InB TileCountX))
  (bind AttributeMathsOp_6 (Attribute%20Maths%20Op :InputSource1 TileIndex :InputSource2 Tile.y :OutputTarget TileIndex))
  (connect DefaultInputNode Points%20To%20Partition AttributeMathsOp_0 InA)
  (connect AttributeMathsOp_0 Out AttributeMathsOp_1 InA)
  (connect AttributeMathsOp_21 Out AttributeMathsOp_2 InA)
  (connect CreateAttribute_19 Out AttributeMathsOp_2 InB)
  (connect AttributeMathsOp_1 Out AttributeMathsOp_3 In)
  (connect AttributeMathsOp_11 Out AttributeMathsOp_5 InA)
  (connect Branch_0 Output%20A AttributeMathsOp_5 InA)
  (connect AttributeMathsOp_5 Out AttributeMathsOp_6 InA)
  (connect AttributeMathsOp_5 Out AttributeMathsOp_6 InB)
  (connect Branch_0 Output%20B AttributeMathsOp_8 InA)
  (connect AttributeMathsOp_2 Out AttributeMathsOp_8 InB)
  (connect AttributeMathsOp_22 Out AttributeMathsOp_11 InA)
  (connect AttributeMathsOp_2 Out AttributeMathsOp_11 InB)
  (connect CreateAttribute_19 Out AttributeMathsOp_21 InB)
  (connect AttributeMathsOp_8 Out AttributeMathsOp_22 In)
  (connect FilterAttributesByName_4 Out AttributePartition_4 In)
  (connect AttributeMathsOp_3 Out Branch_0 In)
  (connect AttributeMathsOp_6 Out FilterAttributesByName_4 In)
  (output Tile%20Partitioned%20Points AttributePartition_4))
