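; PCG subgraph for Biome Core that filters runtime assembly paths down to those that load as PCG data assets with Root-tagged data.
; It merges the input with a default entry, removes duplicate hashes, drops empty and "_MS" paths, then sorts by hash.
; Loading tags each result with its source index; indices from Root-tagged data select the surviving entries, whose paths are output as soft object paths.
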(graph BiomeCore_ValidateAssetPaths
  (bind CreateAttribute_1 (Create%20Constant :AttributeTypes (struct :Type String) :OutputTarget RuntimeAssembly))
  (bind AddAttribute_0 (Add%20Attribute :AttributeTypes (struct :Type Integer32) :OutputTarget RuntimeAssemblyHashValue))
  (bind MergeAttributes_1 (Merge%20Attributes))
  (bind AttributeRemoveDuplicates_0 (Attribute%20Remove%20Duplicates :AttributeSelectors [RuntimeAssemblyHashValue]))
  (bind AttributeFilter_1 (Filter%20Attribute%20Elements :AttributeTypes (struct :Type String) :bUseConstantThreshold true :Operator NotEqual :TargetAttribute RuntimeAssembly :ThresholdAttribute @Last))
  (bind AttributeFilter_2 (Filter%20Attribute%20Elements :AttributeTypes (struct :Type String :StringValue "_MS") :bUseConstantThreshold true :Operator NotEqual :TargetAttribute RuntimeAssembly :ThresholdAttribute @Last))
  (bind MergeAttributes_2 (Merge%20Attributes))
  (bind SortAttributes_0 (Sort%20Attributes :bUseStableSort false :SortAttributes [(struct :InputSource RuntimeAssemblyHashValue)]))
  (bind PCGLoadDataAsset_0 (Load%20PCG%20Data%20Asset :AssetReferenceSelector RuntimeAssembly :bLoadFromInput true :bWarnIfNoAsset false :DataIndexTag data))
  (bind FilterDataByTag_0 (Filter%20Data%20By%20Tag :SelectedTags "Root"))
  (bind DeleteTags_0 (Delete%20Tags :Operation KeepOnlySelectedTags :Operator Substring :SelectedTags "Data"))
  (bind TagsToAttributeSet_0 (Data%20Tags%20To%20Attribute%20Set))
  (bind AttributeCast_1 (Attribute%20Cast :OutputType Integer32))
  (bind MergeAttributes_0 (Merge%20Attributes))
  (bind AttributeReduce_0 (Attribute%20Reduce :Operation Join))
  (bind FilterElementsByIndex_0 (Filter%20Elements%20By%20Index :bSelectIndicesByInput false :IndexSelectionAttribute $Index))
  (bind AttributeCast_0 (Attribute%20Cast :InputSource RuntimeAssembly :OutputType SoftObjectPath))
  (connect CreateAttribute_1 Out AddAttribute_0 In)
  (connect FilterElementsByIndex_0 In%20Filter AttributeCast_0 In)
  (connect TagsToAttributeSet_0 Out AttributeCast_1 In)
  (connect AttributeRemoveDuplicates_0 Out AttributeFilter_1 In)
  (connect AttributeFilter_1 InsideFilter AttributeFilter_2 In)
  (connect MergeAttributes_0 Out AttributeReduce_0 In)
  (connect MergeAttributes_1 Out AttributeRemoveDuplicates_0 In)
  (connect FilterDataByTag_0 InsideFilter DeleteTags_0 In)
  (connect PCGLoadDataAsset_0 Out FilterDataByTag_0 In)
  (connect SortAttributes_0 Out FilterElementsByIndex_0 In)
  (connect AttributeReduce_0 Out FilterElementsByIndex_0 SelectedIndices)
  (connect AttributeCast_1 Out MergeAttributes_0 In)
  (connect AddAttribute_0 Out MergeAttributes_1 In)
  (connect DefaultInputNode In MergeAttributes_1 In2)
  (connect AttributeFilter_2 InsideFilter MergeAttributes_2 In)
  (connect SortAttributes_0 Out PCGLoadDataAsset_0 In)
  (connect MergeAttributes_2 Out SortAttributes_0 In)
  (connect DeleteTags_0 Out TagsToAttributeSet_0 In)
  (output Out AttributeCast_0))
