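; PCG example and placeholder graph that builds a local biome cache point from an actor volume's bounds, logging an error that a valid LocalBiomeCache graph is required.
; It selects between the BiomeDefinition read from the actor's Definition and its DefaultDefinition, then copies those attributes onto the bounds point.
; Point color is set from BiomeColor, then cast to string, combined with the actor reference and hashed into BiomeID.
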(graph BiomeLocalCache
  (outputs (LocalBiomeCache :type Point))
  (bind Print%20String_0 (Print%20String :bPrefixWithNode false :bPrefixWithOwner true :bPrintToScreen true :PrintString "A valid LocalBiomeCache graph is required" :PrintToScreenColor (struct :R 255 :A 255) :PrintToScreenDuration 0.0 :Verbosity Error))
  ;[0 0 0] EXAMPLE: How to Get Actor Volume Bounds and assign Biome for the expected Bounds Output
  (bind GetActorData_0 (Get%20Actor%20Data :Mode GetSinglePoint))
  (bind GetActorProperty_1 (Get%20Actor%20Property :bAlwaysRequeryActors false :bSanitizeOutputAttributeName false :PropertyName Definition))
  (bind GetPropertyFromObjectPath_0 (Get%20Property%20From%20Object%20Path :bSanitizeOutputAttributeName false :ObjectExtractorBehavior Extract :PropertyName BiomeDefinition :StructExtractorBehavior Extract))
  (bind ApplyScaleToBounds_0 (Apply%20Scale%20To%20Bounds))
  (bind AttributeCompareOp_1 (Attribute%20Compare%20Op))
  (bind GetActorProperty_2 (Get%20Actor%20Property :bAlwaysRequeryActors false :bSanitizeOutputAttributeName false :PropertyName DefaultDefinition))
  (bind Select_1 (Select))
  (bind CopyAttributes_0 (Copy%20Attributes :bCopyAllAttributes true :InputSource @Last))
  (bind AttributeMathsOp_1 (Attribute%20Maths%20Op :InputSource1 $Color :InputSource2 BiomeColor :Operation Set))
  (bind GetActorData_1 (Get%20Actor%20Data :Mode GetActorReference))
  (bind AttributeCast_0 (Attribute%20Cast :InputSource $Color :OutputTarget BiomeID :OutputType String))
  (bind AttributeStringOp_0 (Attribute%20String%20Op))
  (bind HashAttribute_0 (Hash%20Attribute :InputSource1 BiomeID :OutputTarget BiomeID))
  (connect GetActorData_0 Out ApplyScaleToBounds_0 In)
  (connect AttributeMathsOp_1 Out AttributeCast_0 In)
  (connect GetActorProperty_1 Out AttributeCompareOp_1 InA)
  (connect CopyAttributes_0 Out AttributeMathsOp_1 InA)
  (connect CopyAttributes_0 Out AttributeMathsOp_1 InB)
  (connect AttributeCast_0 Out AttributeStringOp_0 InA)
  (connect GetActorData_1 Out AttributeStringOp_0 InB)
  (connect Select_1 Out CopyAttributes_0 Source)
  (connect ApplyScaleToBounds_0 Out CopyAttributes_0 Target)
  (connect GetActorProperty_1 Out GetPropertyFromObjectPath_0 In)
  (connect AttributeStringOp_0 Out HashAttribute_0 In)
  (connect DefaultInputNode In Print%20String_0 In)
  (connect AttributeCompareOp_1 Out Select_1 bUseInputB)
  (connect GetPropertyFromObjectPath_0 Out Select_1 Input%20A)
  (connect GetActorProperty_2 Out Select_1 Input%20B)
  (output LocalBiomeCache HashAttribute_0))
