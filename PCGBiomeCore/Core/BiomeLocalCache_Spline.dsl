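; PCG biome graph that fills the interiors of tagged biome splines with surface-projected points to form a local biome cache.
; Biome actor properties drive the sampling spacing and point Z extents, and existing point tags are replaced by tags from the Definition lookup.
; The BiomeDefinition from Definition, or DefaultDefinition as fallback, is copied onto each point, setting its color from BiomeColor.
; BiomeID is a hash of that color combined with the owning actor reference.
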(graph BiomeLocalCache_Spline
  (outputs (LocalBiomeCache :type Point))
  ;[0 0 0 0.5] Get Actor Volume Bounds and assign Biome
  (bind GetActorData_1 (Get%20Actor%20Data :ComponentSelector (struct :ComponentSelectionTag PCG_BiomeActorSpline)))
  (bind GetActorProperty_1 (Get%20Actor%20Property :bAlwaysRequeryActors false :bSanitizeOutputAttributeName false :bTrackActorsOnlyWithinBounds true :PropertyName Definition))
  (bind GetTags_1 (Get%20Tags))
  (bind AttributeReduce_1 (Attribute%20Reduce :Operation Join))
  (bind FilterDataByType_0 (Filter%20Data%20By%20Type :TargetType (struct :Ids [(struct :Struct /Script/PCG.PCGDataTypeInfoPolyline)])))
  (bind GetActorProperty_2 (Get%20Actor%20Property :ActorSelector (struct :ActorSelectionTag PCG_BiomeActor :bSelectMultiple true) :bAlwaysRequeryActors false :bProcessAllComponents false :bSanitizeOutputAttributeName false :bTrackActorsOnlyWithinBounds true :OutputAttributeName LocalCacheSize :PropertyName LocalCacheSize))
  (bind SplineSampler_0 (Spline%20Sampler :SamplerParams (struct :Dimension OnInterior :InteriorSampleSpacing 8e+02 :InteriorBorderSampleSpacing 8e+02 :bProjectOntoSurface true :bUnbounded true :PointSteepness 1.0)))
  (bind DeleteTags_1 (Delete%20Tags :Operation KeepOnlySelectedTags))
  (bind AddTags_1 (Add%20Tags))
  (bind GetActorProperty_3 (Get%20Actor%20Property :ActorSelector (struct :ActorSelectionTag PCG_BiomeActor :bSelectMultiple true) :bAlwaysRequeryActors false :bProcessAllComponents false :bSanitizeOutputAttributeName false :bTrackActorsOnlyWithinBounds true :OutputAttributeName LocalCacheZExtents :PropertyName LocalCacheZExtents))
  (bind AttributeMathsOp_0 (Attribute%20Maths%20Op :InputSource1 $Extents.z :Operation Set))
  (bind AttributeCompareOp_1 (Attribute%20Compare%20Op))
  (bind GetActorProperty_4 (Get%20Actor%20Property :bAlwaysRequeryActors false :bSanitizeOutputAttributeName false :bTrackActorsOnlyWithinBounds true :PropertyName DefaultDefinition))
  (bind GetPropertyFromObjectPath_0 (Get%20Property%20From%20Object%20Path :bSanitizeOutputAttributeName false :ObjectExtractorBehavior Extract :PropertyName BiomeDefinition :StructExtractorBehavior Extract))
  (bind Select_1 (Select))
  (bind CopyAttributes_1 (Copy%20Attributes :bCopyAllAttributes true :InputSource @Last))
  (bind AttributeMathsOp_1 (Attribute%20Maths%20Op :InputSource1 $Color :InputSource2 BiomeColor :Operation Set))
  (bind GetActorData_2 (Get%20Actor%20Data :Mode GetActorReference))
  (bind AttributeCast_0 (Attribute%20Cast :InputSource $Color :OutputTarget BiomeID :OutputType String))
  (bind AttributeStringOp_0 (Attribute%20String%20Op))
  (bind HashAttribute_0 (Hash%20Attribute :InputSource1 BiomeID :OutputTarget BiomeID))
  (connect DeleteTags_1 Out AddTags_1 In)
  (connect AttributeReduce_1 Out AddTags_1 TagsToAdd)
  (connect AttributeMathsOp_1 Out AttributeCast_0 In)
  (connect GetActorProperty_1 Out AttributeCompareOp_1 InA)
  (connect AddTags_1 Out AttributeMathsOp_0 InA)
  (connect GetActorProperty_3 Out AttributeMathsOp_0 InB)
  (connect CopyAttributes_1 Out AttributeMathsOp_1 InA)
  (connect CopyAttributes_1 Out AttributeMathsOp_1 InB)
  (connect GetTags_1 Out AttributeReduce_1 In)
  (connect AttributeCast_0 Out AttributeStringOp_0 InA)
  (connect GetActorData_2 Out AttributeStringOp_0 InB)
  (connect Select_1 Out CopyAttributes_1 Source)
  (connect AttributeMathsOp_0 Out CopyAttributes_1 Target)
  (connect SplineSampler_0 Out DeleteTags_1 In)
  (connect GetActorData_1 Out FilterDataByType_0 In)
  (connect GetActorProperty_1 Out GetPropertyFromObjectPath_0 In)
  (connect GetActorProperty_1 Out GetTags_1 In)
  (connect AttributeStringOp_0 Out HashAttribute_0 In)
  (connect AttributeCompareOp_1 Out Select_1 bUseInputB)
  (connect GetPropertyFromObjectPath_0 Out Select_1 Input%20A)
  (connect GetActorProperty_4 Out Select_1 Input%20B)
  (connect GetActorProperty_2 Out SplineSampler_0 InteriorBorderSampleSpacing)
  (connect GetActorProperty_2 Out SplineSampler_0 InteriorSampleSpacing)
  (connect FilterDataByType_0 InsideFilter SplineSampler_0 Spline)
  (output LocalBiomeCache HashAttribute_0))
