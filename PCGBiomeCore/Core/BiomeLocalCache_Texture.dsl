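; Builds a local biome point cache from an actor's biome texture, part of PCG Biome Core.
; Texture samples at the biome actor's texel size become points, take the actor's rotation and configured Z extents, and are filtered by biome color within a tolerance.
; Survivors receive attributes from the biome definition or default definition and a BiomeID hashed from color and actor reference.
; A snap-to-surface setting optionally projects them via world raycast.
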(graph BiomeLocalCache_Texture
  (outputs (LocalBiomeCache :type Point))
  ;[0 0 0 0.5] Get Actor Texture and assign Biome
  (bind GetActorData_1 (Get%20Actor%20Data :Mode GetSinglePoint))
  (bind ApplyScaleToBounds_1 (Apply%20Scale%20To%20Bounds))
  (bind GetActorProperty_1 (Get%20Actor%20Property :bAlwaysRequeryActors false :bSanitizeOutputAttributeName false :bTrackActorsOnlyWithinBounds true :OutputAttributeName BiomeTexture :PropertyName BiomeTexture))
  (bind GetActorProperty_3 (Get%20Actor%20Property :ActorSelector (struct :ActorSelectionTag PCG_BiomeActor :bSelectMultiple true) :bAlwaysRequeryActors false :bProcessAllComponents false :bSanitizeOutputAttributeName false :bTrackActorsOnlyWithinBounds true :OutputAttributeName LocalCacheTexelSize :PropertyName LocalCacheTexelSize))
  (bind GetTextureData_1 (Get%20Texture%20Data :bForceEditorOnlyCPUSampling true :bUseDensitySourceChannel false :Filter Point :TexelSize 8e+02))
  (bind ToPoint_0 (To%20Point))
  (bind GetActorProperty_5 (Get%20Actor%20Property :bAlwaysRequeryActors false :bSanitizeOutputAttributeName false :bTrackActorsOnlyWithinBounds true :PropertyName Definition))
  (bind GetTags_0 (Get%20Tags))
  (bind AttributeReduce_0 (Attribute%20Reduce :Operation Join))
  (bind DeleteTags_0 (Delete%20Tags :Operation KeepOnlySelectedTags))
  (bind AddTags_0 (Add%20Tags))
  (bind CopyAttributes_0 (Copy%20Attributes :InputSource $Rotation :OutputTarget $Rotation))
  (bind GetActorProperty_6 (Get%20Actor%20Property :ActorSelector (struct :ActorSelectionTag PCG_BiomeActor :bSelectMultiple true) :bAlwaysRequeryActors false :bProcessAllComponents false :bSanitizeOutputAttributeName false :bTrackActorsOnlyWithinBounds true :OutputAttributeName LocalCacheSnapToSurface :PropertyName LocalCacheSnapToSurface))
  (bind GetActorProperty_2 (Get%20Actor%20Property :bAlwaysRequeryActors false :bSanitizeOutputAttributeName false :bTrackActorsOnlyWithinBounds true :PropertyName BiomeColorTolerance))
  (bind AttributeCompareOp_0 (Attribute%20Compare%20Op))
  (bind GetActorProperty_7 (Get%20Actor%20Property :bAlwaysRequeryActors false :bSanitizeOutputAttributeName false :bTrackActorsOnlyWithinBounds true :PropertyName DefaultDefinition))
  (bind GetPropertyFromObjectPath_2 (Get%20Property%20From%20Object%20Path :bSanitizeOutputAttributeName false :ObjectExtractorBehavior Extract :PropertyName BiomeDefinition :StructExtractorBehavior Extract))
  (bind Select_0 (Select))
  (bind AttributeMathsOp_0 (Attribute%20Maths%20Op :InputSource1 BiomeColor.rgb))
  (bind GetActorProperty_4 (Get%20Actor%20Property :ActorSelector (struct :ActorSelectionTag PCG_BiomeActor :bSelectMultiple true) :bAlwaysRequeryActors false :bProcessAllComponents false :bSanitizeOutputAttributeName false :bTrackActorsOnlyWithinBounds true :OutputAttributeName LocalCacheZExtents :PropertyName LocalCacheZExtents))
  (bind AttributeMathsOp_1 (Attribute%20Maths%20Op :InputSource1 $Extents.z :Operation Set))
  (bind AttributeMathsOp_3 (Attribute%20Maths%20Op :InputSource1 BiomeColor.rgb :Operation Subtract))
  (bind AttributeFilterRange_1 (Filter%20Attribute%20Elements%20by%20Range :MaxThreshold (struct :ThresholdAttribute @Last :bUseSpatialQuery false) :MinThreshold (struct :ThresholdAttribute @Last :bUseSpatialQuery false) :TargetAttribute $Color.rgb))
  (bind CopyAttributes_2 (Copy%20Attributes :bCopyAllAttributes true :InputSource @Last))
  (bind AttributeMathsOp_2 (Attribute%20Maths%20Op :InputSource1 $Color :InputSource2 BiomeColor :Operation Set))
  (bind AttributeCast_1 (Attribute%20Cast :InputSource $Color :OutputTarget BiomeID :OutputType String))
  (bind GetActorData_3 (Get%20Actor%20Data :Mode GetActorReference))
  (bind AttributeStringOp_1 (Attribute%20String%20Op))
  (bind HashAttribute_1 (Hash%20Attribute :InputSource1 BiomeID :OutputTarget BiomeID))
  (bind Branch_0 (Branch))
  (bind WorldRayHitQuery_0 (World%20Ray%20Hit%20Query))
  (bind Projection_0 (Projection :ProjectionParams (struct :bProjectRotations false :AttributeMergeOperation SourceValue)))
  (connect DeleteTags_0 Out AddTags_0 In)
  (connect AttributeReduce_0 Out AddTags_0 TagsToAdd)
  (connect GetActorData_1 Out ApplyScaleToBounds_1 In)
  (connect AttributeMathsOp_2 Out AttributeCast_1 In)
  (connect GetActorProperty_5 Out AttributeCompareOp_0 InA)
  (connect AttributeMathsOp_0 Out AttributeFilterRange_1 FilterMax)
  (connect AttributeMathsOp_3 Out AttributeFilterRange_1 FilterMin)
  (connect AttributeMathsOp_1 Out AttributeFilterRange_1 In)
  (connect Select_0 Out AttributeMathsOp_0 InA)
  (connect GetActorProperty_2 Out AttributeMathsOp_0 InB)
  (connect CopyAttributes_0 Out AttributeMathsOp_1 InA)
  (connect GetActorProperty_4 Out AttributeMathsOp_1 InB)
  (connect CopyAttributes_2 Out AttributeMathsOp_2 InA)
  (connect CopyAttributes_2 Out AttributeMathsOp_2 InB)
  (connect Select_0 Out AttributeMathsOp_3 InA)
  (connect GetActorProperty_2 Out AttributeMathsOp_3 InB)
  (connect GetTags_0 Out AttributeReduce_0 In)
  (connect AttributeCast_1 Out AttributeStringOp_1 InA)
  (connect GetActorData_3 Out AttributeStringOp_1 InB)
  (connect GetActorProperty_6 Out Branch_0 bOutputToB)
  (connect HashAttribute_1 Out Branch_0 In)
  (connect ApplyScaleToBounds_1 Out CopyAttributes_0 Source)
  (connect AddTags_0 Out CopyAttributes_0 Target)
  (connect Select_0 Out CopyAttributes_2 Source)
  (connect AttributeFilterRange_1 InsideFilter CopyAttributes_2 Target)
  (connect ToPoint_0 Out DeleteTags_0 In)
  (connect GetActorProperty_5 Out GetPropertyFromObjectPath_2 In)
  (connect GetActorProperty_5 Out GetTags_0 In)
  (connect GetActorProperty_3 Out GetTextureData_1 TexelSize)
  (connect GetActorProperty_1 Out GetTextureData_1 Texture)
  (connect AttributeStringOp_1 Out HashAttribute_1 In)
  (connect Branch_0 Output%20B Projection_0 In)
  (connect WorldRayHitQuery_0 Out Projection_0 Projection%20Target)
  (connect AttributeCompareOp_0 Out Select_0 bUseInputB)
  (connect GetPropertyFromObjectPath_2 Out Select_0 Input%20A)
  (connect GetActorProperty_7 Out Select_0 Input%20B)
  (connect GetTextureData_1 Out ToPoint_0 In)
  (output LocalBiomeCache Branch_0.Output%20A)
  (output LocalBiomeCache Projection_0))
