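; PCG subgraph for Biome Core that turns a biome volume actor into a local biome cache point covering its scaled bounds.
; It reads the biome from the actor's Definition asset, with a compare and select switching to DefaultDefinition, and copies those attributes onto the point.
; BiomeColor becomes the point color, which is cast to a string, combined with the actor reference and hashed into the BiomeID.
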(graph BiomeLocalCache_Volume
  (outputs (LocalBiomeCache :type Point))
  ;[0 0 0 0.5] Get Actor Volume Bounds and assign Biome
  (bind GetActorData_1 (Get%20Actor%20Data :Mode GetSinglePoint))
  (bind GetActorProperty_0 (Get%20Actor%20Property :bAlwaysRequeryActors false :bSanitizeOutputAttributeName false :PropertyName Definition))
  (bind GetPropertyFromObjectPath_1 (Get%20Property%20From%20Object%20Path :bSanitizeOutputAttributeName false :ObjectExtractorBehavior Extract :PropertyName BiomeDefinition :StructExtractorBehavior Extract))
  (bind ApplyScaleToBounds_1 (Apply%20Scale%20To%20Bounds))
  (bind AttributeCompareOp_0 (Attribute%20Compare%20Op))
  (bind GetActorProperty_1 (Get%20Actor%20Property :bAlwaysRequeryActors false :bSanitizeOutputAttributeName false :PropertyName DefaultDefinition))
  (bind Select_0 (Select))
  (bind CopyAttributes_1 (Copy%20Attributes :bCopyAllAttributes true :InputSource @Last))
  (bind AttributeMathsOp_0 (Attribute%20Maths%20Op :InputSource1 $Color :InputSource2 BiomeColor :Operation Set))
  (bind GetActorData_3 (Get%20Actor%20Data :Mode GetActorReference))
  (bind AttributeCast_1 (Attribute%20Cast :InputSource $Color :OutputTarget BiomeID :OutputType String))
  (bind AttributeStringOp_1 (Attribute%20String%20Op))
  (bind HashAttribute_1 (Hash%20Attribute :InputSource1 BiomeID :OutputTarget BiomeID))
  (connect GetActorData_1 Out ApplyScaleToBounds_1 In)
  (connect AttributeMathsOp_0 Out AttributeCast_1 In)
  (connect GetActorProperty_0 Out AttributeCompareOp_0 InA)
  (connect CopyAttributes_1 Out AttributeMathsOp_0 InA)
  (connect CopyAttributes_1 Out AttributeMathsOp_0 InB)
  (connect AttributeCast_1 Out AttributeStringOp_1 InA)
  (connect GetActorData_3 Out AttributeStringOp_1 InB)
  (connect Select_0 Out CopyAttributes_1 Source)
  (connect ApplyScaleToBounds_1 Out CopyAttributes_1 Target)
  (connect GetActorProperty_0 Out GetPropertyFromObjectPath_1 In)
  (connect AttributeStringOp_1 Out HashAttribute_1 In)
  (connect AttributeCompareOp_0 Out Select_0 bUseInputB)
  (connect GetPropertyFromObjectPath_1 Out Select_0 Input%20A)
  (connect GetActorProperty_1 Out Select_0 Input%20B)
  (output LocalBiomeCache HashAttribute_1))
